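# Item layer Anim Blueprint for the FeyCrab Launcher on the Fortnite player skeleton, adding hand-specific aim offsets to the final pose.
# Each update reads the launcher's hand and enables aiming only while a montage plays and the player is not ziplining.
# ItemFinalPoseOverride, the only layer with logic, selects the right or left aim offset over the cached pose, or passes it through, smoothing switches with inertialization.
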AnimBlueprint /DryBox/Items/FeyCrab/Animation/AnimBP/FeyCrab_Launcher_LayerAnimBP
target_skeleton=&Fortnite_M_Avg_Player_Skeleton
graphs [
 AnimGraph {nodes [Output Pose() {layerGroup=None}]}
 EventGraph {nodes [
   Event BlueprintUpdateAnimation()
   n1 = Get FortPlayerPawn() -> (FortPlayerPawn:Fort Player Pawn Object Reference)
   > n2 = Cast To B_FeyCrab_Launcher!(execute, Object=Get CurrentWeapon!(self:Fort Pawn Object Reference=@n1) -> (CurrentWeapon:Fort Weapon Object Reference)) -> (then, CastFailed, AsB Fey Crab Launcher:B Fey Crab Launcher Object Reference) {is=title K2Node_DynamicCast}
   > Set Is Right Hand(Is Right Hand:bool=Get IsRightHand!(self:B Fey Crab Launcher Object Reference=@n2) -> (IsRightHand:bool)) -> (Output_Get:bool)
   > n3 = IsZiplining!(execute, self:Fort Player Pawn Object Reference=@n1) -> (then, ReturnValue:bool)
   > Set IsUseAimoffset(IsUseAimoffset:bool=AND Boolean(A=Get bIsMontagePlaying() -> (bIsMontagePlaying:bool), B=NOT Boolean(A=@n3))) -> (Output_Get:bool)
  ]}
 ItemAimOffset {nodes [
   Input Pose!() -> (Pose, AimOffsetAlpha:double, Yaw:double, Pitch:double) {name=InPose_AimOffset}
   Input Pose!() -> (Pose) {name=InPose_PreAimOffsets}
   Output Pose() {layerGroup=Item Override}
  ]}
 ItemFallAdditive {nodes [
   Input Pose() {name=InPoseItemFallAdditive}
   Output Pose() {layerGroup=Item Override}
  ]}
 ItemFallLandAdditive {nodes [
   Input Pose() {name=InPoseItemFallLandAdditive}
   Output Pose() {layerGroup=Item Override}
  ]}
 ItemFallLevel2Override {nodes [
   Input Pose() {name=InPose_FallLevel2}
   Output Pose() {layerGroup=Item Override}
  ]}
 ItemFinalPoseOverride {nodes [
   n1 = Blend Poses by bool(
    BlendPose_0=AimOffset Player 'FeyCrab_AO'!(X:float=Get AimYaw() -> (AimYaw:float), Y:float=Get AimPitch(), BasePose:Pose Link Structure=Use cached pose 'FinalPose'(), Alpha:float=1) -> (Pose:Pose Link Structure) {blendSpace=/DryBox/Items/FeyCrab/Animation/Offsets/FeyCrab_AO, alphaBoolBlend {blendOption=Linear}}
    BlendPose_1=AimOffset Player 'L_FeyCrab_AO'!(X:float=Get AimYaw(), Y:float=Get AimPitch(), BasePose:Pose Link Structure=Use cached pose 'FinalPose'(), Alpha:float=1) -> (Pose:Pose Link Structure) {blendSpace=/DryBox/Items/FeyCrab/Animation/Offsets/L_FeyCrab_AO, alphaBoolBlend {blendOption=Linear}}
    bActiveValue=Get Is Right Hand() -> (Is Right Hand:bool)
   ) {blendType=Linear, transitionType=Inertialization}
   n2 = Use cached pose 'FinalPose'()
   n3 = Get IsUseAimoffset() -> (IsUseAimoffset:bool)
   Output Pose(Result=Inertialization!(Source:Pose Link Structure=Blend Poses by bool(BlendPose_0=@n1, BlendPose_1=@n2, bActiveValue=@n3) {blendType=Linear, transitionType=Inertialization}) -> (Pose:Pose Link Structure) {is=title AnimGraphNode_Inertialization}) {layerGroup=Item Override}
   Save cached pose 'FinalPose'(Pose=Input Pose() {name=InFinalPose}) {cacheName=FinalPose}
  ]}
 ItemFullBodyAimOffsets {nodes [
   Input Pose!() -> (Pose, FullBody_AimOffsetAlpha:double, Aim_Yaw:double, Aim_Pitch:double) {name=InPoseFullBodyAimOffsets}
   Input Pose!() -> (Pose) {name=InPosePreFullBodyAimOffsets}
   Output Pose() {layerGroup=Item Override}
  ]}
 ItemFullBodyOverride {nodes [
   Input Pose() {name=InPoseFullBody}
   Output Pose() {layerGroup=Item Override}
  ]}
 ItemIdleOverride {nodes [
   Input Pose() {name=InPose_Idle}
   Output Pose() {layerGroup=Item Override}
  ]}
 ItemJetPackJumpAdditive {nodes [Output Pose() {layerGroup=Item Override}]}
 ItemJetPackStartAdditive {nodes [Output Pose() {layerGroup=Item Override}]}
 ItemJumpLoopAdditive {nodes [
   Input Pose() {name=InPoseItemJumpLoopAdditive}
   Output Pose() {layerGroup=Item Override}
  ]}
 ItemJumpOverride {nodes [
   Input Pose() {name=InPoseJump}
   Output Pose() {layerGroup=Item Override}
  ]}
 ItemJumpUpAdditive {nodes [
   Input Pose() {name=InPoseItemJumpUpAdditive}
   Output Pose() {layerGroup=Item Override}
  ]}
 ItemPreAimOffsets {nodes [
   Input Pose() {name=InPosePreAimOffsets}
   Output Pose() {layerGroup=Item Override}
  ]}
 ItemPreIK {nodes [
   Input Pose() {name=InPose_PreIK}
   Output Pose() {layerGroup=Item Override}
  ]}
 ItemRightHandIK {nodes [
   Input Pose() {name=InPosePreRightHandIK}
   Input Pose() {name=InPosePostRightHandIK}
   Output Pose() {layerGroup=Item Override}
  ]}
 ItemSkeletalControl {nodes [
   Input Pose()
   Output Pose() {layerGroup=Item Override}
  ]}
 ItemSwimJumpFallLoop {nodes [Output Pose() {layerGroup=Item Override}]}
 ItemSwimJumpSurfaceEnd {nodes [Output Pose() {layerGroup=Item Override}]}
 ItemSwimJumpSurfaceLoop {nodes [Output Pose() {layerGroup=Item Override}]}
 ItemUpperBody {nodes [
   Input Pose!() -> (Pose, InputParam:Fort Anim Input Adjusted Aim Structure) {name=InPoseUpperBody}
   Output Pose() {layerGroup=Item Override}
  ]}
 ItemZipLineStartAdditive {nodes [Output Pose() {layerGroup=Item Override}]}
]
variables [
 Is Right Hand {type=boolean}
 IsUseAimoffset {type=boolean}
]
