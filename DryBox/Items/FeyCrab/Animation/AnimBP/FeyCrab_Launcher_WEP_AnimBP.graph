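# Weapon AnimBlueprint for the FeyCrab Launcher that animates its spears through equip, fire, empty and reload.
# IsRightHand, not set in this graph, selects the RightHand or LeftHand state machine, which share sequences and route through conduits.
# Each update copies reload, equip, hand and ammo values from the owning launcher to drive transitions, with left emptying below two rounds, right at zero.
# Idle and Empty poses are frozen frames of the fire animation.
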AnimBlueprint /DryBox/Items/FeyCrab/Animation/AnimBP/FeyCrab_Launcher_WEP_AnimBP
target_skeleton=/DryBox/Items/FeyCrab/Meshes/FeyCrab_Launcher_Skeleton
graphs [
 AnimGraph {nodes [
   AnimGraphNode_StateMachine_0 = RightHand() {is=title AnimGraphNode_StateMachine, bAllowConduitEntryStates=true}
   AnimGraphNode_StateMachine_2 = LeftHand() {is=title AnimGraphNode_StateMachine, bAllowConduitEntryStates=true}
   Output Pose(Result=Blend Poses by bool(BlendPose_0=@AnimGraphNode_StateMachine_0, BlendPose_1=@AnimGraphNode_StateMachine_2, bActiveValue=Get IsRightHand() -> (IsRightHand:bool)) {blendType=Linear, childUpateMode=ResetChildOnActivate}) {layerGroup=None}
   Slot 'DefaultSlot'()
  ]}
 RightHand {
  entry=Conduit
  path=AnimGraph.AnimGraphNode_StateMachine_0.RightHand
  ids=id
  nodes [
   Alias() {is=title AnimStateAliasNode, also {aliases [Empty_Right,Fire_Right,Idle_Right]}}
   AnimStateConduitNode_0 = Conduit()
   AnimStateNode_0 = Equip() {is=title AnimStateNode}
   AnimStateNode_1 = Fire_Right() {is=title AnimStateNode}
   AnimStateNode_2 = Reload() {is=title AnimStateNode, bAlwaysResetOnEntry=true}
   AnimStateNode_3 = Empty_Right() {is=title AnimStateNode}
   AnimStateNode_4 = Idle_Right() {is=title AnimStateNode}
   AnimStateTransitionNode_0 = AnimStateTransitionNode() {priorityOrder=100}
   AnimStateTransitionNode_10 = AnimStateTransitionNode() {priorityOrder=2}
   AnimStateTransitionNode_13 = AnimStateTransitionNode()
   AnimStateTransitionNode_2 = AnimStateTransitionNode() {bAutomaticRuleBasedOnSequencePlayerInState=true}
   AnimStateTransitionNode_3 = AnimStateTransitionNode()
   AnimStateTransitionNode_4 = AnimStateTransitionNode() {bAutomaticRuleBasedOnSequencePlayerInState=true}
   AnimStateTransitionNode_5 = AnimStateTransitionNode()
   AnimStateTransitionNode_6 = AnimStateTransitionNode()
   AnimStateTransitionNode_7 = AnimStateTransitionNode()
   AnimStateTransitionNode_8 = AnimStateTransitionNode() {crossfadeDuration=0, priorityOrder=3}
  ]
  links [
   {
    dst=Reload
    rule=AnimGraph.AnimGraphNode_StateMachine_0.RightHand.AnimStateTransitionNode_5.Transition
    src=Alias
    via=AnimStateTransitionNode_5
   }
   {
    dst=Empty_Right
    rule=AnimGraph.AnimGraphNode_StateMachine_0.RightHand.AnimStateTransitionNode_8.Transition
    src=Conduit
    via=AnimStateTransitionNode_8
   }
   {
    dst=Equip
    rule=AnimGraph.AnimGraphNode_StateMachine_0.RightHand.AnimStateTransitionNode_10.Transition
    src=Conduit
    via=AnimStateTransitionNode_10
   }
   {
    dst=Idle_Right
    rule=AnimGraph.AnimGraphNode_StateMachine_0.RightHand.AnimStateTransitionNode_0.Transition
    src=Conduit
    via=AnimStateTransitionNode_0
   }
   {
    dst=Reload
    rule=AnimGraph.AnimGraphNode_StateMachine_0.RightHand.AnimStateTransitionNode_6.Transition
    src=Conduit
    via=AnimStateTransitionNode_6
   }
   {
    dst=Idle_Right
    rule=AnimGraph.AnimGraphNode_StateMachine_0.RightHand.AnimStateTransitionNode_13.Transition
    src=Empty_Right
    via=AnimStateTransitionNode_13
   }
   {
    dst=Conduit
    rule=AnimGraph.AnimGraphNode_StateMachine_0.RightHand.AnimStateTransitionNode_2.Transition
    src=Equip
    via=AnimStateTransitionNode_2
   }
   {
    dst=Empty_Right
    rule=AnimGraph.AnimGraphNode_StateMachine_0.RightHand.AnimStateTransitionNode_4.Transition
    src=Fire_Right
    via=AnimStateTransitionNode_4
   }
   {
    dst=Fire_Right
    rule=AnimGraph.AnimGraphNode_StateMachine_0.RightHand.AnimStateTransitionNode_7.Transition
    src=Idle_Right
    via=AnimStateTransitionNode_7
   }
   {
    dst=Conduit
    rule=AnimGraph.AnimGraphNode_StateMachine_0.RightHand.AnimStateTransitionNode_3.Transition
    src=Reload
    via=AnimStateTransitionNode_3
   }
  ]
 }
 Conduit {path=AnimGraph.AnimGraphNode_StateMachine_0.RightHand.AnimStateConduitNode_0.Conduit, nodes [Result(bCanEnterTransition=true)]}
 Equip {
  path=AnimGraph.AnimGraphNode_StateMachine_0.RightHand.AnimStateNode_0.Equip
  nodes [Output Animation Pose(Result=Sequence Player 'SpearOnly_Equip_Right'() {bLoopAnimation=false, sequence=/DryBox/Items/FeyCrab/Animation/SpearOnly_Equip_Right})]
 }
 Fire_Right {
  path=AnimGraph.AnimGraphNode_StateMachine_0.RightHand.AnimStateNode_1.Fire_Right
  nodes [Output Animation Pose(Result=Sequence Player 'TS_FireLeft_RootAnim_SpearLeft'() {bLoopAnimation=false, sequence=/DryBox/Items/FeyCrab/Animation/TS_FireLeft_RootAnim_SpearLeft})]
 }
 Reload {
  path=AnimGraph.AnimGraphNode_StateMachine_0.RightHand.AnimStateNode_2.Reload
  nodes [Output Animation Pose(Result=Sequence Player 'SpearOnly_Reload_Right'() {bLoopAnimation=false, sequence=/DryBox/Items/FeyCrab/Animation/SpearOnly_Reload_Right})]
 }
 Empty_Right {
  path=AnimGraph.AnimGraphNode_StateMachine_0.RightHand.AnimStateNode_3.Empty_Right
  nodes [Output Animation Pose(Result="Sequence Evaluator (by time) 'TS_FireLeft_RootAnim_SpearLeft'"!(ExplicitTime:float=0.82) -> (Pose:Pose Link Structure) {is=title AnimGraphNode_SequenceEvaluator, sequence=&TS_FireLeft_RootAnim_SpearLeft})]
 }
 Idle_Right {
  path=AnimGraph.AnimGraphNode_StateMachine_0.RightHand.AnimStateNode_4.Idle_Right
  nodes [Output Animation Pose(Result="Sequence Evaluator (by time) 'TS_FireLeft_RootAnim_SpearLeft'"(ExplicitTime=0) {sequence=&TS_FireLeft_RootAnim_SpearLeft})]
 }
 Transition {path=AnimGraph.AnimGraphNode_StateMachine_0.RightHand.AnimStateTransitionNode_0.Transition, nodes [Result(bCanEnterTransition=true)]}
 Transition {
  path=AnimGraph.AnimGraphNode_StateMachine_0.RightHand.AnimStateTransitionNode_10.Transition
  nodes [
   n1 = Get Magazine Ammo Count() -> (Magazine Ammo Count:int=0)
   Result(bCanEnterTransition=AND Boolean(A=Get IsEquipping() -> (IsEquipping:bool), B="integer > integer"(A=@n1)))
  ]
 }
 Transition {
  path=AnimGraph.AnimGraphNode_StateMachine_0.RightHand.AnimStateTransitionNode_13.Transition
  nodes [
   n1 = Get Magazine Ammo Count() -> (Magazine Ammo Count=0)
   Result(bCanEnterTransition=NOT Boolean(A="Equal (Integer)"(A=@n1)))
  ]
 }
 Transition {path=AnimGraph.AnimGraphNode_StateMachine_0.RightHand.AnimStateTransitionNode_2.Transition, nodes [Result()]}
 Transition {
  path=AnimGraph.AnimGraphNode_StateMachine_0.RightHand.AnimStateTransitionNode_3.Transition
  nodes [
   n1 = NOT Boolean(A=Get IsReloading() -> (IsReloading:bool))
   n2 = "Equal (Integer)"(A=Get Magazine Ammo Count() -> (Magazine Ammo Count=0), B=1)
   n3 = Get Magazine Ammo Count() -> (Magazine Ammo Count=0)
   Result(bCanEnterTransition=OR Boolean!(A="Equal (Integer)"(A=@n3, B=2), B=Get IsEquipping(), C:bool=AND Boolean(A=@n1, B=@n2)) -> (ReturnValue))
  ]
 }
 Transition {path=AnimGraph.AnimGraphNode_StateMachine_0.RightHand.AnimStateTransitionNode_4.Transition, nodes [Result()]}
 Transition {
  path=AnimGraph.AnimGraphNode_StateMachine_0.RightHand.AnimStateTransitionNode_5.Transition
  nodes [Result(bCanEnterTransition=Get IsReloading())]
 }
 Transition {
  path=AnimGraph.AnimGraphNode_StateMachine_0.RightHand.AnimStateTransitionNode_6.Transition
  nodes [
   n1 = Get Magazine Ammo Count() -> (Magazine Ammo Count=0)
   Result(bCanEnterTransition=AND Boolean(A=Get IsReloading(), B="Equal (Integer)"(A=@n1)))
  ]
 }
 Transition {
  path=AnimGraph.AnimGraphNode_StateMachine_0.RightHand.AnimStateTransitionNode_7.Transition
  nodes [Result(bCanEnterTransition=Get IsRightFire() -> (IsRightFire:bool))]
 }
 Transition {
  path=AnimGraph.AnimGraphNode_StateMachine_0.RightHand.AnimStateTransitionNode_8.Transition
  nodes [Result(bCanEnterTransition="Equal (Integer)"(A=Get Magazine Ammo Count() -> (Magazine Ammo Count=0)))]
 }
 LeftHand {
  entry=Conduit
  path=AnimGraph.AnimGraphNode_StateMachine_2.LeftHand
  ids=id
  nodes [
   Alias() {also {aliases [Empty_Left,Fire_Left,Idle_Left]}}
   AnimStateConduitNode_0 = Conduit()
   AnimStateNode_0 = IsEquipping() {is=title AnimStateNode}
   AnimStateNode_1 = Fire_Left() {is=title AnimStateNode}
   AnimStateNode_2 = Reload()
   AnimStateNode_3 = Empty_Left() {
    is=title AnimStateNode
    stateEntered {guid=#1}
    stateFullyBlended {guid=#2}
    stateLeft {guid=#3}
   }
   AnimStateNode_4 = Idle_Left() {
    is=title AnimStateNode
    stateEntered {guid=#4}
    stateFullyBlended {guid=#5}
    stateLeft {guid=#6}
   }
   AnimStateTransitionNode_1 = AnimStateTransitionNode()
   AnimStateTransitionNode_12 = AnimStateTransitionNode() {priorityOrder=2}
   AnimStateTransitionNode_14 = AnimStateTransitionNode() {
    transitionEnd {guid=#7}
    transitionInterrupt {guid=#8}
    transitionStart {guid=#9}
   }
   AnimStateTransitionNode_2 = AnimStateTransitionNode() {
    crossfadeDuration=0
    priorityOrder=3
    transitionEnd {guid=#10}
    transitionInterrupt {guid=#11}
    transitionStart {guid=#12}
   }
   AnimStateTransitionNode_3 = AnimStateTransitionNode() {priorityOrder=100}
   AnimStateTransitionNode_4 = AnimStateTransitionNode() {bAutomaticRuleBasedOnSequencePlayerInState=true}
   AnimStateTransitionNode_5 = AnimStateTransitionNode() {bAutomaticRuleBasedOnSequencePlayerInState=true}
   AnimStateTransitionNode_6 = AnimStateTransitionNode() {crossfadeDuration=0}
   AnimStateTransitionNode_7 = AnimStateTransitionNode()
   AnimStateTransitionNode_8 = AnimStateTransitionNode()
   AnimStateTransitionNode_9 = AnimStateTransitionNode()
  ]
  links [
   {
    dst=Reload
    rule=AnimGraph.AnimGraphNode_StateMachine_2.LeftHand.AnimStateTransitionNode_1.Transition
    src=Alias
    via=AnimStateTransitionNode_1
   }
   {
    dst=Empty_Left
    rule=AnimGraph.AnimGraphNode_StateMachine_2.LeftHand.AnimStateTransitionNode_2.Transition
    src=Conduit
    via=AnimStateTransitionNode_2
   }
   {
    dst=Idle_Left
    rule=AnimGraph.AnimGraphNode_StateMachine_2.LeftHand.AnimStateTransitionNode_3.Transition
    src=Conduit
    via=AnimStateTransitionNode_3
   }
   {
    dst=IsEquipping
    rule=AnimGraph.AnimGraphNode_StateMachine_2.LeftHand.AnimStateTransitionNode_12.Transition
    src=Conduit
    via=AnimStateTransitionNode_12
   }
   {
    dst=Reload
    rule=AnimGraph.AnimGraphNode_StateMachine_2.LeftHand.AnimStateTransitionNode_8.Transition
    src=Conduit
    via=AnimStateTransitionNode_8
   }
   {
    dst=Idle_Left
    rule=AnimGraph.AnimGraphNode_StateMachine_2.LeftHand.AnimStateTransitionNode_14.Transition
    src=Empty_Left
    via=AnimStateTransitionNode_14
   }
   {
    dst=Empty_Left
    rule=AnimGraph.AnimGraphNode_StateMachine_2.LeftHand.AnimStateTransitionNode_4.Transition
    src=Fire_Left
    via=AnimStateTransitionNode_4
   }
   {
    dst=Empty_Left
    rule=AnimGraph.AnimGraphNode_StateMachine_2.LeftHand.AnimStateTransitionNode_6.Transition
    src=Idle_Left
    via=AnimStateTransitionNode_6
   }
   {
    dst=Fire_Left
    rule=AnimGraph.AnimGraphNode_StateMachine_2.LeftHand.AnimStateTransitionNode_7.Transition
    src=Idle_Left
    via=AnimStateTransitionNode_7
   }
   {
    dst=Conduit
    rule=AnimGraph.AnimGraphNode_StateMachine_2.LeftHand.AnimStateTransitionNode_5.Transition
    src=IsEquipping
    via=AnimStateTransitionNode_5
   }
   {
    dst=Conduit
    rule=AnimGraph.AnimGraphNode_StateMachine_2.LeftHand.AnimStateTransitionNode_9.Transition
    src=Reload
    via=AnimStateTransitionNode_9
   }
  ]
 }
 Conduit {path=AnimGraph.AnimGraphNode_StateMachine_2.LeftHand.AnimStateConduitNode_0.Conduit, nodes [Result(bCanEnterTransition=true)]}
 IsEquipping {
  path=AnimGraph.AnimGraphNode_StateMachine_2.LeftHand.AnimStateNode_0.IsEquipping
  nodes [Output Animation Pose(Result=Sequence Player 'SpearOnly_Equip_Right'() {bLoopAnimation=false, sequence=&SpearOnly_Equip_Right})]
 }
 Fire_Left {
  path=AnimGraph.AnimGraphNode_StateMachine_2.LeftHand.AnimStateNode_1.Fire_Left
  nodes [Output Animation Pose(Result=Sequence Player 'TS_FireLeft_RootAnim_SpearLeft'() {bLoopAnimation=false, sequence=&TS_FireLeft_RootAnim_SpearLeft})]
 }
 Reload {
  path=AnimGraph.AnimGraphNode_StateMachine_2.LeftHand.AnimStateNode_2.Reload
  nodes [Output Animation Pose(Result=Sequence Player 'SpearOnly_Reload_Right'() {bLoopAnimation=false, sequence=&SpearOnly_Reload_Right})]
 }
 Empty_Left {
  path=AnimGraph.AnimGraphNode_StateMachine_2.LeftHand.AnimStateNode_3.Empty_Left
  nodes [Output Animation Pose(Result="Sequence Evaluator (by time) 'TS_FireLeft_RootAnim_SpearLeft'"(ExplicitTime=0.82) {sequence=&TS_FireLeft_RootAnim_SpearLeft})]
 }
 Idle_Left {
  path=AnimGraph.AnimGraphNode_StateMachine_2.LeftHand.AnimStateNode_4.Idle_Left
  nodes [Output Animation Pose(Result="Sequence Evaluator (by time) 'TS_FireLeft_RootAnim_SpearLeft'"(ExplicitTime=0) {sequence=&TS_FireLeft_RootAnim_SpearLeft})]
 }
 Transition {
  path=AnimGraph.AnimGraphNode_StateMachine_2.LeftHand.AnimStateTransitionNode_1.Transition
  nodes [
   n1 = Get RemainingAmmo() -> (RemainingAmmo:int=0)
   Result(bCanEnterTransition=AND Boolean(A=Get IsReloading(), B="integer > integer"(A=@n1, B=1)))
  ]
 }
 Transition {
  path=AnimGraph.AnimGraphNode_StateMachine_2.LeftHand.AnimStateTransitionNode_12.Transition
  nodes [
   n1 = Get Magazine Ammo Count() -> (Magazine Ammo Count=0)
   Result(bCanEnterTransition=AND Boolean(A=Get IsEquipping(), B="integer > integer"(A=@n1, B=1)))
  ]
 }
 Transition {
  path=AnimGraph.AnimGraphNode_StateMachine_2.LeftHand.AnimStateTransitionNode_14.Transition
  nodes [Result(bCanEnterTransition="Equal (Integer)"(A=Get Magazine Ammo Count() -> (Magazine Ammo Count=0), B=2))]
 }
 Transition {
  path=AnimGraph.AnimGraphNode_StateMachine_2.LeftHand.AnimStateTransitionNode_2.Transition
  nodes [Result(bCanEnterTransition="integer < integer"(A=Get Magazine Ammo Count() -> (Magazine Ammo Count=0), B=2))]
 }
 Transition {path=AnimGraph.AnimGraphNode_StateMachine_2.LeftHand.AnimStateTransitionNode_3.Transition, nodes [Result(bCanEnterTransition=true)]}
 Transition {path=AnimGraph.AnimGraphNode_StateMachine_2.LeftHand.AnimStateTransitionNode_4.Transition, nodes [Result()]}
 Transition {path=AnimGraph.AnimGraphNode_StateMachine_2.LeftHand.AnimStateTransitionNode_5.Transition, nodes [Result()]}
 Transition {
  path=AnimGraph.AnimGraphNode_StateMachine_2.LeftHand.AnimStateTransitionNode_6.Transition
  nodes [Result(bCanEnterTransition="Equal (Integer)"(A=Get Magazine Ammo Count() -> (Magazine Ammo Count=0), B=1))]
 }
 Transition {
  path=AnimGraph.AnimGraphNode_StateMachine_2.LeftHand.AnimStateTransitionNode_7.Transition
  nodes [Result(bCanEnterTransition=Get IsLeftFire() -> (IsLeftFire:bool))]
 }
 Transition {
  path=AnimGraph.AnimGraphNode_StateMachine_2.LeftHand.AnimStateTransitionNode_8.Transition
  nodes [
   n1 = Get Magazine Ammo Count() -> (Magazine Ammo Count=0)
   Result(bCanEnterTransition=AND Boolean(A=Get IsReloading(), B="Equal (Integer)"(A=@n1)))
  ]
 }
 Transition {
  path=AnimGraph.AnimGraphNode_StateMachine_2.LeftHand.AnimStateTransitionNode_9.Transition
  nodes [
   n1 = Get Magazine Ammo Count() -> (Magazine Ammo Count=0)
   Result(bCanEnterTransition=OR Boolean!(A="Equal (Integer)"(A=@n1, B=2), B=Get IsEquipping()) -> (ReturnValue))
  ]
 }
 EventGraph {nodes [
   Event BlueprintUpdateAnimation()
   > n1 = Cast To B_FeyCrab_Launcher!(execute, Object=GetOwningActor()) -> (then, CastFailed, AsB Fey Crab Launcher:B Fey Crab Launcher Object Reference) {is=title K2Node_DynamicCast}
   > Set IsReloading(IsReloading:bool=IsReloading!(self:Fort Weapon Object Reference=@n1) -> (ReturnValue:bool)) -> (Output_Get:bool)
   > Set IsEquipping(IsEquipping:bool=IsBeingEquipped!(self:Fort Weapon Object Reference=@n1) -> (ReturnValue:bool)) -> (Output_Get:bool)
   > Set IsRightFire(IsRightFire:bool=Get IsRightHand!(self:B Fey Crab Launcher Object Reference=@n1) -> (IsRightHand)) -> (Output_Get:bool)
   > Set IsLeftFire(IsLeftFire:bool=Get Lefthand!(self:B Fey Crab Launcher Object Reference=@n1) -> (Lefthand:bool)) -> (Output_Get:bool)
   > Set RemainingAmmo(RemainingAmmo:int=GetRemainingAmmo!(self:Fort Weapon Object Reference=@n1) -> (ReturnValue:int=0)) -> (Output_Get:int=0)
   > Set Magazine Ammo Count(Magazine Ammo Count:int=GetMagazineAmmoCount!(self:Fort Weapon Object Reference=@n1) -> (ReturnValue:int=0)) -> (Output_Get:int=0)
   Event BlueprintInitializeAnimation()
  ]}
]
variables [
 IsEquipping {type=boolean}
 IsLeftFire {type=boolean}
 IsLeftReloded {type=boolean}
 IsReloading {type=boolean}
 IsRightFire {type=boolean}
 IsRightHand {type=boolean}
 IsStopLeftFire {type=boolean}
 IsStopRightFire {type=boolean}
 IsWeaponFired {type=boolean}
 Magazine Ammo Count {type=integer}
 RemainingAmmo {type=integer}
]
