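# Sound cue for the FeyCrab item's ADS (aim-down-sights) foley.
# A Random node picks one of four wave variants with equal weight and is excluded from branch culling.
# A Modulator at the root varies pitch downward and randomizes volume on each play.
# It plays at reduced volume through the weapon foley sound class, using weapon reload attenuation for spatial falloff.
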SoundCue /DryBox/Items/FeyCrab/Audio/FeyCrab_ADS_Cue
root=Modulator_0
nodes [
 {id=Modulator_0, children [Random_0], properties {pitchMax=1, pitchMin=0.75, volumeMin=0.75}}
 {
  id=Random_0
  children [WavePlayer_0,WavePlayer_1,WavePlayer_2,WavePlayer_3]
  properties {bShouldExcludeFromBranchCulling=true, weights [1,1,1,1]}
 }
 {id=WavePlayer_0, properties {soundWaveAssetPtr=&/FeyCrab_ADS_04}}
 {id=WavePlayer_1, properties {soundWaveAssetPtr=&/FeyCrab_ADS_01}}
 {id=WavePlayer_2, properties {soundWaveAssetPtr=&/FeyCrab_ADS_02}}
 {id=WavePlayer_3, properties {soundWaveAssetPtr=&/FeyCrab_ADS_03}}
]
properties {
 attenuationSettings=&Fort_Weapon_Reload_Att
 maxDistance=2250
 soundClassObject=&Fort_WeaponFoley
 volumeMultiplier=0.501
}
