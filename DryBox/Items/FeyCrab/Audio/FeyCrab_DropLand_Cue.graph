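# Sound cue for the FeyCrab item's drop-landing sound, assigned to the Fort_Loot_Collect sound class.
# A Random node picks one of three equally weighted FeyCrab_DropLand wave variants.
# A root Modulator then varies volume and pitch slightly so repeated landings differ.
# Attenuation is overridden on the cue itself to set a short-range falloff.
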SoundCue /DryBox/Items/FeyCrab/Audio/FeyCrab_DropLand_Cue
root=Modulator_0
nodes [
 {id=Modulator_0, children [Random_1], properties {pitchMax=1, volumeMax=1.1, volumeMin=0.9}}
 {id=Random_1, children [WavePlayer_2,WavePlayer_3,WavePlayer_6], properties {weights [1,1,1]}}
 {id=WavePlayer_2, properties {soundWaveAssetPtr=&/FeyCrab_DropLand_03}}
 {id=WavePlayer_3, properties {soundWaveAssetPtr=&/FeyCrab_DropLand_01}}
 {id=WavePlayer_6, properties {soundWaveAssetPtr=&/FeyCrab_DropLand_02}}
]
properties {
 bOverrideAttenuation=true
 maxDistance=1000
 soundClassObject=&Fort_Loot_Collect
 volumeMultiplier=1
 attenuationOverrides {falloffDistance=1000, attenuationShapeExtents {x=0}}
}
