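# Sound cue for equipping the FeyCrab item, played as weapon foley with the weapon reload attenuation.
# A Random node picks one of three equip recordings with equal weight, and a Modulator slightly varies its pitch.
# A concurrency override allows one instance per owner, stopping the oldest when a new one starts.
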SoundCue /DryBox/Items/FeyCrab/Audio/FeyCrab_Equip_Cue
root=Modulator_0
nodes [
 {id=Modulator_0, children [Random_0], properties {pitchMax=1.1, pitchMin=0.9}}
 {id=Random_0, children [WavePlayer_0,WavePlayer_1,WavePlayer_2], properties {weights [1,1,1]}}
 {id=WavePlayer_0, properties {soundWaveAssetPtr=&/FeyCrab_Equip_03}}
 {id=WavePlayer_1, properties {soundWaveAssetPtr=&/FeyCrab_Equip_01}}
 {id=WavePlayer_2, properties {soundWaveAssetPtr=&/FeyCrab_Equip_02}}
]
properties {
 attenuationSettings=&Fort_Weapon_Reload_Att
 bOverrideConcurrency=true
 maxDistance=2250
 soundClassObject=&Fort_WeaponFoley
 volumeMultiplier=0.601
 concurrencyOverrides {bLimitToOwner=true, maxCount=1, resolutionRule=StopOldest}
}
