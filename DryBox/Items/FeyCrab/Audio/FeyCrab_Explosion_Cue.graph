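# Sound cue for the FeyCrab item's explosion, playing one of five explosion wave variations.
# An equally weighted Random node picks the wave and feeds a Modulator at the root for per-play variation.
# The cue uses a dedicated long-range attenuation asset, the large-explosives sound class and a FeyCrab explosion concurrency set.
# The Random node is excluded from branch culling.
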SoundCue /DryBox/Items/FeyCrab/Audio/FeyCrab_Explosion_Cue
root=Modulator_0
nodes [
 {id=Modulator_0, children [Random_2]}
 {
  id=Random_2
  children [WavePlayer_0,WavePlayer_1,WavePlayer_4,WavePlayer_6,WavePlayer_7]
  properties {bShouldExcludeFromBranchCulling=true, weights [1,1,1,1,1]}
 }
 {id=WavePlayer_0, properties {soundWaveAssetPtr=&/FeyCrab_Explosion_01}}
 {id=WavePlayer_1, properties {soundWaveAssetPtr=&/FeyCrab_Explosion_02}}
 {id=WavePlayer_4, properties {soundWaveAssetPtr=&/FeyCrab_Explosion_03}}
 {id=WavePlayer_6, properties {soundWaveAssetPtr=&/FeyCrab_Explosion_04}}
 {id=WavePlayer_7, properties {soundWaveAssetPtr=&/FeyCrab_Explosion_05}}
]
properties {
 attenuationSettings=/Game/Sounds/Attenuation/Fort_FeyCrab_Explosion_Att
 maxDistance=20000
 priority=5
 soundClassObject=/Game/Sounds/FortSoundClassesAndModes/Fort_Explosives_Large_PSM
 volumeMultiplier=0.5307
 concurrencyOverrides {maxCount=4, resolutionRule=StopOldest, volumeScale=0.5}
 concurrencySet [/Game/Sounds/Concurrency_Rules/FeyCrab_Explosion_Conc]
}
