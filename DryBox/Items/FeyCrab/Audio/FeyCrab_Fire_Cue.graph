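# Sound cue for the FeyCrab item's firing sound, routed to the third-person ranged weapon fire sound class.
# A random node picks one of six fire wave variations with equal weight and feeds a modulator at the root.
# It borrows an RPG explosion attenuation asset for long-range falloff, slightly boosts volume, and joins local and global gunfire concurrency sets to limit overlapping shots.
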SoundCue /DryBox/Items/FeyCrab/Audio/FeyCrab_Fire_Cue
root=Modulator_0
nodes [
 {id=Modulator_0, children [Random_0]}
 {
  id=Random_0
  children [WavePlayer_0,WavePlayer_1,WavePlayer_2,WavePlayer_3,WavePlayer_4,WavePlayer_5]
  properties {bShouldExcludeFromBranchCulling=true, weights [1,1,1,1,1,1]}
 }
 {id=WavePlayer_0, properties {soundWaveAssetPtr=&/FeyCrab_Fire_06}}
 {id=WavePlayer_1, properties {soundWaveAssetPtr=&/FeyCrab_Fire_01}}
 {id=WavePlayer_2, properties {soundWaveAssetPtr=&/FeyCrab_Fire_02}}
 {id=WavePlayer_3, properties {soundWaveAssetPtr=&/FeyCrab_Fire_03}}
 {id=WavePlayer_4, properties {soundWaveAssetPtr=&/FeyCrab_Fire_04}}
 {id=WavePlayer_5, properties {soundWaveAssetPtr=&/FeyCrab_Fire_05}}
]
properties {
 attenuationSettings=/Game/Sounds/Attenuation/Fort_Explosion_RPG_Att
 maxDistance=13000
 soundClassObject=&Fort_WeaponRanged_Fire_3P_PSM
 volumeMultiplier=1.2
 concurrencySet [&Weapon_Gunfire_Local_Conc, &Weapon_Gunfire_Global_Conc]
}
