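# Sound cue for the FeyCrab's impact on an enemy, played with third-person impact attenuation and a ranged-weapon fire sound class.
# A mixer layers a single enemy-impact "2D" wave, pitched down by a fixed amount, with a slightly quieter random variation.
# That variation comes from five equally weighted impact waves, pitched down within a narrow random range.
# The cue boosts overall volume, and its random node is excluded from branch culling.
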SoundCue /DryBox/Items/FeyCrab/Audio/FeyCrab_Impact_Enemy_Cue
root=Mixer_0
nodes [
 {id=Mixer_0, children [Modulator_1,Modulator_0], properties {inputVolume [1,0.8]}}
 {id=Modulator_0, children [Random_0], properties {pitchMax=0.85, pitchMin=0.8}}
 {id=Modulator_1, children [WavePlayer_0], properties {pitchMax=0.9, pitchMin=0.9}}
 {
  id=Random_0
  children [WavePlayer_3,WavePlayer_4,WavePlayer_5,WavePlayer_6,WavePlayer_7]
  properties {bShouldExcludeFromBranchCulling=true, weights [1,1,1,1,1]}
 }
 {id=WavePlayer_0, properties {soundWaveAssetPtr=&/FeyCrab_Impact_Enemy_2D}}
 {id=WavePlayer_3, properties {soundWaveAssetPtr=&/FeyCrab_Impact_01}}
 {id=WavePlayer_4, properties {soundWaveAssetPtr=&/FeyCrab_Impact_02}}
 {id=WavePlayer_5, properties {soundWaveAssetPtr=&/FeyCrab_Impact_03}}
 {id=WavePlayer_6, properties {soundWaveAssetPtr=&/FeyCrab_Impact_04}}
 {id=WavePlayer_7, properties {soundWaveAssetPtr=&/FeyCrab_Impact_05}}
]
properties {
 attenuationSettings=/Game/Sounds/Attenuation/Fort_FeyCrab_Impact_3P_Att
 maxDistance=8000
 soundClassObject=&Fort_WeaponRanged_Fire_3P_PSM
 volumeMultiplier=2
}
