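# Sound cue for FeyCrab scenery impacts, playing one of five equally weighted impact waves at random.
# An IsLocal branch sends local sounds through at full level; otherwise an IsTeam branch cuts team sounds to 20% volume via a modulator.
# All paths share the FeyCrab explosion attenuation, a third-person ranged weapon fire sound class override and the same random node, which is excluded from branch culling.
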SoundCue /DryBox/Items/FeyCrab/Audio/FeyCrab_Impact_Scenery_Cue
root=Branch_4
nodes [
 {id=Attenuation_4, children [SoundClass_4], properties {attenuationSettings=/Game/Sounds/Attenuation/Fort_FeyCrab_Explosion_Att}}
 {id=Attenuation_5, children [SoundClass_5], properties {attenuationSettings=&Fort_FeyCrab_Explosion_Att}}
 {id=Branch_4, children [Attenuation_4, Branch_5, Branch_5], properties {boolParameterName=IsLocal}}
 {id=Branch_5, children [Modulator_3, Attenuation_5, Attenuation_5], properties {boolParameterName=IsTeam}}
 {id=Modulator_3, children [Attenuation_5], properties {pitchMax=1, pitchMin=1, volumeMax=0.2, volumeMin=0.2}}
 {
  id=Random_0
  children [WavePlayer_0,WavePlayer_1,WavePlayer_2,WavePlayer_3,WavePlayer_4]
  properties {bShouldExcludeFromBranchCulling=true, weights [1,1,1,1,1]}
 }
 {id=SoundClass_4, children [Random_0], properties {soundClassOverride=&Fort_WeaponRanged_Fire_3P_PSM}}
 {id=SoundClass_5, children [Random_0], properties {soundClassOverride=&Fort_WeaponRanged_Fire_3P_PSM}}
 {id=WavePlayer_0, properties {soundWaveAssetPtr=&/FeyCrab_Impact_01}}
 {id=WavePlayer_1, properties {soundWaveAssetPtr=&/FeyCrab_Impact_02}}
 {id=WavePlayer_2, properties {soundWaveAssetPtr=&/FeyCrab_Impact_03}}
 {id=WavePlayer_3, properties {soundWaveAssetPtr=&/FeyCrab_Impact_04}}
 {id=WavePlayer_4, properties {soundWaveAssetPtr=&/FeyCrab_Impact_05}}
]
properties {maxDistance=20000, volumeMultiplier=1.3}
