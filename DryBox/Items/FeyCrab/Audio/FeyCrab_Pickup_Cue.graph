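# Sound cue for picking up the FeyCrab item in DryBox.
# A Random node chooses between three FeyCrab pickup wave variants with equal weight, and a Modulator applies a slight downward pitch variation.
# It uses the shared Fortnite weapon pickup attenuation and loot collect sound class, with slightly reduced volume.
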SoundCue /DryBox/Items/FeyCrab/Audio/FeyCrab_Pickup_Cue
root=Modulator_0
nodes [
 {id=Modulator_0, children [Random_0], properties {pitchMax=1, pitchMin=0.9}}
 {id=Random_0, children [WavePlayer_0,WavePlayer_1,WavePlayer_5], properties {weights [1,1,1]}}
 {id=WavePlayer_0, properties {soundWaveAssetPtr=&/FeyCrab_PickUp_03}}
 {id=WavePlayer_1, properties {soundWaveAssetPtr=&/FeyCrab_PickUp_01}}
 {id=WavePlayer_5, properties {soundWaveAssetPtr=&/FeyCrab_PickUp_02}}
]
properties {
 attenuationSettings=&Fort_Weapon_Pickup_Att
 maxDistance=3500
 soundClassObject=&Fort_Loot_Collect
 volumeMultiplier=0.796
}
