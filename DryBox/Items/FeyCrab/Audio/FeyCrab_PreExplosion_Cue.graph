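# Sound cue for the FeyCrab item's pre-explosion sound, playing one of five random wave variants through a modulator in the large explosives sound class.
# IsLocal and IsTeam branches choose the path, with non-local teammates' sounds reduced in volume by an extra modulator.
# All paths share the same attenuation asset and random pool, which is excluded from branch culling.
# Concurrency settings include a shared FeyCrab explosion set and a one-per-owner, prevent-new override.
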SoundCue /DryBox/Items/FeyCrab/Audio/FeyCrab_PreExplosion_Cue
root=Branch_4
nodes [
 {id=Attenuation_4, children [Modulator_0], properties {attenuationSettings=/Game/Sounds/Attenuation/Fort_FeyCrab_PreExplosion_Att}}
 {id=Attenuation_5, children [Modulator_0], properties {attenuationSettings=&Fort_FeyCrab_PreExplosion_Att}}
 {id=Branch_4, children [Attenuation_4, Branch_5, Branch_5], properties {boolParameterName=IsLocal}}
 {id=Branch_5, children [Modulator_3, Attenuation_5, Attenuation_5], properties {boolParameterName=IsTeam}}
 {id=Modulator_0, children [Random_0]}
 {id=Modulator_3, children [Attenuation_5], properties {pitchMax=1, pitchMin=1, volumeMax=0.4, volumeMin=0.4}}
 {
  id=Random_0
  children [WavePlayer_0,WavePlayer_2,WavePlayer_3,WavePlayer_4,WavePlayer_5]
  properties {bShouldExcludeFromBranchCulling=true, weights [1,1,1,1,1]}
 }
 {id=WavePlayer_0, properties {soundWaveAssetPtr=&/FeyCrab_PreExplosion_01}}
 {id=WavePlayer_2, properties {soundWaveAssetPtr=&/FeyCrab_PreExplosion_02}}
 {id=WavePlayer_3, properties {soundWaveAssetPtr=&/FeyCrab_PreExplosion_03}}
 {id=WavePlayer_4, properties {soundWaveAssetPtr=&/FeyCrab_PreExplosion_04}}
 {id=WavePlayer_5, properties {soundWaveAssetPtr=&/FeyCrab_PreExplosion_05}}
]
properties {
 maxDistance=16000
 soundClassObject=/Game/Sounds/FortSoundClassesAndModes/Fort_Explosives_Large_PSM
 volumeMultiplier=0.3538
 concurrencyOverrides {bLimitToOwner=true, maxCount=1, resolutionRule=PreventNew}
 concurrencySet [/Game/Sounds/Concurrency_Rules/FeyCrab_Explosion_Conc]
}
