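# Sound cue for the pin-pull sound of the FeyCrab item.
# A Random node picks one of three equally weighted pull-pin waves and feeds a Modulator that varies pitch.
# It borrows the Fort lobber-projectile close attenuation and a third-person ranged weapon fire sound class, with boosted volume.
# A concurrency override allows one instance per owner and blocks new plays while it is active.
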SoundCue /DryBox/Items/FeyCrab/Audio/FeyCrab_PullPin_Cue
root=Modulator_0
nodes [
 {id=Modulator_0, children [Random_0], properties {pitchMax=1.1}}
 {id=Random_0, children [WavePlayer_0,WavePlayer_1,WavePlayer_2], properties {weights [1,1,1]}}
 {id=WavePlayer_0, properties {soundWaveAssetPtr=&/FeyCrab_PullPin_01}}
 {id=WavePlayer_1, properties {soundWaveAssetPtr=&/FeyCrab_PullPin_02}}
 {id=WavePlayer_2, properties {soundWaveAssetPtr=&/FeyCrab_PullPin_03}}
]
properties {
 attenuationSettings=/Game/Sounds/Attenuation/Fort_LobberProjectileClose_Att
 bOverrideConcurrency=true
 maxDistance=1700
 soundClassObject=&Fort_WeaponRanged_Fire_3P_PSM
 volumeMultiplier=1.5
 concurrencyOverrides {bLimitToOwner=true, maxCount=1, resolutionRule=PreventNew}
}
