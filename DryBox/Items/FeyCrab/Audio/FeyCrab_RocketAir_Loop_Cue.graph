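# Looping in-air sound cue for the FeyCrab rocket projectile.
# A Random node picks one of three equally weighted looping waves, excluded from branch culling.
# An Enveloper ramps volume from silence to full over roughly the first half second, then a Doppler node at the root adds motion-based pitch shift.
# It uses shared projectile in-air attenuation and a third-person ranged weapon fire sound class.
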SoundCue /DryBox/Items/FeyCrab/Audio/FeyCrab_RocketAir_Loop_Cue
root=Doppler_0
nodes [
 {id=Doppler_0, children [Enveloper_1]}
 {
  id=Enveloper_1
  children [Random_0]
  properties {pitchCurve {editorCurveData {keys [{time=0}]}}, volumeCurve {editorCurveData {keys [<at 0 0>,<at 0.4482 0.3442>,<at 0.6 1>]}}}
 }
 {id=Random_0, children [WavePlayer_0,WavePlayer_1,WavePlayer_2], properties {bShouldExcludeFromBranchCulling=true, weights [1,1,1]}}
 {id=WavePlayer_0, properties {bLooping=true, soundWaveAssetPtr=&/FeyCrab_RocketAir_Loop_01}}
 {id=WavePlayer_1, properties {bLooping=true, soundWaveAssetPtr=&/FeyCrab_RocketAir_Loop_02}}
 {id=WavePlayer_2, properties {bLooping=true, soundWaveAssetPtr=&/FeyCrab_RocketAir_Loop_03}}
]
properties {
 attenuationSettings=/Game/Sounds/Attenuation/Fort_Weapon_Projectile_InAir_Loop_Att
 maxDistance=4500
 soundClassObject=&Fort_WeaponRanged_Fire_3P_PSM
}
