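; Fey Crab launcher weapon: a two-handed ranged item that fires projectiles, launches or recoils the player, and leaves inventory when out of ammo.
; Magazine count picks the firing hand; pawn state picks an upper-body, in-air or ground montage.
; Montage notifies drive projectile spawning, jump-apex and knockback gameplay events, wire pulls and muzzle FX, mirrored to clients via OnFiredMulticast.
; It also handles offhand mesh attachment, low-gravity apex, reload FX and outline materials.
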(event EventOnWeaponVisibilityChanged
  (params (bVisible :type Boolean) (bSetForLocalControllerOnly :type Boolean)))

(event EventMuzzleFlashFX
  (params (Persistent%20Fire :type Boolean) (OffHand :type Boolean))
  ;[1 1 1] Event on muzzle flash
  (|CallParentFunction false false))

(event FortWeapon|EventOnLocalUnEquip
  ;[1 1 1] Event on unequip
  (Transformation|DetachFromComponent (Variables|Default|GetWeapon%20Mesh%20Offhand) "KeepRelative" "KeepRelative" "KeepRelative" true)
  (CallFunction|RemoveWeaponfromInventory))

(event FortWeapon|EventOnUnEquippedAndDetached
  ;[1 1 1] Event on unequip
  (Transformation|DetachFromComponent (Variables|Default|GetWeapon%20Mesh%20Offhand) "KeepRelative" "KeepRelative" "KeepRelative" true)
  (CallFunction|RemoveWeaponfromInventory))

(event EventBeginPlay
  ;[1 1 1] Begin play set anim to right hand and apply low end outlines
  (bind _asfey_crab_launcher_wep_anim_bp (Utilities|Casting|CastToFeyCrab_Launcher_WEP_AnimBP (Components|SkeletalMesh|GetAnimInstance (Variables|Visual|GetWeaponMesh))))
  (Variables|/DryBox/Items/FeyCrab/Animation/AnimBP/FeyCrab_Launcher_WEP_AnimBP.FeyCrab_Launcher_WEP_AnimBP_C|SetIsRightHand true _asfey_crab_launcher_wep_anim_bp)
  (CallFunction|ApplyLowEndOutlines))

(event Custom|PlayFailedReason
  ;[1 1 1] OnEquip W/ Zero
  (if (<= (FortWeapon|GetLocalMagazineAmmoCount) 0)
    (FailedAction|UnabletoPerformAction (Variables|References|GetAs%20Fort%20Player%20Pawn) false (Variables|FailedToReload|GetFailed%20Reason))
    (CallFunction|PlayCannotDoSound)))

(event EventOnPlayReloadFX
  (params (ReloadStage :type EFortReloadFXState%20Enum :default "ReloadStart"))
  ;[1 1 1] Reload
  (if (CallFunction|IsBeingEquipped)
    (CallFunction|FXonReload (== (FortWeapon|GetLocalMagazineAmmoCount) 1) (== (FortWeapon|GetLocalMagazineAmmoCount) 0))
    (else
      (CallFunction|FXonReload (== (FortWeapon|GetRemainingAmmo) 1)))))

(event FortWeapon|EventOnEquippedAndAttached
  ;[1 1 1] On Equipped
  (|CallParentFunction)
  (bind _asfort_player_pawn (Utilities|Casting|CastToFortPlayerPawn (Game|GetInstigator)))
  (bind _output_get (Variables|References|SetAs%20Fort%20Player%20Pawn _asfort_player_pawn))
  (if (GameplayTags|HasMatchingGameplayTag (Variables|References|GetAs%20Fort%20Player%20Pawn) (Variables|Tags|GetSwingingTag))
    ;[1 1 1] OnEquip W/ Zero
    (Utilities|Time|SetTimerbyEvent (delegate Custom|PlayFailedReason) 0.2))
  (Utilities|IsValid (Variables|/Script/Engine.Character|GetMesh _output_get)
    (:Is%20Valid
      (Transformation|AttachComponentToComponent (Variables|Default|GetWeapon%20Mesh%20Offhand) (Variables|/Script/Engine.Character|GetMesh _output_get) (Variables|Sockets|GetOffhand%20Socket%20Name) "SnapToTarget" "KeepRelative" "SnapToTarget" true)
      (Networking|SwitchHasAuthority
        (:Authority
          (bind _as_fort_player_pawn (Variables|References|GetAs%20Fort%20Player%20Pawn)
            (:then
              (bind _playercameramanager (Variables|/Script/Engine.PlayerController|GetPlayerCameraManager (Fortnite|GetFortPlayerControllerfromActor _as_fort_player_pawn))
                (:then
                  (Variables|Camera|SetCameraManager _playercameramanager))
                (:else)))
            (:else)))))))

(event Custom|RemoveWeapon
  ;[1 1 1] Remove weapon if ammo count is 0
  (CallFunction|RemoveWeaponfromInventory))

(event Custom|OnFired
  (params (GA%20REf :type Fort%20Gameplay%20Ability%20Object%20Reference))
  (bind _as_b_prj_ranged_fey_crab (Variables|References|GetAs%20B%20Prj%20Ranged%20Fey%20Crab))
  ;[1 1 1] Firing Event
  (Variables|States|SetHasFired true)
  (Networking|SwitchHasAuthority
    (:Authority
      (CallFunction|SetFiringHand (Variables|References|GetMagazineAmmoCount))
      (CallFunction|AddProjectiletoWorld)
      (Utilities|FlowControl|Sequence
        (:then_0
          ;[1 1 1] Select Montage based on ammo
          (Networking|SwitchHasAuthority
            (:Authority
              (CallFunction|OnFiredMulticast))
            (:Remote
              _))
          (CallFunction|SelectMontageToPlay))
        (:then_1
          ;[1 1 1] Animation to gameplay triggers
          (Ability|SendGameplayEventtoActor (Variables|References|GetAs%20Fort%20Player%20Pawn) "(TagName=\"Event.Ability.FeyCrab.Launch\")")
          (bind _notifyname (Animation|Montage|PlayMontage (Variables|/Script/Engine.Character|GetMesh (Variables|References|GetAs%20Fort%20Player%20Pawn)) (Variables|Animations|GetMontage%20to%20Play))
            (:OnCompleted
              _)
            (:OnInterrupted
              _)
            (:OnNotifyBegin
              (switch name _notifyname
                (:JumpTrigger
                  ;[1 1 1] Jump Apex
                  ;[1 1 1] Jump Apex Event
                  (Ability|SendGameplayEventtoActor (Variables|References|GetAs%20Fort%20Player%20Pawn) "(TagName=\"Event.Ability.FeyCrab.JumpApex\")"))
                (:LaunchMissle
                  ;[1 1 1] Launch Missile
                  (Audio|PlaySoundatLocation "/DryBox/Items/FeyCrab/Audio/FeyCrab_Fire_Cue.FeyCrab_Fire_Cue" (Transformation|GetActorLocation (Variables|References|GetAs%20Fort%20Player%20Pawn))) ; rocket launch sound
                  (Variables|Projectile|SetSpawn%20Aim%20Rotation (Pawn|GetBaseAimRotation (Variables|References|GetAs%20Fort%20Player%20Pawn)))
                  (Variables|Projectile|SetSpawnLocation (CallFunction|CalculateSpawnLocation))
                  (Utilities|IsValid (Variables|References|GetAs%20Fort%20Player%20Pawn)
                    (:Is%20Valid
                      (bind _returnvalue (Projectile|SpawnProjectile :ProjectileClass "/DryBox/Items/FeyCrab/B_Prj_Ranged_FeyCrab.B_Prj_Ranged_FeyCrab_C" :RequestedBy (Variables|References|GetAs%20Fort%20Player%20Pawn) :SpawnLocation (Variables|Projectile|GetSpawnLocation) :SpawnRotation (Variables|Projectile|GetSpawn%20Aim%20Rotation) :SpawnDirection (Variables|Projectile|GetSpawn%20Aim%20Rotation) :bUseDefaultPhysics false :InitialSpeed (ScalableFloat|GetValueatLevel (Variables|HotfixValues|GetProjectileInitialSpeed)) :GravityScale 0.0 :ChargePercent 0.0))
                      (Variables|States|SetHasFired false)
                      (Networking|SwitchHasAuthority
                        (:Authority
                          (bind _asb_prj_ranged_fey_crab (Utilities|Casting|CastToB_Prj_Ranged_FeyCrab _returnvalue))
                          (Variables|References|SetAs%20B%20Prj%20Ranged%20Fey%20Crab _asb_prj_ranged_fey_crab))))))
                (:Launch
                  ;[1 1 1] Knockback
                  (if (ScalableFloat|GetValueasBool (Variables|HotfixValues|GetInAirFireRecoilEnabled))
                    (Ability|SendGameplayEventtoActor (Variables|References|GetAs%20Fort%20Player%20Pawn) "(TagName=\"Event.Ability.FeyCrab.Knockback\")")))
                (:YankWire
                  ;[1 1 1] Trigger pull wire event in projectile
                  (Utilities|IsValid _as_b_prj_ranged_fey_crab
                    (:Is%20Valid
                      (Class|BPrjRangedFeyCrab|PulledWire _as_b_prj_ranged_fey_crab _as_b_prj_ranged_fey_crab))
                    (:Is%20Not%20Valid
                      (Development|PrintString "FeyCrabLauncher: Failed to Ref Prj"))))
                (:MuzzleRFX
                  ;[1 1 1] Muzzle Flash FX
                  (Networking|SwitchHasAuthority
                    (:Remote
                      (CallFunction|FeycrabLauncherFireMuzzleFlashFX true))))
                (:MuzzleLFX
                  (Networking|SwitchHasAuthority
                    (:Remote
                      (CallFunction|FeycrabLauncherFireMuzzleFlashFX false))))))) ; Uses notifies in TS_FireRight_RootAnim_Montage\nto trigger the different events
          (Variables|States|SetIsRightHand false)
          (Variables|States|SetLefthand false)
          (Variables|States|SetJumpFromGround false))))))

(event Custom|ReloadAmmo
  ;[1 1 1] Reload if lower than 0 in magazine
  (CallFunction|ReloadCheck))

(event Custom|OnFiredMulticast
  (bind _as_b_prj_ranged_fey_crab (Variables|References|GetAs%20B%20Prj%20Ranged%20Fey%20Crab))
  ;[1 1 1] Firing Event
  (Networking|SwitchHasAuthority
    (:Remote
      (Variables|References|SetMagazineAmmoCount (FortWeapon|GetLocalMagazineAmmoCount))
      (CallFunction|SetFiringHand (Variables|References|GetMagazineAmmoCount))
      (CallFunction|AddProjectiletoWorld)
      (Utilities|FlowControl|Sequence
        (:then_0
          ;[1 1 1] Select Montage based on ammo
          (Networking|SwitchHasAuthority
            (:Authority
              (CallFunction|OnFiredMulticast))
            (:Remote
              _))
          (CallFunction|SelectMontageToPlay))
        (:then_1
          ;[1 1 1] Animation to gameplay triggers
          (Ability|SendGameplayEventtoActor (Variables|References|GetAs%20Fort%20Player%20Pawn) "(TagName=\"Event.Ability.FeyCrab.Launch\")")
          (bind _notifyname (Animation|Montage|PlayMontage (Variables|/Script/Engine.Character|GetMesh (Variables|References|GetAs%20Fort%20Player%20Pawn)) (Variables|Animations|GetMontage%20to%20Play))
            (:OnCompleted
              _)
            (:OnInterrupted
              _)
            (:OnNotifyBegin
              (switch name _notifyname
                (:JumpTrigger
                  ;[1 1 1] Jump Apex
                  ;[1 1 1] Jump Apex Event
                  (Ability|SendGameplayEventtoActor (Variables|References|GetAs%20Fort%20Player%20Pawn) "(TagName=\"Event.Ability.FeyCrab.JumpApex\")"))
                (:LaunchMissle
                  ;[1 1 1] Launch Missile
                  (Audio|PlaySoundatLocation "/DryBox/Items/FeyCrab/Audio/FeyCrab_Fire_Cue.FeyCrab_Fire_Cue" (Transformation|GetActorLocation (Variables|References|GetAs%20Fort%20Player%20Pawn))) ; rocket launch sound
                  (Variables|Projectile|SetSpawn%20Aim%20Rotation (Pawn|GetBaseAimRotation (Variables|References|GetAs%20Fort%20Player%20Pawn)))
                  (Variables|Projectile|SetSpawnLocation (CallFunction|CalculateSpawnLocation))
                  (Utilities|IsValid (Variables|References|GetAs%20Fort%20Player%20Pawn)
                    (:Is%20Valid
                      (bind _returnvalue (Projectile|SpawnProjectile :ProjectileClass "/DryBox/Items/FeyCrab/B_Prj_Ranged_FeyCrab.B_Prj_Ranged_FeyCrab_C" :RequestedBy (Variables|References|GetAs%20Fort%20Player%20Pawn) :SpawnLocation (Variables|Projectile|GetSpawnLocation) :SpawnRotation (Variables|Projectile|GetSpawn%20Aim%20Rotation) :SpawnDirection (Variables|Projectile|GetSpawn%20Aim%20Rotation) :bUseDefaultPhysics false :InitialSpeed (ScalableFloat|GetValueatLevel (Variables|HotfixValues|GetProjectileInitialSpeed)) :GravityScale 0.0 :ChargePercent 0.0))
                      (Variables|States|SetHasFired false)
                      (Networking|SwitchHasAuthority
                        (:Authority
                          (bind _asb_prj_ranged_fey_crab (Utilities|Casting|CastToB_Prj_Ranged_FeyCrab _returnvalue))
                          (Variables|References|SetAs%20B%20Prj%20Ranged%20Fey%20Crab _asb_prj_ranged_fey_crab))))))
                (:Launch
                  ;[1 1 1] Knockback
                  (if (ScalableFloat|GetValueasBool (Variables|HotfixValues|GetInAirFireRecoilEnabled))
                    (Ability|SendGameplayEventtoActor (Variables|References|GetAs%20Fort%20Player%20Pawn) "(TagName=\"Event.Ability.FeyCrab.Knockback\")")))
                (:YankWire
                  ;[1 1 1] Trigger pull wire event in projectile
                  (Utilities|IsValid _as_b_prj_ranged_fey_crab
                    (:Is%20Valid
                      (Class|BPrjRangedFeyCrab|PulledWire _as_b_prj_ranged_fey_crab _as_b_prj_ranged_fey_crab))
                    (:Is%20Not%20Valid
                      (Development|PrintString "FeyCrabLauncher: Failed to Ref Prj"))))
                (:MuzzleRFX
                  ;[1 1 1] Muzzle Flash FX
                  (Networking|SwitchHasAuthority
                    (:Remote
                      (CallFunction|FeycrabLauncherFireMuzzleFlashFX true))))
                (:MuzzleLFX
                  (Networking|SwitchHasAuthority
                    (:Remote
                      (CallFunction|FeycrabLauncherFireMuzzleFlashFX false))))))) ; Uses notifies in TS_FireRight_RootAnim_Montage\nto trigger the different events
          (Variables|States|SetIsRightHand false)
          (Variables|States|SetLefthand false)
          (Variables|States|SetJumpFromGround false))))))

(event Custom|ApexGravityMulti
  ;[1 1 1] Change gravity at the apex
  (Utilities|IsValid (Variables|References|GetAs%20Fort%20Player%20Pawn)
    (:Is%20Valid
      (Utilities|FlowControl|Sequence
        (:then_0
          (CallFunction|ApplyLowGrav))
        (:then_1
          (Timeline|GravityScalar PlayFromStart
            (:Finished
              (Animation|Montage|MontageSetPlayRate (Components|SkeletalMesh|GetAnimInstance (Variables|/Script/Engine.Character|GetMesh (Variables|References|GetAs%20Fort%20Player%20Pawn))) (Variables|Animations|GetMontage%20to%20Play) 1.5)
              (Pawn|Components|CharacterMovement|SetMovementMode (Variables|/Script/Engine.Character|GetCharacterMovement (Variables|References|GetAs%20Fort%20Player%20Pawn)) "MOVE_Falling" 0))))))))

(event Custom|RemoveProjectileFromWorld
  ;[1 1 1] Remove projectile count in world
  (bind _result (Math|Integer|DecrementInt (Variables|References|GetProjectileInWorld))))

(event Custom|AddProjectileToWorld
  ;[1 1 1] Add projectile count in world
  (bind _result (Math|Integer|IncrementInt (Variables|References|GetProjectileInWorld))))

(event Custom|FeycrabLauncher_FireMuzzleFlashFX
  (params (IsRH :type Boolean))
  (Components|Activation|Activate (select IsRH (Variables|Default|GetFX_Muzzle_RHS) (Variables|Default|GetFX_Muzzle_LHS)) true))

(event Custom|FX_OnReload
  (params (bHasOneWeapon :type Boolean) (bIsEmpty :type Boolean))
  (if bIsEmpty
    _
    (elif bHasOneWeapon
      (bind _returnvalue (Niagara|SpawnSystemAttached "/DryBox/Items/FeyCrab/FX/NS_FeycrabLauncher_Respawn.NS_FeycrabLauncher_Respawn" (Variables|Default|GetWeapon%20Mesh%20TEMP) "spear" "0, 0, 0" "0, 0, 0" "SnapToTarget"))
      (Variables|Default|SetNS_WeaponGlow _returnvalue)
      (else
        (for _array_element (Utilities|Array|MakeArray (Variables|Default|GetWeapon%20Mesh%20TEMP) (Variables|Default|GetWeapon%20Mesh%20Offhand))
          (bind _returnvalue_2 (Niagara|SpawnSystemAttached "/DryBox/Items/FeyCrab/FX/NS_FeycrabLauncher_Respawn.NS_FeycrabLauncher_Respawn" _array_element "spear" "0, 0, 0" "0, 0, 0" "SnapToTarget"))
          (Variables|Default|SetNS_WeaponGlow _returnvalue_2))))))

(fn ApplyLowEndOutlines
  (if (Fortnite|IsMobilePlatform)
    (Rendering|Material|SetOverlayMaterial (Variables|Visual|GetWeaponMesh) "/DryBox/Items/FeyCrab/Materials/FeyCrab_Sharp_LowEndOutline.FeyCrab_Sharp_LowEndOutline" false 0 :!class /Script/Engine.MeshComponent)
    (else
      (Rendering|Material|SetOverlayMaterial (CallFunction|GetLeftHandWeaponMesh) "/DryBox/Items/FeyCrab/Materials/FeyCrab_Sharp_Outlines.FeyCrab_Sharp_Outlines" false 0 :!class /Script/Engine.MeshComponent))))

(fn ApplyLowGrav
  (GameplayEffects|ApplyGameplayEffectToSelf :self (Ability|GetAbilitySystemComponent (Variables|References|GetAs%20Fort%20Player%20Pawn)) :GameplayEffectClass "/DryBox/Items/FeyCrab/GE_Feycrab_LowGrav.GE_Feycrab_LowGrav_C" :Level 0.0))

(fn CalculateSpawnLocation
  (bind _as_fort_player_pawn (Variables|References|GetAs%20Fort%20Player%20Pawn)
    (:then
      (return (+ (Camera|GetCameraLocation (Variables|/Script/Engine.PlayerController|GetPlayerCameraManager (Fortnite|GetFortPlayerControllerfromActor _as_fort_player_pawn))) (Math|Vector|ProjectVectorontoVector (- (Transformation|GetActorLocation _as_fort_player_pawn) (Camera|GetCameraLocation (Variables|/Script/Engine.PlayerController|GetPlayerCameraManager (Fortnite|GetFortPlayerControllerfromActor _as_fort_player_pawn)))) (Math|Vector|GetForwardVector (Camera|GetCameraRotation (Variables|/Script/Engine.PlayerController|GetPlayerCameraManager (Fortnite|GetFortPlayerControllerfromActor _as_fort_player_pawn))))))))
    (:else)))

(fn InAirOrSwinging
  (return (or (AI|Components|NavMovement|IsFalling (Variables|/Script/Engine.Character|GetCharacterMovement (Variables|References|GetAs%20Fort%20Player%20Pawn))) (GameplayTags|HasMatchingGameplayTag (Variables|References|GetAs%20Fort%20Player%20Pawn) (Variables|Tags|GetSwingingTag)))))

(fn KnockbackCharacter
  (params (bInsideBuilding :type Boolean))
  (if (not (GameplayTags|HasMatchingGameplayTag (Variables|References|GetAs%20Fort%20Player%20Pawn) (Variables|Tags|GetKnockbagTag)))
    (bind _as_fort_player_pawn (Variables|References|GetAs%20Fort%20Player%20Pawn)
      (:then
        (bind _cameramanager (Variables|Camera|GetCameraManager)
          (:then
            (Character|LaunchCharacter (Variables|References|GetAs%20Fort%20Player%20Pawn) (Utilities|Operators|Multiply (Math|Vector|SelectVector (Math|Vector|GetForwardVector (Transformation|GetWorldRotation (Variables|/Script/Engine.Character|GetCapsuleComponent _as_fort_player_pawn))) (Math|Vector|GetForwardVector (Camera|GetCameraRotation _cameramanager)) (> (.z (Math|Vector|GetForwardVector (Camera|GetCameraRotation _cameramanager))) 0)) (Math|Float|SelectFloat (* (ScalableFloat|GetValueatLevel (Variables|HotfixValues|GetLaunchStrength)) (ScalableFloat|GetValueatLevel (Variables|HotfixValues|GetLaunchStrengthIndoorMultiplier))) (Math|Float|SelectFloat (* (ScalableFloat|GetValueatLevel (Variables|HotfixValues|GetLaunchStrength)) (ScalableFloat|GetValueatLevel (Variables|HotfixValues|GetLaunchStrengthSwingingMultiplier))) (ScalableFloat|GetValueatLevel (Variables|HotfixValues|GetLaunchStrength)) (GameplayTags|HasMatchingGameplayTag (Variables|References|GetAs%20Fort%20Player%20Pawn) (Variables|Tags|GetSwingingTag))) (Variables|GetbInsideBuilding)) -1.0) true false))
          (:else
            _))
        (GameplayCue|ExecuteGameplayCue%28Burst%29Local (Variables|References|GetAs%20Fort%20Player%20Pawn) (Variables|GameplayCues|GetOnFireInAirKnockbackCueTag) (Ability|GameplayCue|MakeGameplayCueParameters 0.0)))
      (:else))))

(fn PlayCannotDoSound
  (if (Pawn|IsLocallyViewed (Variables|References|GetAs%20Fort%20Player%20Pawn))
    (Audio|PlaySound2D (Variables|FailedtoReload|GetCantDoThatSound))))

(fn ReloadCheck
  (bind _currentweapon (Variables|/Script/FortniteGame.FortPawn|GetCurrentWeapon (Variables|References|GetAs%20Fort%20Player%20Pawn))
    (:then
      (if (<= (FortWeapon|GetRemainingAmmo _currentweapon) 0)
        _
        (elif (<= (FortWeapon|GetLocalMagazineAmmoCount _currentweapon) 0)
          (FortWeapon|Reload _currentweapon))))
    (:else)))

(fn RemoveWeaponFromInventory
  (bind _as_fort_player_pawn (Variables|References|GetAs%20Fort%20Player%20Pawn))
  (bind _currentweapon (Variables|/Script/FortniteGame.FortPawn|GetCurrentWeapon _as_fort_player_pawn))
  (if (<= (FortWeapon|GetRemainingAmmo) 0)
    (if (== _currentweapon self)
      (Inventory|RemoveItemFromPlayerByGuid (Fortnite|GetFortPlayerControllerfromActor _as_fort_player_pawn) (FortWeapon|GetInventoryGUID _currentweapon) 1 true))))

(fn SelectMontagetoPlay
  (if (GameplayTags|HasAnyMatchingGameplayTags (Variables|References|GetAs%20Fort%20Player%20Pawn) (Variables|Tags|GetPlay%20Upper%20Body%20Only%20Tags))
    ;[1 1 1] Upper Body Only
    (Variables|Animations|SetMontage%20to%20Play (select (Variables|States|GetIsRightHand) (Variables|Animations|GetUpperBodyOnly_RightFire) (Variables|Animations|GetUpperBodyOnly_LeftFire)))
    (elif (CallFunction|InAirorSwinging)
      ;[1 1 1] In Air or Swinging
      (Variables|Animations|SetMontage%20to%20Play (select (Variables|States|GetIsRightHand) (Variables|Animations|GetFireRightInAirMontage) (Variables|Animations|GetFireLeftInAirMontage)))
      (else
        ;[1 1 1] On Ground
        (Variables|Animations|SetMontage%20to%20Play (select (Variables|States|GetIsRightHand) (Variables|Animations|GetFireRightGroundMontage) (Variables|Animations|GetFireLeftGroundMontage)))))))

(fn SetFiringHand
  (params (Ammocount :type Integer))
  (bind _output_get (Variables|References|SetMagazineAmmoCount (Math|Integer|SelectInt (- (Variables|GetAmmocount) 1) (Variables|GetAmmocount) (Utilities|Enum|Equal(Enum) (Networking|GetLocalRole (Variables|References|GetAs%20Fort%20Player%20Pawn)) "ROLE_SimulatedProxy"))))
  (Variables|States|SetIsRightHand (== _output_get 0))
  (Variables|States|SetLefthand (== _output_get 1)))

(fn ConstructionScript
  (|CallParentFunction)
  (CallFunction|ApplyLowEndOutlines))
