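; B_Pickups_DryBox is a pickup Blueprint that gives the pickup's skeletal mesh a FeyCrab outline.
; The construction script runs the parent's logic, then calls SetOutlineMaterial on the mesh component obtained from the parent.
; That function applies the outline as an overlay material, choosing a low-end variant on mobile and the sharp outlines material on other platforms.
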(fn SetOutlineMaterial
  (bind _skelmeshcomponent (Variables|FortPickupsParent|GetSkelMeshComponent))
  (if (TouchPlatform|IsMobileGame)
    (Rendering|Material|SetOverlayMaterial _skelmeshcomponent "/DryBox/Items/FeyCrab/Materials/FeyCrab_Sharp_LowEndOutline.FeyCrab_Sharp_LowEndOutline" false 0 :!class /Script/Engine.MeshComponent)
    (else
      (Rendering|Material|SetOverlayMaterial _skelmeshcomponent "/DryBox/Items/FeyCrab/Materials/FeyCrab_Sharp_Outlines.FeyCrab_Sharp_Outlines" false 0 :!class /Script/Engine.MeshComponent))))

(fn ConstructionScript
  (|CallParentFunction)
  (CallFunction|SetOutlineMaterial))
