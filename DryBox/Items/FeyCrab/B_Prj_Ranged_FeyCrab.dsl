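; FeyCrab launcher projectile: a tethered explosive that sticks to surfaces, buildings and pawns, then detonates.
; Pulling the wire plus an impact primes the fuse; a max-duration timer after wire detach, an in-air timer, or death of the stuck target can also trigger detonation.
; The authority runs two chained async targeting passes (contact, then in-wall) from temporary trigger spheres to deal damage and wreck vehicle parts.
; Cosmetic paths play surface or water explosion effects and signal the owner via gameplay event.
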(event Custom|DetachWire
  ;[0.367 1 0.286] Detach Wire and start fuse
  (Utilities|FlowControl|Sequence
    (:then_0
      (Utilities|Time|ClearandInvalidateTimerbyHandle (Variables|WireFuse|GetWireUpdateTimer))
      (Development|SetHiddeninGame (Variables|Default|GetWire) true false)
      (Utilities|Time|SetTimerbyEvent (EventDispatchers|CreateEvent :!event Custom|FuseTimerMax) (ScalableFloat|GetValueatLevel (Variables|Hotfixes|GetFuseTimerMaxDuration)))
      (Timeline|Timeline Play))))

(event Custom|SecondTarget
  (bind _targetingfront (Variables|Default|GetTargetingFront))
  ;[1 1 1] Second Targeting
  (bind _returnvalue (Game|SpawnActorfromClass "/Script/Engine.TriggerSphere" (Transformation|GetWorldTransform _targetingfront) "Undefined" "MultiplyWithRoot" 0 0))
  (Variables|Damage|SetDummyFrontSphere _returnvalue)
  (TargetingSystem|AsyncRequest|StartAsyncTargetingRequest (GameInstanceSubsystems|GetTargetingSubsystem) "/DryBox/Items/FeyCrab/TP_FeyCrab_InwallExplosion.TP_FeyCrab_InwallExplosion" (Utilities|Struct|MakeTargetingSourceContext (Variables|Damage|GetDummyFrontSphere) (Variables|Actor|GetInstigator) (Transformation|GetWorldLocation _targetingfront)) (delegate Custom|SecondTargetingDamage)))

(event Custom|FirstTargeting
  (bind _targetingback (Variables|Default|GetTargetingBack))
  ;[1 1 1] First Targeting
  (bind _returnvalue (Game|SpawnActorfromClass "/Script/Engine.TriggerSphere" (Transformation|GetWorldTransform _targetingback) "Undefined" "MultiplyWithRoot" 0 0))
  (Variables|Damage|SetDummyRearSphere _returnvalue)
  (TargetingSystem|AsyncRequest|StartAsyncTargetingRequest (GameInstanceSubsystems|GetTargetingSubsystem) "/DryBox/Items/FeyCrab/TP_FeyCrab_ContactExplosion.TP_FeyCrab_ContactExplosion" (Utilities|Struct|MakeTargetingSourceContext (Variables|Damage|GetDummyRearSphere) (Variables|Actor|GetInstigator) (Transformation|GetWorldLocation _targetingback)) (delegate Custom|FirstTargetingDamage)))

(event EventBeginPlay
  (bind _ns_tetherwire (Variables|Default|GetNS_TetherWire))
  ;[1 1 1] EVENT -> Begin Play
  (Variables|InAirTime|SetServerWorldTime (Fortnite|Sync|Time|GetServerWorldTimeSeconds))
  (Utilities|FlowControl|Sequence
    (:then_0
      (if (Utilities|Operators|Equal(==) :A (Transformation|GetVelocity) :ErrorTolerance 1.0)
        (Networking|CanExecuteCosmeticEvents
          (:True
            (Variables|States|SetSurfaceImpacted true)
            (CallFunction|FuseActivated)))))
    (:then_1
      (bind _returnvalue (Utilities|Time|SetTimerbyEvent (EventDispatchers|CreateEvent :!event Custom|UpdateTheWire) 0.005 true))
      (Variables|WireFuse|SetWireUpdateTimer _returnvalue)
      (Utilities|FlowControl|Sequence
        (:then_0
          (Networking|CanExecuteCosmeticEvents
            (:True
              (Utilities|FlowControl|Sequence
                (:then_0
                  (bind _returnvalue_2 (Niagara|SpawnSystemAttached "/DryBox/Items/FeyCrab/FX/NS_FeycrabLauncher_Prj_Trail.NS_FeycrabLauncher_Prj_Trail" (Variables|Collision|GetCapsuleComponent) "None" "-160.000000,0.000000,0.000000" "0, 0, 0" "SnapToTarget")) ; Spawn trail; needs to manage its own life
                  (Variables|Default|SetRocketTrailPS_Niagara _returnvalue_2))))))
        (:then_1
          (Utilities|Time|SetTimerbyEvent (EventDispatchers|CreateEvent :!event Custom|DetachWire) 1.0)
          (bind _returnvalue_3 (CallFunction|GetOwnerWeapon))
          (Utilities|IsValid _returnvalue_3
            (:Is%20Valid
              (bind _asfort_player_pawn_athena (Utilities|Casting|CastToFortPlayerPawnAthena (Actor|GetOwner _returnvalue_3)))
              (bind _output_get (Variables|References|SetAs%20Fort%20Player%20Pawn%20Athena _asfort_player_pawn_athena))
              (Transformation|AttachComponentToComponent (Variables|Default|GetNS_TetherWire) (Visual|GetWeaponMesh _returnvalue_3) "None" "SnapToTarget" "SnapToTarget" "KeepRelative" true) ; Jira #FORT-586042 - attaching this to a pawn mesh caused the projectile itself to be hidden, if the owning pawn was also hidden
              (Components|Activation|Activate _ns_tetherwire true)
              (CallFunction|UpdateFXState 0)
              (bind _output_get_2 (Variables|States|SetIsLeftHanded (not (or (.bSuccess (Utilities|Casting|CastToB_Weapon_NyxGlass (Weapon|GetCurrentWeapon _asfort_player_pawn_athena))) (== (FortWeapon|GetMagazineAmmoCount (Weapon|GetCurrentWeapon _asfort_player_pawn_athena)) 0)))))
              (Niagara|SetNiagaraVariable(Bool) _ns_tetherwire "IsLeftHand" _output_get_2)
              (FX|UpdateFXWire (Variables|Default|GetNS_TetherWire) (Transformation|GetSocketLocation (Variables|Default|GetFeyCrab_Projectile) "AttachmentPoint")))))))))

(event Custom|UpdateTheWire
  ;[0.5383 1 0.4202] Update Wire Length
  (Utilities|DoOncePerFrame
    (:%20%20
      (bind _returnvalue (CallFunction|GetOwnerWeapon))
      (Utilities|FlowControl|Sequence
        (:then_1
          (FX|UpdateFXWire (Variables|Default|GetNS_TetherWire) (Transformation|GetSocketLocation (Variables|Default|GetFeyCrab_Projectile) "AttachmentPoint")))))))

(event Custom|FirstTargetingDamage
  (params (TargetingRequestHandle :type Targeting%20Request%20Handle%20Structure))
  (bind _returnvalue (GameInstanceSubsystems|GetTargetingSubsystem))
  (bind _targets (TargetingSystem|TargetingResults|GetTargetingResultsActors _returnvalue TargetingRequestHandle))
  ;[1 1 1] First Targeting
  (CallFunction|BlastHitEnemyCheck _targets)
  (for _array_element _targets
    (Utilities|Array|Add (Variables|Damage|GetCurrentTargets) _array_element)
    (CallFunction|DealExplosiveDamage (Ability|GetFortAbilitySystemComponent _array_element)))
  (CallFunction|SecondTarget)
  (Actor|DestroyActor (Variables|Damage|GetDummyRearSphere)))

(event Custom|SecondTargetingDamage
  (params (TargetingRequestHandle :type Targeting%20Request%20Handle%20Structure))
  (bind _initialhitactor (Variables|Damage|GetInitialHitActor))
  ;[1 1 1] Second Targeting
  (for _array_element (TargetingSystem|TargetingResults|GetTargetingResultsActors (GameInstanceSubsystems|GetTargetingSubsystem) TargetingRequestHandle)
    (if (not (Utilities|Array|ContainsItem (Variables|Damage|GetCurrentTargets) _array_element))
      (Utilities|Array|Add (Variables|Damage|GetCurrentTargets) _array_element)
      (CallFunction|DealExplosiveDamage (Ability|GetFortAbilitySystemComponent _array_element))))
  (CallFunction|BlastHitEnemyCheck (Variables|Damage|GetCurrentTargets) true)
  (for _array_element_2 (Variables|Damage|GetCurrentTargets)
    (bind _asfort_athena_vehicle (Utilities|Casting|CastToFortAthenaVehicle _array_element_2))
    (if (not (== (Actor|GetOwner) (Vehicle|GetDriver (Utilities|Casting|CastToFortVehicleInterface _asfort_athena_vehicle))))
      (Class|FortAthenaVehicle|DestroyAllDamageableParts _asfort_athena_vehicle)))
  (if (not (Utilities|Array|ContainsItem (Variables|Damage|GetCurrentTargets) _initialhitactor))
    (CallFunction|DealExplosiveDamage (Ability|GetFortAbilitySystemComponent _initialhitactor)))
  (Utilities|Array|Clear (Variables|Damage|GetCurrentTargets))
  (Actor|DestroyActor (Variables|Damage|GetDummyFrontSphere)))

(event Custom|InAirExplosion
  ;[0.367 1 0.286] Hide wire when it's been pulled (triggered by anim notify in COPIED_GA_FEYCRAB_ExplosiveLaunch)
  (if (<= (+ (Variables|InAirTime|GetServerWorldTime) (ScalableFloat|GetValueatLevel (Variables|InAirTime|GetInAirExplosionTimer))) (- (Fortnite|Sync|Time|GetServerWorldTimeSeconds) (ScalableFloat|GetValueatLevel (Variables|InAirTime|GetInAirExplosionTimer))))
    (if (not (Variables|States|GetHitWall))
      (CallFunction|TriggerExplosion))
    (Utilities|Time|ClearandInvalidateTimerbyHandle (Variables|InAirTime|GetTH_InairExplosion))))

(event Custom|OnTouched:%20No%20Bounce
  (params (Hit%20Result :type Hit%20Result%20Structure))
  (bind (_bblockinghit _binitialoverlap _time _distance _location _impactpoint _normal _impactnormal _physmat _hitactor _hitcomponent) (Collision|BreakHitResult Hit%20Result))
  ;[1 1 1] EVENT -> Touched: Bounce off "NoBounce"
  (if (Components|ComponentHasTag _hitcomponent "ShouldCauseBounce")
    (Variables|/Script/Engine.MovementComponent|SetVelocity (* (Variables|/Script/Engine.MovementComponent|GetVelocity (Variables|Movement|GetProjectileMovementComponent)) 2.0) (Variables|Movement|GetProjectileMovementComponent))
    (else
      (Variables|Projectile|SetbStopSimulatingOnHit true))))

(event Custom|PulledWire
  (params (Target :type Self%20Object%20Reference))
  ;[0.367 1 0.286] Hide wire when it's been pulled (triggered by anim notify in COPIED_GA_FEYCRAB_ExplosiveLaunch)
  (bind _returnvalue (Utilities|Time|SetTimerbyEvent (EventDispatchers|CreateEvent :!event Custom|InAirExplosion) 0.1 true))
  (Variables|InAirTime|SetTH_InairExplosion _returnvalue)
  (CallFunction|UpdateFXState Target 2)
  (Utilities|Time|ClearandInvalidateTimerbyHandle (Variables|WireFuse|GetWireUpdateTimer))
  (Development|SetHiddeninGame (Variables|Default|GetWire) true false)
  (Variables|WireFuse|SetWirePulled true)
  (CallFunction|FuseActivated))

(event Custom|HitPlayer_Sound
  ;[1 1 1] Attach to Surface and Damage Players
  (Networking|SwitchHasAuthority
    (:Remote
      (Audio|PlaySoundatLocation "/DryBox/Items/FeyCrab/Audio/FeyCrab_Impact_Enemy_Cue.FeyCrab_Impact_Enemy_Cue" (Transformation|GetActorLocation (Variables|References|GetStuckPawn))))))

(event Custom|Activate%20Fuse
  ;[0.367 1 0.286] Hide wire when it's been pulled (triggered by anim notify in COPIED_GA_FEYCRAB_ExplosiveLaunch)
  (Variables|WireFuse|SetWirePulled true)
  (CallFunction|FuseActivated))

(event EventOnTouched
  (params (OtherActor :type Actor%20Object%20Reference) (OtherComp :type Primitive%20Component%20Object%20Reference) (HitResult :type Hit%20Result%20Structure%20%28by%20ref%29) (bIsOverlap :type Boolean))
  (bind (_returnvalue_x _returnvalue_y _returnvalue_z) (Math|Vector|BreakVector (Transformation|GetActorLocation)))
  (bind (_returnvalue_x_2 _returnvalue_y_2) (Math|Vector|BreakVector (Transformation|GetActorLocation (Variables|References|GetStuckPawn))))
  (bind (_bblockinghit _binitialoverlap _time _distance _location _impactpoint _normal _impactnormal _physmat _hitactor _hitcomponent) (Collision|BreakHitResult HitResult))
  (bind _projectilemovementcomponent (Variables|Movement|GetProjectileMovementComponent))
  ;[1 1 1] Attach to Surface and Damage Players
  (Variables|States|SetSurfaceImpacted true)
  (Utilities|FlowControl|Sequence
    (:then_0
      (bind _asfort_water_body_actor (Utilities|Casting|CastToFortWaterBodyActor OtherActor)
        (:then
          (CallFunction|OnProjectileWaterHit HitResult _asfort_water_body_actor))
        (:CastFailed
          (bind _asbuilding_actor (Utilities|Casting|CastToBuildingActor OtherActor)
            (:then
              (Variables|References|SetStuckBuildingActor _asbuilding_actor)
              (EventDispatchers|BindEventtoOnDied (Variables|References|GetStuckBuildingActor) (delegate Custom|OnBuilldingActorDeath))
              (bind _returnvalue (CallFunction|GetOwnerWeapon))
              (GameplayEffects|ApplyGameplayEffectToTarget :self (Ability|GetFortAbilitySystemComponent (Variables|/Script/Engine.Actor|GetInstigator _returnvalue)) :GameplayEffectClass "/DryBox/Items/FeyCrab/GE_FeyCrab_ImpactDamage.GE_FeyCrab_ImpactDamage_C" :Target (Ability|GetAbilitySystemComponent (Variables|References|GetStuckBuildingActor)) :Level 0.0))
            (:CastFailed
              (Utilities|SticktoAllyCheck OtherActor
                (:StickToActor
                  (if (not (Components|ComponentHasTag _hitcomponent "NoStick"))
                    (Variables|Projectile|SetbStopSimulatingOnHit true)
                    (Networking|SwitchHasAuthority
                      (:Authority
                        (bind _aspawn (Utilities|Casting|CastToPawn OtherActor))
                        (bind _output_get (Variables|References|SetStuckPawn _aspawn))
                        (CallFunction|InitialHitEnemyCheck _output_get)
                        (Utilities|FlowControl|Gate false
                          (:Exit
                            (CallFunction|BindEventtoPlayerDeath)
                            (if (ScalableFloat|GetValueasBool (Variables|Hotfixes|GetStickToPlayers))
                              (Gameplay|AuthorityAttachtoActor (Variables|Default|GetFortAttachableProjectile) HitResult (Variables|References|GetStuckPawn) (Variables|StickingLogic|GetOffsetDistanceFromPhysicsMesh) (Variables|StickingLogic|GetOffsetDistanceFromBone) 1.0)
                              (bind _returnvalue_2 (CallFunction|GetOwnerWeapon))
                              (GameplayEffects|ApplyGameplayEffectToTarget :self (Ability|GetFortAbilitySystemComponent (Variables|/Script/Engine.Actor|GetInstigator _returnvalue_2)) :GameplayEffectClass "/DryBox/Items/FeyCrab/GE_FeyCrab_ImpactDamage.GE_FeyCrab_ImpactDamage_C" :Target (Ability|GetAbilitySystemComponent (Variables|References|GetStuckPawn)) :Level 0.0)
                              (CallFunction|HitPlayerSound)
                              (else
                                (Variables|/Script/Engine.ProjectileMovementComponent|SetProjectileGravityScale 0.3 _projectilemovementcomponent)
                                (Projectile|SetVelocity _projectilemovementcomponent (Utilities|CalculateBounceOffVelocity))
                                (Transformation|SetActorLocation 0 (Math|Vector|MakeVector _returnvalue_x_2 _returnvalue_y_2 _returnvalue_z) false false)
                                (CallFunction|TriggerExplosion))))) ; Prevent atachment to pawns if already stopped.
                      ))))
                (:NoAllyStick
                  _))
              (Variables|Projectile|SetbStopSimulatingOnHit false))))))
    (:then_1
      (CallFunction|ImpactAndExplosionEffects HitResult))
    (:then_2
      (CallFunction|OnTouched:NoBounce HitResult))))

(event Custom|PlayPreExplosionSound
  (bind _mesh (Variables|Projectile_Rocket|GetMesh))
  (bind _returnvalue (Transformation|GetWorldLocation _mesh))
  ;[1 1 1] Fuse Activation and Explosion Timer
  (Audio|PlaySoundatLocation "/DryBox/Items/FeyCrab/Audio/FeyCrab_PreExplosion_Cue.FeyCrab_PreExplosion_Cue" _returnvalue)
  (Fortnite|ShowSoundIndicatorLocal :Location _returnvalue :MaxAudibleDistance 5e+03 :IndicatorType "Gunshot" :Instigator self :Tint "(R=0.000000,G=0.000000,B=0.000000,A=0.000000)" :AttenuationMaxRangeMultiplier 1.0 :AttenuationIntensityMultiplier 1.0))

(event Custom|FuseActivated
  (bind _fusewarningtimer (Variables|WireFuse|GetFuseWarningTimer))
  ;[1 1 1] Fuse Activation and Explosion Timer
  (if (not (Variables|WireFuse|GetFusePrimed))
    (if (and (Variables|States|GetSurfaceImpacted) (Variables|WireFuse|GetWirePulled))
      (if (Utilities|Time|IsValidTimerHandle _fusewarningtimer)
        (Utilities|Time|ClearandInvalidateTimerbyHandle _fusewarningtimer)
        (else
          _))
      (Variables|WireFuse|SetFusePrimed true)
      (CallFunction|FXonFuseCountdown)
      (CallFunction|PlayPreExplosionSound)
      (Networking|SwitchHasAuthority
        (:Authority
          (bind _returnvalue (Utilities|Time|SetTimerbyEvent (EventDispatchers|CreateEvent :!event Custom|TriggerExplosion) (ScalableFloat|GetValueatLevel (Variables|WireFuse|GetFuseActivatedTimer))))
          (Variables|WireFuse|SetFuseWarningTimer _returnvalue))))))

(event OnProjectileStop(ProjectileMovementComponent)
  (params (ImpactResult :type Hit%20Result%20Structure%20%28by%20ref%29))
  (bind (_returnvalue_x _returnvalue_y _returnvalue_z) (Math|Vector|BreakVector (Transformation|GetActorLocation)))
  (bind (_returnvalue_x_2 _returnvalue_y_2) (Math|Vector|BreakVector (Transformation|GetActorLocation (Variables|References|GetStuckPawn))))
  (bind _projectilemovementcomponent (Variables|Movement|GetProjectileMovementComponent))
  ;[1 1 1] Attach to Surface and Damage Players
  (Variables|States|SetHas%20Stopped true)
  (Utilities|FlowControl|Gate false
    (:Exit
      (CallFunction|BindEventtoPlayerDeath)
      (if (ScalableFloat|GetValueasBool (Variables|Hotfixes|GetStickToPlayers))
        (Gameplay|AuthorityAttachtoActor (Variables|Default|GetFortAttachableProjectile) (stale EventOnTouched HitResult) (Variables|References|GetStuckPawn) (Variables|StickingLogic|GetOffsetDistanceFromPhysicsMesh) (Variables|StickingLogic|GetOffsetDistanceFromBone) 1.0)
        (bind _returnvalue (CallFunction|GetOwnerWeapon))
        (GameplayEffects|ApplyGameplayEffectToTarget :self (Ability|GetFortAbilitySystemComponent (Variables|/Script/Engine.Actor|GetInstigator _returnvalue)) :GameplayEffectClass "/DryBox/Items/FeyCrab/GE_FeyCrab_ImpactDamage.GE_FeyCrab_ImpactDamage_C" :Target (Ability|GetAbilitySystemComponent (Variables|References|GetStuckPawn)) :Level 0.0)
        (CallFunction|HitPlayerSound)
        (else
          (Variables|/Script/Engine.ProjectileMovementComponent|SetProjectileGravityScale 0.3 _projectilemovementcomponent)
          (Projectile|SetVelocity _projectilemovementcomponent (Utilities|CalculateBounceOffVelocity))
          (Transformation|SetActorLocation 0 (Math|Vector|MakeVector _returnvalue_x_2 _returnvalue_y_2 _returnvalue_z) false false)
          (CallFunction|TriggerExplosion))))) ; Prevent atachment to pawns if already stopped.
)

(event Custom|FuseTimerMax
  ;[1 1 1] Fuse Activation and Explosion Timer
  (if (and (not (Variables|WireFuse|GetFusePrimed)) (not (Variables|States|GetbTriggeredExplosion)))
    (Variables|WireFuse|SetFusePrimed true)
    (CallFunction|FXonFuseCountdown)
    (CallFunction|PlayPreExplosionSound)
    (Networking|SwitchHasAuthority
      (:Authority
        (bind _returnvalue (Utilities|Time|SetTimerbyEvent (EventDispatchers|CreateEvent :!event Custom|TriggerExplosion) (ScalableFloat|GetValueatLevel (Variables|WireFuse|GetFuseActivatedTimer))))
        (Variables|WireFuse|SetFuseWarningTimer _returnvalue)))))

(event Custom|OnBuilldingActorDeath
  (params (DamagedActor :type Actor%20Object%20Reference) (Damage :type Float%20%28single-precision%29) (InstigatedBy :type Controller%20Object%20Reference) (DamageCauser :type Actor%20Object%20Reference) (HitLocation :type Vector) (FHitComponent :type Primitive%20Component%20Object%20Reference) (BoneName :type Name) (Momentum :type Vector))
  ;[1 1 1] Attach to Surface and Damage Players
  (Networking|SwitchHasAuthority
    (:Authority
      (CallFunction|TriggerExplosion))))

(event Custom|TriggerExplosion
  ;[1 1 1] Trigger Explosion
  (Networking|SwitchHasAuthority
    (:Authority
      (if (not (Variables|States|GetbTriggeredExplosion))
        (Variables|States|SetbTriggeredExplosion true)
        (CallFunction|TriggerExplosion)))
    (:Remote
      _))
  (Variables|Effects|SetStopLocation (Transformation|GetWorldLocation (Variables|Projectile_Rocket|GetMesh)))
  (if (not (Delivery|IsBeingKilled))
    (Fortnite|ShowSoundIndicatorLocal :Location (Transformation|GetWorldLocation (Variables|Projectile_Rocket|GetMesh)) :MaxAudibleDistance 5e+03 :IndicatorType "Gunshot" :Instigator self :Tint "(R=0.000000,G=0.000000,B=0.000000,A=0.000000)" :AttenuationMaxRangeMultiplier 1.0 :AttenuationIntensityMultiplier 1.0)
    (Networking|SwitchHasAuthority
      (:Authority
        (CallFunction|FirstTargeting))
      (:Remote
        (CallFunction|ManualExplode)))))

(event Custom|Impact%20and%20Explosion%20Effects
  (params (Hit%20Result :type Hit%20Result%20Structure))
  (bind (_bblockinghit _binitialoverlap _time _distance _location _impactpoint _normal _impactnormal _physmat _hitactor) (Collision|BreakHitResult Hit%20Result))
  (bind _returnvalue_location (Math|Transform|BreakTransform (Transformation|GetWorldTransform (Variables|Default|GetNS_FeycrabLauncher_Prj_SurfImpact))))
  (bind _returnvalue_location_2 (Math|Transform|BreakTransform (Transformation|GetWorldTransform (Variables|Projectile_Rocket|GetMesh))))
  (bind _ns_feycrablauncher_prj_surfimpact (Variables|Default|GetNS_FeycrabLauncher_Prj_SurfImpact))
  ;[1 1 1] EVENT -> Stop: Impact and Explosion Stuff
  (Utilities|FlowControl|Sequence
    (:then_0
      (Variables|Damage|SetInitialHitActor _hitactor)
      (Variables|Effects|Explosion|Decal|SetDecalLocation _impactpoint)
      (bind _output_get (Variables|Effects|SetStopLocation (+ _impactpoint (* _impactnormal (Variables|Effects|Explosion|GetExplosion%20Impact%20Offset)))))
      (Variables|Effects|SetStopRotZ (Math|Rotator|MakeRotfromZ _impactnormal))
      (Utilities|IsValid _physmat
        (:Is%20Valid
          (Variables|Effects|SetSurfaceType (Variables|/Script/PhysicsCore.PhysicalMaterial|GetSurfaceType _physmat)))
        (:Is%20Not%20Valid
          (Variables|Effects|SetSurfaceType "SurfaceType_Default")))
      (Class|BPSurfaceTypeSoundComponent|SetSurfaceType (Variables|Sound|GetBP_SurfaceTypeSoundComponent) (Variables|Effects|GetSurfaceType) (Variables|References|GetStuckPawn))
      (Variables|States|SetHitWall true)
      (Utilities|Time|ClearandInvalidateTimerbyHandle (Variables|InAirTime|GetTH_InairExplosion))
      (CallFunction|UpdateFXState 1)
      (Components|Activation|Deactivate (Variables|Default|GetRocketTrailPS_Niagara))
      (Variables|RocketTrail|SetRocket%20Trial%20Asset 0)
      (CallFunction|FuseActivated)
      (CallFunction|FXCreateFuseFX 0))
    (:then_1
      (Utilities|FlowControl|DelayUntilNextTick)
      (Components|Activation|Activate _ns_feycrablauncher_prj_surfimpact true)
      (Niagara|SetNiagaraVariable(Int32) _ns_feycrablauncher_prj_surfimpact "PhysicalSurface" (Math|Conversions|ToInteger(Byte) (Variables|Effects|GetSurfaceType)))
      (Utilities|FlowControl|Sequence
        (:then_0
          (Rendering|Debug|DrawDebugLine _returnvalue_location _returnvalue_location_2 "(R=0.541667,G=0.223664,B=0.000000,A=1.000000)" 1e+01 4.0))
        (:then_1
          (CallFunction|GetOwnerWeapon)
          (Audio|PlaySoundatLocation "/DryBox/Items/FeyCrab/Audio/FeyCrab_Impact_Scenery_Cue.FeyCrab_Impact_Scenery_Cue" _output_get))))))

(event Custom|ManualExplode
  (bind (_outxfm_location _outxfm_rotation _outxfm_scale) (Math|Transform|BreakTransform (CallFunction|FXGetExplosionTransform)))
  (bind (_outxfm_location_2 _outxfm_rotation_2 _outxfm_scale_2) (Math|Transform|BreakTransform (CallFunction|FXGetExplosionTransform)))
  (bind _rockettrailps_niagara (Variables|Default|GetRocketTrailPS_Niagara))
  (bind _rockettrailps (Variables|RocketTrail|GetRocketTrailPS))
  ;[1 0.5861 0.08121] EVENT -> Exploded
  (bind _asb_fey_crab_launcher (Utilities|Casting|CastToB_FeyCrab_Launcher (Variables|/Script/FortniteGame.FortPawn|GetCurrentWeapon (Variables|References|GetAs%20Fort%20Player%20Pawn%20Athena)))
    (:then
      (if (>= (Variables|/DryBox/Items/FeyCrab/B_FeyCrab_Launcher.B_FeyCrab_Launcher_C|GetProjectileInWorld _asb_fey_crab_launcher) 2)
        (Utilities|Time|SetTimerbyEvent (delegate Custom|SendExplodeEvent) (Variables|WeaponReload|GetDelayRetryReload))
        (else
          (CallFunction|SendExplodeEvent)))
      (Utilities|FlowControl|DoOnce false
        (:Completed
          (Networking|CanExecuteCosmeticEvents
            (:True
              (Rendering|SetActorHiddenInGame true)
              (Utilities|IsValid _rockettrailps
                (:Is%20Valid
                  (Components|Activation|Deactivate _rockettrailps))
                (:Is%20Not%20Valid
                  _))
              (Utilities|IsValid _rockettrailps_niagara
                (:Is%20Valid
                  (Components|Activation|Deactivate _rockettrailps_niagara))
                (:Is%20Not%20Valid
                  _))
              (if (or (Utilities|Enum|Equal(Enum) (Variables|Effects|GetSurfaceType) "SurfaceType8") (Water|IsinWater (Variables|Water|GetWaterInteractionComponent)))
                ;;[0.6757 0 1] Spawn Water FX and Sound
                (Class|BPSurfaceTypeSoundComponent|SetSurfaceType (Variables|Sound|GetBP_SurfaceTypeSoundComponent) "SurfaceType8" 0)
                (bind _returnvalue (Effects|Components|ParticleSystem|SpawnFXAtLocation (Variables|Effects|Explosion|GetExplosion%20Asset_Water) _outxfm_location _outxfm_rotation _outxfm_scale))
                (FX|FXonSpawnExplosion _returnvalue)
                (Utilities|SpawnFNRadialForce (Utilities|StopLocationTransform) (Variables|PhysicsFields|GetPlayrate) (Variables|PhysicsFields|GetRadius) (Variables|PhysicsFields|GetStrength) (Variables|PhysicsFields|GetTime%20Decay))
                (Audio|PlaySoundatLocation (Variables|Sounds|GetRocket_Projectile_Explosion_Water_Sound) (Variables|Effects|GetStopLocation))
                (else
                  ;;[0.2043 0 1 1 30] Niagara Explosion Branch
                  (bind _returnvalue_2 (Effects|Components|ParticleSystem|SpawnFXAtLocation (Variables|Effects|Explosion|GetExplosion%20Asset) _outxfm_location_2 _outxfm_rotation_2 _outxfm_scale_2))
                  (FX|FXonSpawnExplosion _returnvalue_2)
                  (Utilities|SpawnFNRadialForce (Utilities|StopLocationTransform) (Variables|PhysicsFields|GetPlayrate) (Variables|PhysicsFields|GetRadius) (Variables|PhysicsFields|GetStrength) (Variables|PhysicsFields|GetTime%20Decay))
                  (Audio|PlaySoundatLocation (Variables|Sounds|GetRocket_Projectile_Explosion_Sound) (Variables|Effects|GetStopLocation))))
              (Class|BPSurfaceTypeSoundComponent|PlaySound (Variables|Sound|GetBP_SurfaceTypeSoundComponent) (Variables|Effects|GetStopLocation))
              (Utilities|FlowControl|Sequence
                (:then_0
                  ;;[1 1 1] FX for surface explosion if underwater (distance checked)
                  (if (Variables|Water|GetIs%20Underwater)
                    (CallFunction|FXonExplodeinWater)))
                (:then_1
                  ;;[0.6661 0.1729 1] Shake that camera / Post Process
                  (if (<= (Math|Vector|VectorLength (- (Variables|Effects|GetStopLocation) (Transformation|GetWorldLocation (Variables|/Script/Engine.PlayerCameraManager|GetTransformComponent (Game|GetPlayerCameraManager 0))))) (Variables|/Script/Engine.PostProcessComponent|GetBlendRadius (Variables|Default|GetPP_Explode)))
                    (bind (_direction _ppfade) (Timeline|PPFader PlayFromStart)
                      (:Update
                        (Variables|/Script/Engine.PostProcessComponent|SetBlendWeight _ppfade (Variables|Default|GetPP_Explode))))))
                (:then_2
                  (GameplayCue|ExecuteGameplayCue%28Burst%29Local self (Variables|Effects|GetFeedbackCue) (Ability|GameplayCue|MakeGameplayCueParameters 0.0)) ; Cam shake + controller vibration
                  (Audio|Components|Audio|FadeOut (Variables|Default|GetRocket_AudioComponent) 1 0.0 "Linear" :!class /Script/Engine.AudioComponent))))))))
    (:CastFailed
      _)))

(event EventOnExploded
  (params (HitActors :type Array%20of%20Actor%20Object%20References) (HitResults :type Array%20of%20Hit%20Result%20Structures))
  (bind (_outxfm_location _outxfm_rotation _outxfm_scale) (Math|Transform|BreakTransform (CallFunction|FXGetExplosionTransform)))
  (bind (_outxfm_location_2 _outxfm_rotation_2 _outxfm_scale_2) (Math|Transform|BreakTransform (CallFunction|FXGetExplosionTransform)))
  (bind _rockettrailps_niagara (Variables|Default|GetRocketTrailPS_Niagara))
  (bind _rockettrailps (Variables|RocketTrail|GetRocketTrailPS))
  ;[1 0.5861 0.08121] EVENT -> Exploded
  (bind _asb_fey_crab_launcher (Utilities|Casting|CastToB_FeyCrab_Launcher (Variables|/Script/FortniteGame.FortPawn|GetCurrentWeapon (Variables|References|GetAs%20Fort%20Player%20Pawn%20Athena)))
    (:then
      (if (>= (Variables|/DryBox/Items/FeyCrab/B_FeyCrab_Launcher.B_FeyCrab_Launcher_C|GetProjectileInWorld _asb_fey_crab_launcher) 2)
        (Utilities|Time|SetTimerbyEvent (delegate Custom|SendExplodeEvent) (Variables|WeaponReload|GetDelayRetryReload))
        (else
          (CallFunction|SendExplodeEvent)))
      (Utilities|FlowControl|DoOnce false
        (:Completed
          (Networking|CanExecuteCosmeticEvents
            (:True
              (Rendering|SetActorHiddenInGame true)
              (Utilities|IsValid _rockettrailps
                (:Is%20Valid
                  (Components|Activation|Deactivate _rockettrailps))
                (:Is%20Not%20Valid
                  _))
              (Utilities|IsValid _rockettrailps_niagara
                (:Is%20Valid
                  (Components|Activation|Deactivate _rockettrailps_niagara))
                (:Is%20Not%20Valid
                  _))
              (if (or (Utilities|Enum|Equal(Enum) (Variables|Effects|GetSurfaceType) "SurfaceType8") (Water|IsinWater (Variables|Water|GetWaterInteractionComponent)))
                ;;[0.6757 0 1] Spawn Water FX and Sound
                (Class|BPSurfaceTypeSoundComponent|SetSurfaceType (Variables|Sound|GetBP_SurfaceTypeSoundComponent) "SurfaceType8" 0)
                (bind _returnvalue (Effects|Components|ParticleSystem|SpawnFXAtLocation (Variables|Effects|Explosion|GetExplosion%20Asset_Water) _outxfm_location _outxfm_rotation _outxfm_scale))
                (FX|FXonSpawnExplosion _returnvalue)
                (Utilities|SpawnFNRadialForce (Utilities|StopLocationTransform) (Variables|PhysicsFields|GetPlayrate) (Variables|PhysicsFields|GetRadius) (Variables|PhysicsFields|GetStrength) (Variables|PhysicsFields|GetTime%20Decay))
                (Audio|PlaySoundatLocation (Variables|Sounds|GetRocket_Projectile_Explosion_Water_Sound) (Variables|Effects|GetStopLocation))
                (else
                  ;;[0.2043 0 1 1 30] Niagara Explosion Branch
                  (bind _returnvalue_2 (Effects|Components|ParticleSystem|SpawnFXAtLocation (Variables|Effects|Explosion|GetExplosion%20Asset) _outxfm_location_2 _outxfm_rotation_2 _outxfm_scale_2))
                  (FX|FXonSpawnExplosion _returnvalue_2)
                  (Utilities|SpawnFNRadialForce (Utilities|StopLocationTransform) (Variables|PhysicsFields|GetPlayrate) (Variables|PhysicsFields|GetRadius) (Variables|PhysicsFields|GetStrength) (Variables|PhysicsFields|GetTime%20Decay))
                  (Audio|PlaySoundatLocation (Variables|Sounds|GetRocket_Projectile_Explosion_Sound) (Variables|Effects|GetStopLocation))))
              (Class|BPSurfaceTypeSoundComponent|PlaySound (Variables|Sound|GetBP_SurfaceTypeSoundComponent) (Variables|Effects|GetStopLocation))
              (Utilities|FlowControl|Sequence
                (:then_0
                  ;;[1 1 1] FX for surface explosion if underwater (distance checked)
                  (if (Variables|Water|GetIs%20Underwater)
                    (CallFunction|FXonExplodeinWater)))
                (:then_1
                  ;;[0.6661 0.1729 1] Shake that camera / Post Process
                  (if (<= (Math|Vector|VectorLength (- (Variables|Effects|GetStopLocation) (Transformation|GetWorldLocation (Variables|/Script/Engine.PlayerCameraManager|GetTransformComponent (Game|GetPlayerCameraManager 0))))) (Variables|/Script/Engine.PostProcessComponent|GetBlendRadius (Variables|Default|GetPP_Explode)))
                    (Timeline|PPFader PlayFromStart)))
                (:then_2
                  (GameplayCue|ExecuteGameplayCue%28Burst%29Local self (Variables|Effects|GetFeedbackCue) (Ability|GameplayCue|MakeGameplayCueParameters 0.0)) ; Cam shake + controller vibration
                  (Audio|Components|Audio|FadeOut (Variables|Default|GetRocket_AudioComponent) 1 0.0 "Linear" :!class /Script/Engine.AudioComponent))))))))
    (:CastFailed
      _)))

(event EventOnDisableForKill
  (bind (_outxfm_location _outxfm_rotation _outxfm_scale) (Math|Transform|BreakTransform (CallFunction|FXGetExplosionTransform)))
  (bind (_outxfm_location_2 _outxfm_rotation_2 _outxfm_scale_2) (Math|Transform|BreakTransform (CallFunction|FXGetExplosionTransform)))
  (bind _rockettrailps_niagara (Variables|Default|GetRocketTrailPS_Niagara))
  (bind _rockettrailps (Variables|RocketTrail|GetRocketTrailPS))
  ;[1 0.5861 0.08121] EVENT -> Exploded
  (bind _asb_fey_crab_launcher (Utilities|Casting|CastToB_FeyCrab_Launcher (Variables|/Script/FortniteGame.FortPawn|GetCurrentWeapon (Variables|References|GetAs%20Fort%20Player%20Pawn%20Athena)))
    (:then
      (if (>= (Variables|/DryBox/Items/FeyCrab/B_FeyCrab_Launcher.B_FeyCrab_Launcher_C|GetProjectileInWorld _asb_fey_crab_launcher) 2)
        (Utilities|Time|SetTimerbyEvent (delegate Custom|SendExplodeEvent) (Variables|WeaponReload|GetDelayRetryReload))
        (else
          (CallFunction|SendExplodeEvent)))
      (Utilities|FlowControl|DoOnce false
        (:Completed
          (Networking|CanExecuteCosmeticEvents
            (:True
              (Rendering|SetActorHiddenInGame true)
              (Utilities|IsValid _rockettrailps
                (:Is%20Valid
                  (Components|Activation|Deactivate _rockettrailps))
                (:Is%20Not%20Valid
                  _))
              (Utilities|IsValid _rockettrailps_niagara
                (:Is%20Valid
                  (Components|Activation|Deactivate _rockettrailps_niagara))
                (:Is%20Not%20Valid
                  _))
              (if (or (Utilities|Enum|Equal(Enum) (Variables|Effects|GetSurfaceType) "SurfaceType8") (Water|IsinWater (Variables|Water|GetWaterInteractionComponent)))
                ;;[0.6757 0 1] Spawn Water FX and Sound
                (Class|BPSurfaceTypeSoundComponent|SetSurfaceType (Variables|Sound|GetBP_SurfaceTypeSoundComponent) "SurfaceType8" 0)
                (bind _returnvalue (Effects|Components|ParticleSystem|SpawnFXAtLocation (Variables|Effects|Explosion|GetExplosion%20Asset_Water) _outxfm_location _outxfm_rotation _outxfm_scale))
                (FX|FXonSpawnExplosion _returnvalue)
                (Utilities|SpawnFNRadialForce (Utilities|StopLocationTransform) (Variables|PhysicsFields|GetPlayrate) (Variables|PhysicsFields|GetRadius) (Variables|PhysicsFields|GetStrength) (Variables|PhysicsFields|GetTime%20Decay))
                (Audio|PlaySoundatLocation (Variables|Sounds|GetRocket_Projectile_Explosion_Water_Sound) (Variables|Effects|GetStopLocation))
                (else
                  ;;[0.2043 0 1 1 30] Niagara Explosion Branch
                  (bind _returnvalue_2 (Effects|Components|ParticleSystem|SpawnFXAtLocation (Variables|Effects|Explosion|GetExplosion%20Asset) _outxfm_location_2 _outxfm_rotation_2 _outxfm_scale_2))
                  (FX|FXonSpawnExplosion _returnvalue_2)
                  (Utilities|SpawnFNRadialForce (Utilities|StopLocationTransform) (Variables|PhysicsFields|GetPlayrate) (Variables|PhysicsFields|GetRadius) (Variables|PhysicsFields|GetStrength) (Variables|PhysicsFields|GetTime%20Decay))
                  (Audio|PlaySoundatLocation (Variables|Sounds|GetRocket_Projectile_Explosion_Sound) (Variables|Effects|GetStopLocation))))
              (Class|BPSurfaceTypeSoundComponent|PlaySound (Variables|Sound|GetBP_SurfaceTypeSoundComponent) (Variables|Effects|GetStopLocation))
              (Utilities|FlowControl|Sequence
                (:then_0
                  ;;[1 1 1] FX for surface explosion if underwater (distance checked)
                  (if (Variables|Water|GetIs%20Underwater)
                    (CallFunction|FXonExplodeinWater)))
                (:then_1
                  ;;[0.6661 0.1729 1] Shake that camera / Post Process
                  (if (<= (Math|Vector|VectorLength (- (Variables|Effects|GetStopLocation) (Transformation|GetWorldLocation (Variables|/Script/Engine.PlayerCameraManager|GetTransformComponent (Game|GetPlayerCameraManager 0))))) (Variables|/Script/Engine.PostProcessComponent|GetBlendRadius (Variables|Default|GetPP_Explode)))
                    (Timeline|PPFader PlayFromStart)))
                (:then_2
                  (GameplayCue|ExecuteGameplayCue%28Burst%29Local self (Variables|Effects|GetFeedbackCue) (Ability|GameplayCue|MakeGameplayCueParameters 0.0)) ; Cam shake + controller vibration
                  (Audio|Components|Audio|FadeOut (Variables|Default|GetRocket_AudioComponent) 1 0.0 "Linear" :!class /Script/Engine.AudioComponent))))))))
    (:CastFailed
      _)))

(event Custom|SendExplodeEvent
  ;[1 0.5861 0.08121] EVENT -> Exploded
  (Ability|SendGameplayEventtoActor (Variables|References|GetAs%20Fort%20Player%20Pawn%20Athena) "(TagName=\"Event.Ability.FeyCrab.Proj.Explode\")"))

(event Custom|BindEventTo%20PlayerDeath
  ;[1 1 1] Comment
  (bind _asplayer_pawn_athena (Utilities|Casting|CastToPlayerPawn_Athena (Variables|References|GetStuckPawn))
    (:then
      (EventDispatchers|AssignOnDied _asplayer_pawn_athena (delegate Custom|PlayerDiedEvent)))
    (:CastFailed
      (bind _asfort_aipawn (Utilities|Casting|CastToFortAIPawn (Variables|References|GetStuckPawn))
        (:then
          (EventDispatchers|AssignOnDied _asfort_aipawn (delegate Custom|AIDiedEvent)))
        (:CastFailed
          (Game|BindEventtoOnDestroyed (Variables|References|GetStuckPawn) (delegate Custom|OnPawnDestroyed)))))))

(event OnFinalizeAttachment(FortAttachableProjectile)
  (params (WorldLocation :type Vector%20%28by%20ref%29 :default "0, 0, 0") (WorldRotation :type Rotator%20%28by%20ref%29 :default "0, 0, 0") (SocketName :type Name%20%28by%20ref%29 :default "None"))
  ;[1 1 1] Handle overriding finalized attachment if head is Hit
  (if (== SocketName "head")
    (Gameplay|OverrideFinalizedAttachmentParams (Variables|Default|GetFortAttachableProjectile) (+ WorldLocation "0.000000,0.000000,-25.000000") WorldRotation (Variables|StickingLogic|Getoffset%20socket))
    (else
      (Gameplay|OverrideFinalizedAttachmentParams (Variables|Default|GetFortAttachableProjectile) WorldLocation WorldRotation SocketName))))

(event Custom|PlayerDiedEvent
  (params (DamagedActor :type Actor%20Object%20Reference) (Damage :type Float%20%28single-precision%29) (InstigatedBy :type Controller%20Object%20Reference) (DamageCauser :type Actor%20Object%20Reference) (HitLocation :type Vector) (FHitComponent :type Primitive%20Component%20Object%20Reference) (BoneName :type Name) (Momentum :type Vector))
  ;[1 1 1] Comment
  (Networking|SwitchHasAuthority
    (:Authority
      (CallFunction|TriggerExplosion))))

(event EventDestroyed
  (bind _rockettrailps (Variables|RocketTrail|GetRocketTrailPS))
  ;[1 1 1] EVENT -> Destroyed: Deactivating Rocket Trail Particle systems on component destroyed
  (Networking|CanExecuteCosmeticEvents
    (:True
      (Utilities|IsValid _rockettrailps
        (:Is%20Valid
          _)
        (:Is%20Not%20Valid
          (Utilities|IsValid (Variables|Default|GetRocketTrailPS_Niagara)
            (:Is%20Valid
              _))))
      (Components|Activation|Deactivate _rockettrailps))))

(event Custom|AIDiedEvent
  (params (DamagedActor :type Actor%20Object%20Reference) (Damage :type Float%20%28single-precision%29) (InstigatedBy :type Controller%20Object%20Reference) (DamageCauser :type Actor%20Object%20Reference) (HitLocation :type Vector) (FHitComponent :type Primitive%20Component%20Object%20Reference) (BoneName :type Name) (Momentum :type Vector))
  ;[1 1 1] Comment
  (Networking|SwitchHasAuthority
    (:Authority
      (CallFunction|TriggerExplosion))))

(event Custom|OnPawnDestroyed
  (params (DestroyedActor :type Actor%20Object%20Reference))
  ;[1 1 1] Comment
  (Networking|SwitchHasAuthority
    (:Authority
      (CallFunction|TriggerExplosion))))

(event Custom|UpdateFXState
  (params (State :type Integer))
  (bind _islefthanded (Variables|States|GetIsLeftHanded))
  ;[1 0.9109 0] State of wire and fuse fx
  ;;[1 1 1] Can never go back; ie cant go from pulled (2) back to hitwall(1)
  (if (not (> (Variables|FX|GetFXState) State))
    (Utilities|FlowControl|Sequence
      (:then_0
        (Variables|FX|SetFXState State))
      (:then_1
        (FX|SetRopeFXState (Variables|Default|GetNS_TetherWire) State))
      (:then_2
        (switch int (Variables|FX|GetFXState)
          (:1
            _)
          (:2
            (Utilities|FlowControl|Sequence
              (:then_0
                _)
              (:then_1
                ;;[1 1 1] Fuse being pulled "POP" FX
                (Effects|SpawnFXSystemAttached :FXSystemAsset "/DryBox/Items/FeyCrab/FX/NS_Feycrab_Prj_WirePulled.NS_Feycrab_Prj_WirePulled" :AttachToComponent (Variables|Default|GetFeyCrab_Projectile) :AttachPointName "AttachmentPoint" :Rotation "0.000000,180.000000,0.000000" :Scale "1.000000,1.000000,1.000000" :LocationType "SnapToTarget" :bAutoDestroy true :PoolingMethod "None" :bAutoActivate true)
                (Niagara|SetNiagaraVariable(Bool) (Variables|Default|GetNS_TetherWire) "IsLeftHand" _islefthanded)
                (Transformation|SetWorldRotation (Variables|Default|GetNS_TetherWire) (Math|Rotator|CombineRotators (select _islefthanded "0, 0, 0" "0, 0, 0") (Math|Rotator|FindLookatRotation (Transformation|GetWorldLocation (Variables|Default|GetNS_TetherWire)) (Transformation|GetActorLocation))) false true)))))
        (Components|Activation|Deactivate (Variables|Default|GetNS_FeycrabLauncher_Prj_Idle))))))

(event Custom|OnProjectileWaterHit
  (params (WaterHit :type Hit%20Result%20Structure%20%28by%20ref%29) (WaterBody :type Fort%20Water%20Body%20Actor%20Object%20Reference))
  (bind (_bblockinghit _binitialoverlap _time _distance _location _impactpoint _normal _impactnormal) (Collision|BreakHitResult WaterHit))
  ;[1 1 1] On enter water volume
  (Variables|Water|SetIs%20Underwater true)
  (Networking|CanExecuteCosmeticEvents
    (:True
      (bind _output_get (Variables|Water|SetWater%20Hit WaterHit))
      (bind (_output_get_bblockinghit _output_get_binitialoverlap _output_get_time _output_get_distance _output_get_location _output_get_impactpoint _output_get_normal _output_get_impactnormal _output_get_physmat) (Collision|BreakHitResult _output_get))
      (Utilities|IsValid _output_get_physmat
        (:Is%20Valid
          (Niagara|SpawnSystematLocation (Variables|Water|FX|GetOnEnterWaterFX) _impactpoint (Math|Conversions|RotationFromXVector _impactnormal))
          (Utilities|SpawnFNRadialForce (Math|Transform|MakeTransform _impactpoint) 0.2 2.5e+02 0.5 0.66))))))

(event Custom|FX_CreateFuseFX
  (params (State :type Integer))
  (Networking|CanExecuteCosmeticEvents
    (:True
      ;[1 1 1] Spawn the fuse FX - will quickly ramp in but not flash at this point (this is once embedded in surface)
      (bind _returnvalue (Effects|SpawnFXSystemAttached :FXSystemAsset "/DryBox/Items/FeyCrab/FX/NS_FeycrabLauncher_Prj_GlowWarn.NS_FeycrabLauncher_Prj_GlowWarn" :AttachToComponent (Variables|Default|GetFeyCrab_Projectile) :AttachPointName "FX_Source" :Scale "1.000000,1.000000,1.000000" :LocationType "SnapToTarget" :bAutoDestroy true :PoolingMethod "None" :bAutoActivate true))
      (bind _asniagara_particle_system_component (Utilities|Casting|CastToNiagaraComponent _returnvalue))
      (Rendering|SetTranslucentSortPriority _asniagara_particle_system_component 1)
      (bind _output_get (Variables|Default|SetNS_FuseFX _asniagara_particle_system_component))
      (Niagara|SetNiagaraVariable(Int32) _output_get "State" State))))

(event Custom|FX_OnExplodeInWater
  ;[0.9497 0.1987 1] Spawn a water-surface explosion, when projectile is underwater & water surface position is within AOE range
  (bind _outwaterposition (CallFunction|GetWaterSurfaceLocationforExplosion (Variables|Effects|GetStopLocation) (Variables|Water|GetWater%20Hit)))
  (bind _distance (Class|MissionBlueprintFunctionLibrary|DistanceBetweenTwoVectors _outwaterposition (Variables|Effects|GetStopLocation)))
  (if (<= _distance (* (ScalableFloat|GetValueatLevel (Variables|Hotfixes|GetExplosionTargetingRadius)) 2.0))
    (Niagara|SpawnSystematLocation "/DryBox/Items/FeyCrab/FX/NS_Feycrab_Prj_WaterSurfaceBlast.NS_Feycrab_Prj_WaterSurfaceBlast" _outwaterposition)))

(event Custom|FX_OnFuseCountdown
  ;[1 1 1] If opposing player force FX
  (Networking|CanExecuteCosmeticEvents
    (:True
      (if (Utilities|Enum|Equal(Enum) (Networking|GetLocalRole (Variables|References|GetAs%20Fort%20Player%20Pawn%20Athena)) "ROLE_SimulatedProxy")
        (Utilities|IsValid self
          (:Is%20Valid
            (CallFunction|FXCreateFuseFX 1)))
        (else
          (bind _ns_fusefx (Variables|Default|GetNS_FuseFX)
            (:then
              (Niagara|SetNiagaraVariable(Int32) _ns_fusefx "State" 1))
            (:else
              ;;[1 1 1] If the projectile wasn't already embedded, create fuse fx anyway
              (CallFunction|FXCreateFuseFX 1))))))))

(fn BlastHitEnemyCheck
  (params (List%20of%20Hit%20Actors :type Array%20of%20Actor%20Object%20References) (Final%20run :type Boolean))
  (for _array_element List%20of%20Hit%20Actors
    (if (Utilities|IsValid _array_element :!class /Script/Engine.KismetSystemLibrary)
      (bind _asfort_player_pawn (Utilities|Casting|CastToFortPlayerPawn _array_element))
      (if (not (== (Variables|References|GetAs%20Fort%20Player%20Pawn%20Athena) _asfort_player_pawn))
        (bind _asfort_player_state_athena (Utilities|Casting|CastToFortPlayerStateAthena (Variables|/Script/Engine.Controller|GetPlayerState (Pawn|GetController _asfort_player_pawn))))
        (if (Class|FortPlayerStateAthena|IsinSquad _asfort_player_state_athena)
          _
          (elif Final%20run
            (if (Guid|IsValidGuid (Variables|References|GetHitPlayer))
              (if (not (== (Variables|References|GetHitPlayer) (Variables|/Script/Engine.Actor|GetActorGuid _asfort_player_pawn)))
                (Ability|SendGameplayEventtoActor (Variables|References|GetAs%20Fort%20Player%20Pawn%20Athena) "(TagName=\"Event.FeyCrab.Hit\")")
                (Variables|States|SetHasHitEnemyAlready true)))
            (else
              (Variables|References|SetHitPlayer (Variables|/Script/Engine.Actor|GetActorGuid _asfort_player_pawn))))))))
  (if Final%20run
    (if (not (Variables|States|GetHasHitEnemyAlready))
      (Ability|SendGameplayEventtoActor (Variables|References|GetAs%20Fort%20Player%20Pawn%20Athena) "(TagName=\"Event.FeyCrab.Miss\")"))))

(fn DealExplosiveDamage
  (params (Target :type Ability%20System%20Component%20Object%20Reference))
  (GameplayEffects|ApplyGameplayEffectToTarget :self (Ability|GetFortAbilitySystemComponent (Variables|Actor|GetInstigator)) :GameplayEffectClass (Variables|Damage|GetGE_ExplosiveDamage) :Target Target :Level 0.0))

(fn FX_GetExplosionTransform
  (return (Math|Transform|MakeTransform (Variables|Effects|GetStopLocation) (Math|Rotator|MakeRotfromZ (Math|Vector|Normalize (- (Transformation|GetWorldLocation (Variables|Default|GetExplosionPointFront)) (Transformation|GetWorldLocation (Variables|Default|GetExplosionPointBack))))) (Math|Conversions|ToVector(Float) 1.5))))

(fn FX_OnSpawnExplosion
  (params (Explosion :type FXSystem%20Component%20Object%20Reference))
  (bind (_water_hit_bblockinghit _water_hit_binitialoverlap _water_hit_time _water_hit_distance _water_hit_location _water_hit_impactpoint) (Collision|BreakHitResult (Variables|GetWater%20Hit_bBlockingHit)))
  (bind (_water_hit_impactpoint_x _water_hit_impactpoint_y _water_hit_impactpoint_z) (Math|Vector|BreakVector _water_hit_impactpoint))
  (bind _asniagara_particle_system_component (Utilities|Casting|CastToNiagaraComponent Explosion))
  (Niagara|SetNiagaraVariable(Bool) _asniagara_particle_system_component "IsUnderwater" (Variables|Water|GetIs%20Underwater))
  (Niagara|SetNiagaraVariable(Float) _asniagara_particle_system_component "WaterHeight" _water_hit_impactpoint_z)
  (Niagara|SetNiagaraVariable(Vector3) _asniagara_particle_system_component "ExplosionPointFront" (Transformation|GetWorldLocation (Variables|Default|GetExplosionPointFront)))
  (Niagara|SetNiagaraVariable(Vector3) _asniagara_particle_system_component "ExplosionPointBack" (Transformation|GetWorldLocation (Variables|Default|GetExplosionPointBack))))

(fn GetWaterSurfaceLocationForExplosion
  (params (PrjLoc :type Vector) (WaterHitResult :type Hit%20Result%20Structure))
  (bind (_bblockinghit _binitialoverlap _time _distance _location _impactpoint _normal _impactnormal _physmat _hitactor) (Collision|BreakHitResult WaterHitResult))
  (bind (_impactpoint_x _impactpoint_y _impactpoint_z) (Math|Vector|BreakVector _impactpoint))
  (bind _y (.y PrjLoc))
  (bind _x (.x PrjLoc))
  (bind _returnvalue (Math|Vector|MakeVector _x _y _impactpoint_z))
  (bind _asfort_water_body_actor (Utilities|Casting|CastToFortWaterBodyActor _hitactor)
    (:then
      (bind _outwatersurfaceinfo (WaterBody|GetWaterSurfaceInfoStruct _asfort_water_body_actor _returnvalue true false false))
      (return (.WaterSurfaceLocation (Utilities|Struct|BreakWaterSurfaceInfo _outwatersurfaceinfo))))
    (:CastFailed
      (return _returnvalue))))

(fn InitialHitEnemyCheck
  (params (Actor%20stuck%20 :type Actor%20Object%20Reference))
  (bind _asfort_player_state_athena (Utilities|Casting|CastToFortPlayerStateAthena (Variables|/Script/Engine.Controller|GetPlayerState (Pawn|GetController (Variables|/Script/Engine.Actor|GetInstigator Actor%20stuck%20)))))
  (if (not (Class|FortPlayerStateAthena|IsinSquad _asfort_player_state_athena))
    (Ability|SendGameplayEventtoActor (Variables|References|GetAs%20Fort%20Player%20Pawn%20Athena) "(TagName=\"Event.FeyCrab.Hit\")")
    (Variables|States|SetHasHitEnemyAlready true)))

(fn SetRopeFXState
  (params (Target :type Niagara%20Particle%20System%20Component%20Object%20Reference) (State :type Integer))
  (Niagara|SetNiagaraVariable(Int32) Target "FXState" State))

(fn UpdateFXWire
  (params (Target :type Niagara%20Particle%20System%20Component%20Object%20Reference) (Location :type Vector))
  (Niagara|SetNiagaraVariable(Vector3) Target "EndPos" Location))

(fn UpdateWire
  (params (AttachPoint :type Vector) (BasePoint :type Vector) (Wire :type Scene%20Component%20Object%20Reference))
  (bind _returnvalue (- AttachPoint BasePoint))
  (Transformation|SetWorldLocation Wire (+ (* _returnvalue 0.5) BasePoint) false true)
  (Transformation|SetWorldRotation Wire (Math|Rotator|MakeRotfromZ (Math|Vector|Normalize _returnvalue)) false false)
  (Transformation|SetWorldScale3D Wire (Math|Vector|MakeVector 0.025 0.025 (* (Math|Vector|VectorLength _returnvalue) 0.01))))

(fn ConstructionScript
  (Utilities|FlowControl|Sequence
    (:then_0
      (Effects|Components|ParticleSystem|SetTemplate (Variables|RocketTrail|GetRibbon_Trail_PSC) (Variables|RocketTrail|GetRibbon_Trail_PS))
      (Effects|Components|ParticleSystem|SetTemplate (Variables|Default|GetExplosion_Flesh_Damage_PSC) (Variables|Effects|Explosion|Surfaces|GetExplosion_Flesh_Damage_PS))
      (Transformation|SetRelativeScale3D (Variables|Default|GetCollision) (Variables|Projectile_Rocket|GetRocketScale)))
    (:then_1
      ;[1 1 1] Load Rocket Trail asset
      (Variables|RocketTrail|SetRocket%20Trial%20Asset (select (Effects|IsNiagaraFXTypeEnabled "Projectile") (Variables|FX|GetNewNiagaraRocketTrailAsset) "/Game/Weapons/FORT_RocketLaunchers/Effects/NS_RocketTrail_01_Athena.NS_RocketTrail_01_Athena")))
    (:then_2
      ;[1 1 1] Load Explosion asset
      (Variables|Effects|Explosion|SetExplosion%20Asset (select (Effects|IsNiagaraFXTypeEnabled "Projectile") (Variables|FX|GetNewNiagaraExplosionAsset) 0)))))
