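# Unlit additive Niagara mesh-particle material for FeyCrab effects, likely a projectile wrapper, drawn with depth testing disabled.
# Particle color, optionally shaped by a normal-mapped Fresnel through the ShowFresnel dynamic parameter, feeds MF_FX_EmissiveAdjustments.
# A manual scene-versus-pixel depth comparison keeps full particle alpha in front of geometry and fades it steeply behind, over the DepthIntersect distance.
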Material /DryBox/Items/FeyCrab/FX/M_FeyCrabPrj_FX_MeshWrapper
graph {
 properties {
  bDisableDepthTest=true
  bUsedWithNiagaraMeshParticles=true
  blendMode=BLEND_Additive
  pixelDepthOffsetMode=PDOM_Legacy
  shadingModel=MSM_Unlit
  usedShadingModels=Unlit
 }
 nodes [
  n1 = SceneDepth()
  n2 = PixelDepth()
  n3 = ParticleColor()
  n4 = Subtract(A=@n2.R, B=@n1.R)
  n5 = ScalarParameter() {parameterName=DepthIntersect}
  n6 = OneMinus(None=Saturate(None=Divide(A=@n4, B=@n5) {constB=100}))
  If(A=@n1.R, B=@n2.R, "A > B"=@n3.A, "A == B"=@n3.A, "A < B"=Multiply(A=@n3.A, B=Power(Base=@n6) {constExponent=4}))
  n7 = Fresnel(Normal=Transform(None=TextureSample() {samplerType=SAMPLERTYPE_Normal, texture=/DryBox/Items/FeyCrab/Textures/T_FeyCrab_N}.RGB)) {exponent=1}
  n8 = DynamicParameter() {paramNames [ShowFresnel,Param2,Param3,Param4]}
  MaterialFunctionCall(Emissive=Multiply(A=LinearInterpolate(B=@n7, Alpha=@n8.ShowFresnel) {constA=1}, B=@n3.RGB)) {materialFunction=&MF_FX_EmissiveAdjustments}
  OneMinus(None=DepthFade(Opacity=@n3.A, FadeDistance=ScalarParameter() {parameterName=Thickness}))
  ScalarParameter()
 ]
}
