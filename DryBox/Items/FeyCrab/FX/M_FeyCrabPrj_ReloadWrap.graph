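# Additive unlit material for Niagara mesh particles, by its name a reload wrap effect on the FeyCrab projectile.
# Particle color, run through MF_FX_EmissiveAdjustments, is multiplied by a Fresnel term that uses the transformed FeyCrab normal map.
# The Fresnel exponent lerps between FresnelMin and FresnelMax by UV V plus the particle-driven FresnelExponent dynamic parameter, so rim falloff varies along the mesh UVs.
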Material /DryBox/Items/FeyCrab/FX/M_FeyCrabPrj_ReloadWrap
graph {
 properties {
  bUsedWithNiagaraMeshParticles=true
  blendMode=BLEND_Additive
  pixelDepthOffsetMode=PDOM_Legacy
  shadingModel=MSM_Unlit
  usedShadingModels=Unlit
 }
 nodes [
  n1 = ScalarParameter() {parameterName=FresnelMin}
  n2 = ScalarParameter() {defaultValue=3, parameterName=FresnelMax}
  n3 = Add(A=ComponentMask(None=TextureCoordinate()) {g=true}, B=DynamicParameter() {paramNames [FresnelExponent, Wipe, R, Param4]}.FresnelExponent)
  n4 = TextureSample() {samplerType=SAMPLERTYPE_Normal, texture=/DryBox/Items/FeyCrab/Textures/T_FeyCrab_N}
  Multiply(
   A=MaterialFunctionCall(Emissive=ParticleColor().RGB) {materialFunction=&MF_FX_EmissiveAdjustments}.Result
   B=Fresnel(ExponentIn=LinearInterpolate(A=@n1, B=@n2, Alpha=@n3), Normal=Transform(None=@n4.RGB))
  )
 ]
}
