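# Translucent unlit material for Niagara mesh particles, by its name the FeyCrab item's ground sphere effect.
# Opacity combines fresnel-shaped fades at center and edge, particle alpha, two panning fire-noise samples and depth fade.
# A dynamic-parameter erosion of panning noise brightens particle-color emissive and, via the CellErosion switch, cuts opacity, skipped on Low and Medium quality.
# Dynamic parameters drive fades, pan speed and switchable view-angle-faded refraction.
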Material /DryBox/Items/FeyCrab/FX/M_FeycrabGroundSphere
graph {
 properties {
  bUsedWithNiagaraMeshParticles=true
  blendMode=BLEND_Translucent
  pixelDepthOffsetMode=PDOM_Legacy
  refractionMethod=RM_IndexOfRefraction
  shadingModel=MSM_Unlit
  usedShadingModels=Unlit
 }
 nodes [
  Constant() {r=1}
  n1 = DynamicParameter() {parameterIndex=1, paramNames [Erode 1-0, Param2, Param3, Param4]}
  n2 = TextureCoordinate()
  n3 = ScalarParameter() {defaultValue=2, parameterName=ErodeTile}
  n4 = TextureSampleParameter2D(UVs=Panner(Coordinate=Multiply(A=@n2, B=@n3), Time=NamedRerouteUsage()) {speedX=1, speedY=-8}) {
   parameterName=ErodeTexture
   samplerType=SAMPLERTYPE_LinearGrayscale
   texture=/Game/Characters/Player/Female/Medium/Bodies/F_MED_Cube_Queen/Textures/TechArt/T_CubeQueenNoiseA
  }
  n5 = ComponentMask(None=Multiply(A=@n4.G)) {g=true}
  n6 = Power(Base=LinearInterpolate(B=@n1.Erode 1-0, Alpha=@n5) {constA=1}, Exp=LinearInterpolate(Alpha=@n1.Erode 1-0) {constA=20, constB=-1})
  n7 = Multiply(A=OneMinus(None=Saturate(None=@n6))) {constB=50}
  n8 = ParticleColor()
  MaterialFunctionCall(Emissive=Add(A=Multiply(A=@n7, B=@n8.RGB), B=@n8.RGB)) {materialFunction=&MF_FX_EmissiveAdjustments}
  n9 = Multiply(A=@n1.Erode 1-0, B=Clamp(None=Multiply(A=@n6) {constB=5}) {clampMode=CMODE_ClampMax})
  n10 = Constant() {r=1}
  n11 = Fresnel(ExponentIn=Constant() {r=0.8}) {baseReflectFraction=0.001}
  n12 = OneMinus(None=@n11)
  n13 = Power(Base=Multiply(A=Multiply(A=@n12, B=@n11), B=ParticleColor().A) {constA=1})
  n14 = DynamicParameter() {defaultValue=<rgba 3 1 1 1>, paramNames [MiddleFade,EdgeFade,PanSpeed,Ref]}
  n15 = OneMinus(None=Fresnel(ExponentIn=@n14.EdgeFade) {baseReflectFraction=0.001})
  n16 = Fresnel(ExponentIn=@n14.MiddleFade) {baseReflectFraction=0.001}
  n17 = OneMinus(None=@n16)
  n18 = Multiply(
   A=Multiply(
    A=StaticSwitchParameter(True=@n15, False=@n17) {desc=Allows independent control of middle and edge fresnel fade, parameterName=Independent Fresnel Fade}
    B=@n16
   )
   B=Constant() {r=1}
  )
  n19 = ScalarParameter() {defaultValue=1, parameterName=NoiseTile}
  n20 = Multiply(A=TextureCoordinate(), B=@n19)
  n21 = NamedRerouteDeclaration(None=Multiply(A=@n14.PanSpeed, B=Time())) {name=PanSpeed, nodeColor=<rgba 0.3359 0 1 1>}
  n22 = Multiply(A=TextureCoordinate() {uTiling=4, vTiling=2}, B=@n19)
  n23 = Multiply(
   A=TextureSample(UVs=Panner(Coordinate=@n20, Time=@n21) {speedY=-1}) {texture=/Game/Global/Textures/Default/Noise/T_FireNoise_02}.G
   B=TextureSample(UVs=Panner(Coordinate=@n22, Time=@n21) {speedY=-1}) {texture=&T_FireNoise_02}.G
  )
  n24 = Constant() {r=4}
  n25 = Clamp(None=Multiply(A=Power(Base=@n18), B=Multiply(A=@n23, B=@n24)))
  Multiply(
   A=StaticSwitchParameter(True=QualitySwitch(Default=@n9, Low=@n10, Medium=@n10), False=@n10) {parameterName=CellErosion}
   B=DepthFade(Opacity=Multiply(A=@n13, B=@n25), FadeDistance=ScalarParameter() {defaultValue=100, parameterName=DepthFadeDIst})
  ) {constA=1}
  OneMinus(None=Divide(A=Time() {bOverride_Period=true, period=2}))
  OneMinus(None=@n5)
  n26 = Constant() {r=0.5}
  SphereMask(A=ComponentMask(None=TextureCoordinate()) {g=true}, B=@n26, Radius=@n26) {attenuationRadius=0.25, hardnessPercent=5}
  n27 = MaterialFunctionCall("FadeSensitivity (optional)"=Constant() {r=2}) {materialFunction=&FadeBasedOnViewAngle}
  StaticSwitchParameter(True=LinearInterpolate(Alpha=Multiply(A=@n14.Ref, B=@n27.Result)) {constA=1, constB=1.1}, False=Constant() {r=1}) {parameterName=UseRefraction}
 ]
}
