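# Unlit additive Niagara ribbon material for the FeyCrab rope effect, with responsive AA enabled.
# A NyxGlass FX texture multiplied by particle color feeds the shared emissive-adjustments function.
# A SmoothStep on a dissolve texture, tiled along V, builds an erosion mask, with its threshold lerped between min and max by the ErodeT dynamic parameter.
# Edge width comes from a scalar parameter rather than the dynamic ErodeW channel.
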Material /DryBox/Items/FeyCrab/FX/M_FeycrabRope_EdgeLit
graph {
 properties {
  bEnableResponsiveAA=true
  bUsedWithNiagaraRibbons=true
  blendMode=BLEND_Additive
  pixelDepthOffsetMode=PDOM_Legacy
  shadingModel=MSM_Unlit
  usedShadingModels=Unlit
 }
 nodes [
  MaterialFunctionCall(Emissive=Multiply(A=TextureSample() {texture=/DryBox/Items/NyxGlass/Textures/T_Gadget_NyxGlass_D_FX}.RGB, B=ParticleColor().RGB)) {materialFunction=&MF_FX_EmissiveAdjustments}
  ScalarParameter() {parameterName=ErodePreview}
  n1 = LinearInterpolate(
   A=ScalarParameter() {parameterName=ErodeMin}
   B=ScalarParameter() {parameterName=ErodeMax}
   Alpha=DynamicParameter() {paramNames [Param1,Param2,ErodeW,ErodeT]}.ErodeT
  )
  SmoothStep(
   Min=@n1
   Max=Add(A=@n1, B=ScalarParameter() {parameterName=ErodeW})
   Value=TextureSample(UVs=TextureCoordinate() {vTiling=0.25}) {texture=/Game/CharClasses/Outlander/Gadgets/AntiMaterialCharge/Materials/T_AntiMaterialCharge_Dissolve_01}.G
  )
 ]
}
