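# Additive unlit Niagara ribbon/sprite material for the Feycrab item's rope effect, with particle color feeding an emissive adjustment function.
# UVs shifted by an offset parameter are mirror-repeated and axis-swapped to sample the green channel of an electronic gradient texture.
# That gradient is multiplied by particle alpha and a sphere mask centered on UV1's U axis to fade both ends.
# A CameraOffset parameter drives a camera-offset function.
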Material /DryBox/Items/FeyCrab/FX/M_Feycrab_Rope_FullyIllum
graph {
 properties {
  bUsedWithNiagaraRibbons=true
  bUsedWithNiagaraSprites=true
  blendMode=BLEND_Additive
  pixelDepthOffsetMode=PDOM_Legacy
  shadingModel=MSM_Unlit
  usedShadingModels=Unlit
 }
 nodes [
  n1 = TextureCoordinate()
  ComponentMask(None=@n1) {g=true}
  n2 = ParticleColor()
  MaterialFunctionCall(Emissive=@n2.RGB) {materialFunction=&MF_FX_EmissiveAdjustments}
  MaterialFunctionCall(Offset Amount=ScalarParameter() {parameterName=CameraOffset}) {materialFunction=/Engine/Functions/Engine_MaterialFunctions02/WorldPositionOffset/CameraOffset}
  n3 = MaterialFunctionCall!(Repeat Count=Constant2Vector() {g=2, r=2}, texcoord=Subtract(A=@n1, B=ScalarParameter() {parameterName=offset}) {constB=0.5}) {materialFunction=/Game/Environments/Papaya/Materials/MaterialFunctions/MF_UVMirror}
  n4 = MaterialFunctionCall(XYZ=@n3.Result, XY=@n3.Result) {materialFunction=&Swizzle}
  n5 = ComponentMask(None=TextureCoordinate() {coordinateIndex=1}) {r=true}
  n6 = Constant() {r=0.5}
  Multiply(
   A=@n2.A
   B=Multiply(
    A=TextureSample(UVs=@n4.YX) {texture=/Game/Weapons/WeaponSkins/Textures/ElectronicGradient_Flipped}.G
    B=SphereMask(A=@n5, B=@n6) {attenuationRadius=0.5, hardnessPercent=75}
   )
  )
 ]
}
