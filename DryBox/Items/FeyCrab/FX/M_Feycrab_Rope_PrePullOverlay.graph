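# Additive unlit Niagara ribbon and sprite material, apparently a glow overlay for the FeyCrab rope's pre-pull effect.
# Particle color RGB feeds a shared emissive adjustment function, and a camera offset function takes an adjustable amount.
# A soft mask centered across V multiplies swirl noise stretched along V, then is halved.
# Particle alpha drives the noise's inverted SmoothStep threshold, eroding or revealing the pattern over particle life.
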Material /DryBox/Items/FeyCrab/FX/M_Feycrab_Rope_PrePullOverlay
graph {
 properties {
  bUsedWithNiagaraRibbons=true
  bUsedWithNiagaraSprites=true
  blendMode=BLEND_Additive
  pixelDepthOffsetMode=PDOM_Legacy
  shadingModel=MSM_Unlit
  usedShadingModels=Unlit
 }
 nodes [
  Fresnel() {baseReflectFraction=0, exponent=0.2}
  MaterialFunctionCall(Offset Amount=ScalarParameter() {defaultValue=3, parameterName=Offset}) {materialFunction=/Engine/Functions/Engine_MaterialFunctions02/WorldPositionOffset/CameraOffset}
  n1 = ParticleColor()
  MaterialFunctionCall(Emissive=@n1.RGB) {materialFunction=&MF_FX_EmissiveAdjustments}
  n2 = SphereMask(A=ComponentMask(None=TextureCoordinate()) {g=true}, B=Constant() {r=0.5}) {attenuationRadius=0.5, hardnessPercent=0}
  n3 = Add(A=@n1.A) {constB=0.2}
  n4 = TextureSample(UVs=TextureCoordinate() {vTiling=0.1}) {samplerType=SAMPLERTYPE_LinearGrayscale, texture=/Game/Global/Textures/Default/Noise/T_SwirlNoise_H2}
  Multiply(A=Multiply(A=Power(Base=@n2), B=SmoothStep(Min=@n3, Max=@n1.A, Value=@n4.G))) {constB=0.5}
  OneMinus(None=@n2)
 ]
}
