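# Master material for the FeyCrab projectile, flagged for skeletal meshes, Niagara mesh particles, Nanite and mesh deformers.
# The FeyCrab diffuse is blended with white, tinted by particle color and modulated by a dynamic-parameter-controlled Fresnel.
# A clamped GlowAmount dynamic parameter fades that result toward black before it enters the FX emissive adjustment function.
# A specular mask drives roughness range and metallic scaling, alongside a normal map sample.
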Material /DryBox/Items/FeyCrab/FX/M_ShellMaster_FeyCrab_Projectile
graph {
 properties {
  bUsedWithMeshDeformer=true
  bUsedWithNanite=true
  bUsedWithNiagaraMeshParticles=true
  bUsedWithSkeletalMesh=true
  pixelDepthOffsetMode=PDOM_Legacy
  previewMesh=/DryBox/Items/FeyCrab/Meshes/FeyCrab_Projectile
  usedShadingModels=Default Lit
 }
 nodes [
  n1 = TextureSampleParameter2D() {
   group=Specular
   parameterName=SpecularMasks
   samplerType=SAMPLERTYPE_Masks
   texture=&T_Fortnite_Default_S
  }
  LinearInterpolate(
   A=ScalarParameter() {defaultValue=0.22, group=Specular, parameterName=SpecRoughnessMin, sliderMax=1}
   B=ScalarParameter() {defaultValue=0.66, group=Specular, parameterName=SpecRoughnessMax, sliderMax=1}
   Alpha=@n1.B
  )
  n2 = Multiply(
   A=LinearInterpolate(B=TextureSampleParameter2D() {parameterName=Diffuse, texture=/DryBox/Items/FeyCrab/Textures/T_FeyCrab_D}.RGB) {constA=1}
   B=ParticleColor().RGB
  )
  n3 = LinearInterpolate(B=Fresnel() {exponent=1}, Alpha=DynamicParameter() {parameterIndex=1, paramNames [FresnelAmount, Param2, Param3, Param4]}.FresnelAmount) {constA=1}
  n4 = DynamicParameter() {defaultValue=<rgba 8 2 0 0.9>, paramNames [R,G,B,GlowAmount]}
  MaterialFunctionCall(Emissive=LinearInterpolate(A=Multiply(A=@n2, B=@n3), Alpha=Clamp(None=@n4.GlowAmount)) {constB=0}) {materialFunction=&MF_FX_EmissiveAdjustments}
  MaterialFunctionCall(R=@n4.R, G=@n4.R, B=@n4.R) {materialFunction=&AppendMany}
  Multiply(A=@n1.G, B=ScalarParameter() {defaultValue=1, group=Specular, parameterName=Metallic, sliderMax=1})
  TextureSampleParameter2D() {parameterName=Normals, samplerType=SAMPLERTYPE_Normal, texture=/DryBox/Items/FeyCrab/Textures/T_FeyCrab_N}
  VectorParameter() {parameterName=DebugColor, defaultValue=<rgba 0.1615 0.1615 0.1615 1>}
 ]
}
