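# Niagara system for the FeyCrab launcher's connecting cable, drawn as layered ribbons between system start and end positions.
# Cabling spaces burst particles along that line, with User.FXState selecting shoot-out unravel, tension wobble or pullback curl offsets driven by waveforms and timeline curves.
# The curl uses time-interpolated shrink targets, alongside a ghost snap-back trail.
# PullbackEffect animates a whipping rope with fading glow and whoosh ribbons, mirrored by User.IsLeftHand.
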NiagaraSystem /DryBox/Items/FeyCrab/FX/NS_FeycrabLauncher_Connect
dependencies {
 usedDataInterfaces [&NiagaraDataInterfaceCurve, &NiagaraDataInterfaceColorCurve, &NiagaraDataInterfaceArrayColor]
 usedDynamicInputs [
  &InvertBool
  &SetBoolByFloatComparison
  &MakeCustomVectorFromBool
 ]
 usedModules [
  &self:SystemSpawnScript.SystemScriptSource.SystemScriptGraph.NiagaraNodeAssignment_7.SetVariables_#1
  &self:SystemSpawnScript.SystemScriptSource.SystemScriptGraph.NiagaraNodeAssignment_8.SetVariables_#2
  &self:SystemSpawnScript.SystemScriptSource.SystemScriptGraph.NiagaraNodeAssignment_9.SetVariables_#3
  &SystemState
  &self:SystemSpawnScript.SystemScriptSource.SystemScriptGraph.NiagaraNodeAssignment_13.SetVariables_#4
  &self:SystemSpawnScript.SystemScriptSource.SystemScriptGraph.NiagaraNodeAssignment_14.SetVariables_#5
  &Timeline
  &self:SystemSpawnScript.SystemScriptSource.SystemScriptGraph.NiagaraNodeAssignment_3.SetVariables_#6
  &self:SystemSpawnScript.SystemScriptSource.SystemScriptGraph.NiagaraNodeAssignment_12.SetVariables_#7
  &self:SystemSpawnScript.SystemScriptSource.SystemScriptGraph.NiagaraNodeAssignment_2.SetVariables_#8
  &self:SystemSpawnScript.SystemScriptSource.SystemScriptGraph.NiagaraNodeAssignment_11.SetVariables_#9
  &self:SystemSpawnScript.SystemScriptSource.SystemScriptGraph.NiagaraNodeAssignment_0.SetVariables_#10
  &self:SystemSpawnScript.SystemScriptSource.SystemScriptGraph.NiagaraNodeAssignment_4.SetVariables_#11
  &self:SystemSpawnScript.SystemScriptSource.SystemScriptGraph.NiagaraNodeAssignment_5.SetVariables_#12
  &self:SystemSpawnScript.SystemScriptSource.SystemScriptGraph.NiagaraNodeAssignment_6.SetVariables_#13
  &self:SystemSpawnScript.SystemScriptSource.SystemScriptGraph.NiagaraNodeAssignment_1.SetVariables_#14
  &self:SystemSpawnScript.SystemScriptSource.SystemScriptGraph.NiagaraNodeAssignment_10.SetVariables_#15
 ]
}
emitters {
 Cabling {
  emitterSpawnScript [/SetVariables(Emitter.RopeGlowWidth=644.2952) {isSetParameters=true}]
  emitterUpdateScript [
   &EmitterState()
   &SpawnBurst_Instantaneous(Spawn Count=30, Use Spawn Probability=false, UseLoopCountLimit=false)
   /SetVariables() {isSetParameters=true}
   /SetVariables(Emitter.ShowGhostRibbon:NiagaraBool=&SetBoolByIntComparison(A=$User.FXState, B=2, Comparison Type=::NewEnumerator4)) {isSetParameters=true}
   /SetVariables() {enabled=false, isSetParameters=true}
   &SpawnBurst_Instantaneous(Spawn Count=30, Spawn Time=4.8232, Use Spawn Probability=false, UseLoopCountLimit=false) {module=SpawnBurst_Instantaneous001}
  ]
  particleSpawnScript [
   &InitializeParticle(
    Color=<rgba 1 0.7872 0.5462 1>
    Lifetime=90
    Ribbon Width=2.5
    Ribbon Width Mode=::NewEnumerator1
    Sprite Size Mode=::NewEnumerator3
    UsePositionOffset=true
    Write Material Random=true
   )
   /SetVariables(Particles.RibbonLinkOrder:NiagaraFloat=&ReturnNormalizedExecIndex()) {isSetParameters=true}
   /SetVariables(Particles.LightglowWidth=35) {isSetParameters=true}
   /SetVariables() {isSetParameters=true}
   /SetVariables(Particles.LitWireWidth=110) {isSetParameters=true}
  ]
  particleUpdateScript [
   &ParticleState()
   &ScaleColor(
    Scale Alpha=&FloatFromCurve(FloatCurve=DataInterface() {propertyValues {Curve {keys [<at 0 1>,<at 1 0>]}}}, CurveIndex=$System.PullbackT)
    Scale RGB=&VectorFromFloat(Value=1)
    ScaleA=true
    ScaleRGB=true
   )
   /SetVariables(Particles.Position=&Lerp_Vector(A=$System.StartPosition, B=$System.EndPosition, Alpha=$Particles.RibbonLinkOrder)) {isSetParameters=true}
   /SetVariables(Particles.GhostTrailSnapBackLine:NiagaraPosition=&Lerp_Vector(
     A=$System.StartPosition
     B=$System.EndPosition
     Alpha=&Multiply_Float(
      A=$Particles.RibbonLinkOrder
      B=&FloatFromCurve(FloatCurve=DataInterface() {propertyValues {Curve {keys [<at 0 0.875>,<at 1 -2.5>]}}}, CurveIndex=$System.PullbackT)
     )
    )) {isSetParameters=true}
   /SetVariables(Particles.UnravelOffset:Vector3f=&MakeVector(
     X=0
     Y=0
     Z=&Multiply_Float(
      A=&Multiply_Float(
       A=&Waveform(
        Waveform Count=::NewEnumerator1
        Global Phase=&Add_Float(
         A=&Multiply_Float(A=$Emitter.Age, B=2)
         B=&ReturnNormalizedExecIndex(Normalized Index Scale=-0.775)
        )
        "Waveform [1]"=::NewEnumerator0
        "[1] Phase Scale"=3
        "[1] Frequency"=1
        "[1] Amplitude Min/Max"=<-1 1>
        "[1] Scale"=200
        "[2] Average"=false
        "Waveform [2]"=::NewEnumerator12
        "[2] Input Position"=0
        "[2] Phase"=$Emitter.Age
        "[2] Phase Scale"=3
        "[2] Compound Frequency"=2
        "[2] Frequency"=8.952
        "[2] Amplitude Min/Max"=<-1 1>
        "[2] Scale"=75
        "[2] Bias"=0
        "[2] Exponent"=1
        "[2] Pulse Duration"=0.1
        "[2] Clip Triangle Min/Max"=<0 0>
        "[2] Random Seed"=0
       )
       B=&FloatFromCurve(
        FloatCurve=DataInterface() {propertyValues {Curve {keys [
            <at 0 0 RCIM_Cubic>
            {arriveTangent=0.7143, interpMode=RCIM_Cubic, leaveTangent=0.7143, time=0.3, value=1}
            {arriveTangent=-1.4286, interpMode=RCIM_Cubic, leaveTangent=-1.4286, time=0.7, value=0.5}
            <at 1 0 RCIM_Cubic>
           ]}}}
        CurveIndex=&ReturnNormalizedExecIndex()
       )
      )
      B=&FloatFromCurve(
       FloatCurve=DataInterface() {propertyValues {Curve {keys [
           <at 0 1 RCIM_Cubic>
           {arriveTangent=-1.5, interpMode=RCIM_Cubic, leaveTangent=-1.5, time=0.2, value=1}
           {arriveTangent=-1.25, interpMode=RCIM_Cubic, leaveTangent=-1.25, time=0.4, value=0.4}
           <at 1 0 RCIM_Cubic>
          ]}}}
       CurveIndex=$System.ShootOutT
      )
     )
    )) {isSetParameters=true}
   /SetVariables(Particles.UnravelOffset:Vector3f=&Add_Vector(
     A=$Particles.UnravelOffset
     B=&TransformVector(
      Vector=&MakeVector(
       X=&Waveform(
        Global Phase=&Add_Float(A=$Emitter.Age, B=&ReturnNormalizedExecIndex(Normalized Index Scale=-1))
        "Waveform [1]"=::NewEnumerator1
        "[1] Frequency"=5
        "[1] Amplitude Min/Max"=<-1 1>
        "[1] Scale"=20
       )
       Y=&Waveform(
        Global Phase=&Add_Float(A=$Emitter.Age, B=&ReturnNormalizedExecIndex(Normalized Index Scale=-1))
        "Waveform [1]"=::NewEnumerator0
        "[1] Frequency"=4
        "[1] Amplitude Min/Max"=<-1 1>
        "[1] Scale"=20
       )
       Z=0
      )
      Destination Space=::Simulation
     )
    )) {enabled=false, isSetParameters=true}
   /SetVariables() {enabled=false, isSetParameters=true}
   /SetVariables(Particles.UnravelPositionOffset:Vector3f=&Multiply_VectorByFloat(
     Vector=$Particles.UnravelOffset
     Float=&Multiply_Float(
      A=&FloatFromCurve(
       FloatCurve=DataInterface() {propertyValues {Curve {keys [
           <at 0 1 RCIM_Cubic>
           {arriveTangent=-1.5, interpMode=RCIM_Cubic, leaveTangent=-1.5, time=0.2, value=1}
           {arriveTangent=-1.25, interpMode=RCIM_Cubic, leaveTangent=-1.25, time=0.4, value=0.4}
           <at 1 0 RCIM_Cubic>
          ]}}}
       CurveIndex=$System.ShootOutT
      )
      B=&Multiply_Float(
       A=&RemapRange(
        Input Value To Remap=&ClampFloat(
         Float=&DistanceBetweenPositions(A=Linked() {linkedVariable {name=System.StartPosition, type=Vector3f}}, B=Linked() {linkedVariable {name=System.EndPosition, type=Vector3f}})
         Max=1500
        )
        Input Min=0
        Input Max=1500
        Output Min=0.125
        Output Max=2
       )
       B=0.7
      )
     )
    )) {isSetParameters=true}
   /SetVariables(Particles.RopeTensionWobbleOffset:Vector3f=&MakeVector(
     X=0
     Y=0
     Z=&Multiply_Float(
      A=&Sine(
       Normalized Angle=&Add_Float(
        A=$Emitter.Age
        B=&Multiply_Float(A=&ReturnNormalizedExecIndex(Normalized Index Scale=0.1), B=1)
       )
       Period=0.15
       Scale=&FloatFromCurve(
        FloatCurve=DataInterface() {propertyValues {Curve {keys [
            {interpMode=RCIM_Cubic, tangentMode=RCTM_User, time=0, value=0}
            {interpMode=RCIM_Cubic, tangentMode=RCTM_User, time=0.1, value=1}
            {tangentMode=RCTM_User, time=1, value=0}
           ]}}}
        CurveIndex=&ReturnNormalizedExecIndex()
        Scale Curve=25
       )
       Bias=0
      )
      B=&FloatFromCurve(
       FloatCurve=DataInterface() {propertyValues {Curve {keys [<at 0 1 RCIM_Cubic>, {arriveTangent=-1, interpMode=RCIM_Cubic, leaveTangent=-1, time=0.6, value=0.7}, <at 1 0 RCIM_Cubic>]}}}
       CurveIndex=$System.ShotToRestingT
      )
     )
    )) {isSetParameters=true}
   /SetVariables(Particles.PositionTarget:NiagaraPosition=$Particles.Position) {isSetParameters=true}
   &self:ShrinkPosA_02(
    Angle:NiagaraFloat=$User.A
    AxisScale:NiagaraFloat=&FloatFromCurve(
     FloatCurve=DataInterface() {propertyValues {Curve {keys [<at 0 -1.5>,<at 0.5 1>]}}}
     CurveIndex=&Subtract_Float(
      A=&Multiply_Float(A=$System.PullbackT, B=1.5)
      B=&ReturnNormalizedExecIndex(Normalized Index Scale=0.5)
     )
     Scale Curve=0.5
    )
    B2:NiagaraFloat=$System.B2
    C:NiagaraFloat=$System.C
    Dist=50
    Pow=1
    R:NiagaraFloat=&Lerp_Float(
     A=10
     B=100
     Alpha=&FloatFromCurve(
      FloatCurve=DataInterface() {propertyValues {Curve {keys [
          {arriveTangent=-1.0959, interpMode=RCIM_Cubic, time=0, value=1}
          {arriveTangent=-0.1288, interpMode=RCIM_Cubic, leaveTangent=-0.1288, tangentMode=RCTM_User, time=0.5, value=0.575}
          <at 1 0 RCIM_Cubic>
         ]}}}
      CurveIndex=$System.PullbackT
     )
    )
    X:Vector3f=<0 1 0>
    Y:Vector3f=<0 0 -1>
    Z:Vector3f=<1 0 0>
   ) {module=ShrinkPosA_02}
   /SetVariables(Particles.TPosition:NiagaraPosition=&Lerp_Vector(
     A=Linked() {linkedVariable {name=Particles.PositionTarget, type=NiagaraPosition}}
     B=Linked() {linkedVariable {name=Particles.ShrinkPosition, type=NiagaraPosition}}
     Alpha=$System.BlendAB
    )) {isSetParameters=true}
   /SetVariables(Particles.GhostTrailPos:NiagaraPosition=&Lerp_Vector(
     A=Linked() {linkedVariable {name=Particles.GhostTrailSnapBackLine, type=NiagaraPosition}}
     B=&AddVectorToPosition(
      Position=$Particles.GhostTrailSnapBackLine
      Vector=&Multiply_VectorByFloat(
       Vector=&MakeVector(
        X=0
        Y=0
        Z=&Waveform(
         Global Phase=&Add_Float(A=$Emitter.Age, B=&ReturnNormalizedExecIndex())
         "Waveform [1]"=::NewEnumerator12
         "[1] Frequency"=4
         "[1] Amplitude Min/Max"=<-1 1>
         "[1] Scale"=200
        )
       )
       Float=&FloatFromCurve(
        FloatCurve=DataInterface() {propertyValues {Curve {keys [
            <at 0 0 RCIM_Cubic>
            {arriveTangent=0.5714, interpMode=RCIM_Cubic, leaveTangent=0.5714, time=0.3, value=0.9}
            {arriveTangent=-1.0714, interpMode=RCIM_Cubic, leaveTangent=-1.0714, time=0.7, value=0.4}
            <at 1 0.15 RCIM_Cubic>
           ]}}}
        CurveIndex=&ReturnNormalizedExecIndex()
       )
      )
     )
     Alpha=&FloatFromCurve(
      FloatCurve=DataInterface() {propertyValues {Curve {keys [{interpMode=RCIM_Cubic, tangentMode=RCTM_User, time=0, value=0}, {interpMode=RCIM_Cubic, tangentMode=RCTM_User, time=1, value=1.6}]}}}
      CurveIndex=$System.Timeline002.Time
     )
    )) {isSetParameters=true}
   &InterpolateOverTime(
    Rate Of Change:NiagaraFloat=&Lerp_Float(
     A=128
     B=&Lerp_Float(A=4, B=32, Alpha=$System.Soft)
     Alpha=&Power(a=&ReturnNormalizedExecIndex(), b=0.25)
    )
    Target Vector Value:NiagaraPosition=$Particles.TPosition
    Variable Type=::NewEnumerator2
   )
   &DynamicMaterialParameters(
    Index 0 Param 1=&FloatFromCurve(
     FloatCurve=DataInterface() {propertyValues {Curve {keys [
         <at 0 0 RCIM_Cubic>
         {arriveTangent=1.25, interpMode=RCIM_Cubic, leaveTangent=1.25, time=0.1, value=1.35}
         {arriveTangent=-1.5, interpMode=RCIM_Cubic, leaveTangent=-1.5, time=0.2, value=0.25}
         <at 1 0 RCIM_Cubic>
        ]}}}
     CurveIndex=$System.PullbackT
     Scale Curve=6
    )
    Index 0 Param 2=1
    Index 0 Param 3=1.345
    Index 0 Param 4=&FloatFromCurve(FloatCurve=DataInterface() {propertyValues {Curve {keys [<at 0 1>,<at 0.125 0>,<at 0.7 1>]}}}, CurveIndex=$System.PullbackT)
    Module.Param0WriteEnabled=true
    Module.Param1WriteEnabled=true
    Module.Param2WriteEnabled=true
    Module.Param3WriteEnabled=true
   )
   /SetVariables(Particles.CurlBackPosition:NiagaraPosition=&RotateVectorByQuaternion(
     Vector To Rotate=&SubtractVectorFromPosition(
      Position=$Particles.InterpolateOverTime.InterpolateOverTime_V3_.Moving Average
      Vector=Linked() {linkedVariable {name=Particles.Position, type=NiagaraPosition}}
     )
     Quaternion=&MakeQuatFromAxisAngle(Axis=$Engine.Owner.SystemZAxis, AngleInDegrees=0)
    )) {isSetParameters=true}
   /SetVariables(Particles.CurlBackPosition:NiagaraPosition=&AddVectorToPosition(
     Position=$Particles.CurlBackPosition
     Vector=&Multiply_VectorByFloat(
      Vector=$Engine.Owner.SystemYAxis
      Float=&Multiply_Float(
       A=&FloatFromCurve(
        FloatCurve=DataInterface() {propertyValues {Curve {keys [
            {arriveTangent=2.5977, interpMode=RCIM_Cubic, leaveTangent=2.5977, tangentMode=RCTM_User, time=0, value=0}
            <at 0.5 1 RCIM_Cubic>
            {arriveTangent=-3.3193, interpMode=RCIM_Cubic, leaveTangent=-3.3193, tangentMode=RCTM_User, time=1, value=0}
           ]}}}
        CurveIndex=&ReturnNormalizedExecIndex()
        Scale Curve=&MakeCustomFloatFromBool(Bool=$User.IsLeftHand, True Float=-50, False Float=75)
       )
       B=&FloatFromCurve(
        FloatCurve=DataInterface() {propertyValues {Curve {keys [<at 0 1 RCIM_Cubic>, {arriveTangent=-1.4286, interpMode=RCIM_Cubic, leaveTangent=-1.4286, time=0.5, value=0.95}, <at 0.7 0 RCIM_Cubic>]}}}
        CurveIndex=$System.PullbackT
       )
      )
     )
    )) {isSetParameters=true}
   /SetVariables(Particles.StateOffset:Vector3f=&self:ScratchDynamicInput_01(
     Choice1:Vector3f=$Particles.UnravelPositionOffset
     Choice2:Vector3f=$Particles.RopeTensionWobbleOffset
     Choice3:Vector3f=Linked() {linkedVariable {name=Particles.CurlBackPosition, type=NiagaraPosition}}
     Index:NiagaraInt32=$User.FXState
    )) {isSetParameters=true}
   /SetVariables(Particles.Position=&AddVectorToPosition(Position=$Particles.Position, Vector=$Particles.StateOffset)) {isSetParameters=true}
   /SetVariables(Particles.WarningColor:LinearColor=&MultiplyLinearColor(
     Linear Color=&ColorFromCurve(
      ColorCurve=DataInterface() {propertyValues {
        AlphaCurve {keys [<at 0 1 RCIM_Cubic>,<at 0.2179 0>]}
        BlueCurve {keys [<at 0.0032 0>,<at 0.0319 8.6458 RCIM_Cubic>]}
        GreenCurve {keys [<at 0.0032 0>,<at 0.0319 15.7789 RCIM_Cubic>]}
        RedCurve {keys [<at 0.0032 0>,<at 0.0319 20 RCIM_Cubic>]}
       }}
      CurveIndex=$System.PullbackT
     )
     Scale Factor=&MakeFloat4FromVectorAndFloat("Vector (RGB)"=<0.3 0.3 0.3>, "Float (Alpha)"=0.25)
    )) {isSetParameters=true}
   /SetVariables(Particles.LightglowColor:LinearColor=&MultiplyLinearColor(Linear Color=$Particles.WarningColor, Scale Factor=<0.02 0.02 0.02 0.2>)) {isSetParameters=true}
   /SetVariables(Particles.RibbonWidth=&MakeCustomFloatFromBool(
     Bool=&SetBoolByIntComparison(A=$User.FXState, B=2, Comparison Type=::NewEnumerator4)
     True Float=&Multiply_Float(
      A=$Particles.Initial.RibbonWidth
      B=&FloatFromCurve(
       FloatCurve=DataInterface() {propertyValues {Curve {keys [<at 0 1 RCIM_Cubic>, {arriveTangent=-1, interpMode=RCIM_Cubic, leaveTangent=-1, time=0.6, value=1}, <at 1 0 RCIM_Cubic>]}}}
       CurveIndex=&ReturnNormalizedExecIndex()
      )
     )
     False Float=$Particles.Initial.RibbonWidth
    )) {isSetParameters=true}
   &Color(Color=&SelectLinearColorFromArray(
     Array Sampling Mode=::NewEnumerator1
     Direct Array Index=$User.FXState
     Color Selection Array=DataInterface() {propertyValues {ColorData [<rgba 0 1 0 1>, <rgba 1 1 0 1>, <rgba 0 0.5756 1 1>, <rgba 1 0 0 0>]}}
     Random Seed=?
    )) {enabled=false}
   &DynamicMaterialParameters(
    Index 0 Param 4=&FloatFromCurve(
     FloatCurve=DataInterface() {propertyValues {Curve {keys [{interpMode=RCIM_Cubic, tangentMode=RCTM_User, time=0, value=0}, {interpMode=RCIM_Cubic, tangentMode=RCTM_User, time=1, value=1}]}}}
     CurveIndex=$Emitter.PullbackTime
    )
    Module.Param0WriteEnabled=true
    Module.Param1WriteEnabled=true
    Module.Param2WriteEnabled=true
    Module.Param3WriteEnabled=true
   ) {enabled=false, module=DynamicMaterialParameters001}
  ]
  rendererClasses [&NiagaraSpriteRendererProperties, &NiagaraRibbonRendererProperties]
  renderers [
   {rendererClass=&NiagaraSpriteRendererProperties}
   {rendererClass=&NiagaraRibbonRendererProperties, rendererIndex=1}
   {rendererClass=&NiagaraRibbonRendererProperties, rendererIndex=2}
   {rendererClass=&NiagaraRibbonRendererProperties, rendererIndex=3}
   {rendererClass=&NiagaraRibbonRendererProperties, rendererIndex=4}
   {rendererClass=&NiagaraRibbonRendererProperties, rendererIndex=5}
  ]
 }
 PullbackEffect {
  emitterUpdateScript [
   &EmitterState()
   &SpawnBurst_Instantaneous(Spawn Count=20, Use Spawn Probability=false, UseLoopCountLimit=false)
   /SetVariables(Emitter.PullbackTime:NiagaraFloat=$System.Timeline004.NormalizedTime) {isSetParameters=true}
  ]
  particleSpawnScript [
   &InitializeParticle(Color Mode=::NewEnumerator0, Lifetime=5, UsePositionOffset=false, Write Material Random=true) {enabled=false}
   &InitializeParticle(
    Color=<rgba 12 9.3741 4.26 1>
    Lifetime=0
    Ribbon Width=2
    Ribbon Width Mode=::NewEnumerator1
    Sprite Size Mode=::NewEnumerator3
    Uniform Sprite Size=8.4062
    UsePositionOffset=false
    Write Material Random=true
   ) {module=InitializeParticle001}
   /SetVariables(Particles.RibbonLinkOrder:NiagaraFloat=&ReturnNormalizedExecIndex()) {isSetParameters=true}
   /SetVariables() {isSetParameters=true}
   /SetVariables() {isSetParameters=true}
  ]
  particleUpdateScript [
   &ParticleState() {enabled=false}
   &ParticleState(
    Kill Particles When Lifetime Has Elapsed=false
    Let Infinitely Lived Particles Die When Emitter Deactivates=true
    Loop Particles Lifetime=true
   ) {module=ParticleState001}
   /SetVariables(Particles.GlowWidth=80.724) {isSetParameters=true}
   /SetVariables(Particles.GlowColor=&MultiplyLinearColor(
     Linear Color=<rgba 1 0.8935 0.5417 0.146>
     Scale Factor=&MakeVector4(
      X=2
      Y=2
      Z=2
      W=&FloatFromCurve(
       FloatCurve=DataInterface() {propertyValues {Curve {keys [
           <at 0 1 RCIM_Cubic>
           {arriveTangent=-4.25, interpMode=RCIM_Cubic, leaveTangent=-4.25, time=0.1, value=1}
           {arriveTangent=-5, interpMode=RCIM_Cubic, leaveTangent=-5, time=0.2, value=0.15}
           <at 0.3 0 RCIM_Cubic>
          ]}}}
       CurveIndex=$Emitter.PullbackTime
      )
     )
    )) {isSetParameters=true}
   /SetVariables(Particles.RibbonWidth=1.5) {enabled=false, isSetParameters=true}
   /SetVariables(Particles.Position=&Lerp_Vector(
     A=<0 0 0>
     B=&MakeVector(
      X=&Multiply_Float(
       A=128
       B=&FloatFromCurve(
        FloatCurve=DataInterface() {propertyValues {Curve {keys [<at 0 1 RCIM_Cubic>, {arriveTangent=-1, interpMode=RCIM_Cubic, leaveTangent=-1, time=0.7, value=0.6}, <at 1 0 RCIM_Cubic>]}}}
        CurveIndex=$Emitter.PullbackTime
       )
      )
      Y=0
      Z=0
     )
     Alpha=&ReturnNormalizedExecIndex()
    )) {isSetParameters=true}
   /SetVariables(Particles.Position=&self:ScratchDynamicInput_03(
     CurrentPosition:NiagaraPosition=$Particles.Position
     Default:Vector3f=<0 0 0>
     EffectScale:Vector3f=&MakeVector(
      X=1
      Y=&Multiply_Float(
       A=&MakeCustomFloatFromBool(Bool=$User.IsLeftHand, True Float=-1, False Float=1)
       B=&FloatFromCurve(
        FloatCurve=DataInterface() {propertyValues {Curve {keys [<at 0 0.875 RCIM_Cubic>, {arriveTangent=-0.875, interpMode=RCIM_Cubic, leaveTangent=-0.875, time=0.4, value=0.4}, <at 1 0 RCIM_Cubic>]}}}
        CurveIndex=$Emitter.PullbackTime
       )
      )
      Z=&FloatFromCurve(FloatCurve=DataInterface() {propertyValues {Curve {keys [<at 0 1.5 RCIM_Cubic>,<at 1 0 RCIM_Cubic>]}}}, CurveIndex=$Emitter.PullbackTime)
     )
     Pitch:NiagaraFloat=&Add_Float(
      A=&FloatFromCurve(
       FloatCurve=DataInterface() {propertyValues {Curve {keys [
           <at 0 0 RCIM_Cubic>
           {arriveTangent=0.5, interpMode=RCIM_Cubic, leaveTangent=0.5, time=0.2, value=-0.3}
           {arriveTangent=0.2, interpMode=RCIM_Cubic, leaveTangent=0.2, time=0.5, value=0.25}
           {arriveTangent=-0.5, interpMode=RCIM_Cubic, leaveTangent=-0.5, time=0.7, value=-0.2}
           <at 1 0 RCIM_Cubic>
          ]}}}
       CurveIndex=&Add_Float(A=$Emitter.PullbackTime, B=&ReturnNormalizedExecIndex(Normalized Index Scale=-0.2))
       Scale Curve=100
      )
      B=-5
     )
     Roll:NiagaraFloat=0
     RopeLengthPercent:NiagaraFloat=1
     Yaw:NiagaraFloat=&Add_Float(
      A=&FloatFromCurve(
       FloatCurve=DataInterface() {propertyValues {Curve {keys [
           <at 0 0.25 RCIM_Cubic>
           {arriveTangent=-2.1131, interpMode=RCIM_Cubic, leaveTangent=-2.1131, tangentMode=RCTM_User, time=0.3, value=0.75}
           <at 0.4965 0.3118 RCIM_Cubic>
           {interpMode=RCIM_Cubic, leaveTangent=-1.3315, time=1, value=0.75}
          ]}}}
       CurveIndex=$Emitter.PullbackTime
       Scale Curve=360
      )
      B=&ReturnNormalizedExecIndex(Normalized Index Scale=&FloatFromCurve(
        FloatCurve=DataInterface() {propertyValues {Curve {keys [
            <at 0 -1 RCIM_Cubic>
            <at 0.3 2 RCIM_Cubic>
            {arriveTangent=-0.7143, interpMode=RCIM_Cubic, leaveTangent=-0.7143, time=0.5, value=-1}
            <at 1 1.5>
           ]}}}
        CurveIndex=&Add_Float(A=$Emitter.PullbackTime, B=&ReturnNormalizedExecIndex(Normalized Index Scale=-0.1))
        Scale Curve=&FloatFromCurve(
         FloatCurve=DataInterface() {propertyValues {Curve {keys [
             <at 0 0 RCIM_Cubic>
             {arriveTangent=-1.8057, interpMode=RCIM_Cubic, leaveTangent=-1.8057, tangentMode=RCTM_User, time=0.1, value=1.2}
             <at 1 0.375 RCIM_Cubic>
            ]}}}
         CurveIndex=$Emitter.PullbackTime
         Scale Curve=150
        )
       ))
     )
    )) {isSetParameters=true}
   /SetVariables(Particles.Position=&Lerp_Vector(
     A=&Lerp_Vector(
      A=<0 0 0>
      B=&TransformPosition(Position=$System.EndPosition, Source Space=::World, Destination Space=::Local)
      Alpha=&ReturnNormalizedExecIndex()
     )
     B=Linked() {linkedVariable {name=Particles.Position, type=NiagaraPosition}}
     Alpha=&FloatFromCurve(FloatCurve=DataInterface() {propertyValues {Curve {keys [<at 0 0 RCIM_Cubic>,<at 0.1 1 RCIM_Cubic>]}}}, CurveIndex=$Emitter.PullbackTime)
    )) {isSetParameters=true}
   /SetVariables(Particles.RibbonWidth=1.5) {isSetParameters=true}
   &DynamicMaterialParameters(
    Index 0 Param 4=&FloatFromCurve(FloatCurve=DataInterface() {propertyValues {Curve {keys [<at 0 0 RCIM_Cubic>,<at 0.7 1>]}}}, CurveIndex=$Emitter.PullbackTime)
    Module.Param0WriteEnabled=true
    Module.Param1WriteEnabled=true
    Module.Param2WriteEnabled=true
    Module.Param3WriteEnabled=true
   ) {module=DynamicMaterialParameters001}
   &ScaleRibbonWidth(Ribbon Width Scale=&FloatFromCurve(
     FloatCurve=DataInterface() {propertyValues {Curve {keys [<at 0 1 RCIM_Cubic>, {arriveTangent=-1, interpMode=RCIM_Cubic, leaveTangent=-1, time=0.8, value=1}, <at 1 0 RCIM_Cubic>]}}}
     CurveIndex=&Add_Float(A=$Emitter.PullbackTime, B=&ReturnNormalizedExecIndex())
    ))
   /SetVariables(Particles.WooshRibbonW=70) {isSetParameters=true}
   /SetVariables(Particles.WooshRibbonColor:LinearColor=&MultiplyLinearColor(
     Linear Color=<rgba 1 1 1 0.5>
     Scale Factor=&MakeFloat4FromVectorAndFloat(
      "Vector (RGB)"=<1 1 1>
      "Float (Alpha)"=&Multiply_Float(
       A=&FloatFromCurve(
        FloatCurve=DataInterface() {propertyValues {Curve {keys [
            <at 0 0 RCIM_Cubic>
            {arriveTangent=0.1667, interpMode=RCIM_Cubic, leaveTangent=0.1667, time=0.1, value=1}
            <at 0.3 0.05>
            <at 1 0 RCIM_Cubic>
           ]}}}
        CurveIndex=$Emitter.PullbackTime
       )
       B=&FloatFromCurve(
        FloatCurve=DataInterface() {propertyValues {Curve {keys [
            <at 0 0 RCIM_Cubic>
            {arriveTangent=1.4167, interpMode=RCIM_Cubic, leaveTangent=1.4167, time=0.2, value=0.875}
            {arriveTangent=-1.0938, interpMode=RCIM_Cubic, leaveTangent=-1.0938, time=0.6, value=0.85}
            <at 1 0 RCIM_Cubic>
           ]}}}
        CurveIndex=&ReturnNormalizedExecIndex()
       )
      )
     )
    )) {isSetParameters=true}
   &CameraOffset(Camera Offset Amount=30)
  ]
  rendererClasses [&NiagaraSpriteRendererProperties, &NiagaraRibbonRendererProperties]
  renderers [
   {rendererClass=&NiagaraSpriteRendererProperties}
   {rendererClass=&NiagaraRibbonRendererProperties, rendererIndex=1}
   {rendererClass=&NiagaraRibbonRendererProperties, rendererIndex=2}
   {rendererClass=&NiagaraRibbonRendererProperties, rendererIndex=3}
   {rendererClass=&NiagaraRibbonRendererProperties, rendererIndex=4}
  ]
 }
}
summary {
 systemName=NS_FeycrabLauncher_Connect
 emitters [
  {emitterName=Cabling, rendererClasses [&NiagaraSpriteRendererProperties, &NiagaraRibbonRendererProperties]}
  {
   emitterName=PullbackEffect
   rendererClasses [&NiagaraSpriteRendererProperties, &NiagaraRibbonRendererProperties]
  }
 ]
 userVariables [
  User.A {type=NiagaraFloat, defaultValue=0, description=""}
  User.DebugFX {type=NiagaraBool}
  User.EndPos {type=Vector3f, description="", defaultValue {struct=Vector3f, value=<233.3837 0 0>}}
  User.FXState {type=NiagaraInt32}
  User.IsFrontEnd {type=NiagaraBool}
  User.IsLeftHand {type=NiagaraBool}
  User.StartPos {type=Vector3f, description="", defaultValue {struct=Vector3f, value=<0 0 0>}}
  User.Target {type=NiagaraPosition, description="", defaultValue {}}
 ]
}
