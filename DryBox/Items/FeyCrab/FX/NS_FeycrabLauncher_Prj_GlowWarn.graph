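# Niagara system for the glow and warning effect on the FeyCrab launcher projectile.
# A point light, glow sprite and wrap mesh share the system's blended highlight color, as do airglow and flare sprites.
# Spawn groups place airglow, flares and a blinking warning mesh at array-defined offsets, driven by system brightness and ramp values.
# Flares scale and camera-offset with distance and dim by time of day; debug and translucent-wrap emitters are disabled.
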NiagaraSystem /DryBox/Items/FeyCrab/FX/NS_FeycrabLauncher_Prj_GlowWarn
dependencies {
 usedDataInterfaces [&NiagaraDataInterfaceCurve, &NiagaraDataInterfaceArrayFloat3, &NiagaraDataInterfaceArrayFloat]
 usedDynamicInputs [&MakeVector2D, &LerpLinearColors]
 usedModules [
  &self:SystemSpawnScript.SystemScriptSource.SystemScriptGraph.NiagaraNodeAssignment_1.SetVariables_#1
  &self:SystemSpawnScript.SystemScriptSource.SystemScriptGraph.NiagaraNodeAssignment_2.SetVariables_#2
  &self:SystemSpawnScript.SystemScriptSource.SystemScriptGraph.NiagaraNodeAssignment_7.SetVariables_#3
  &self:SystemSpawnScript.SystemScriptSource.SystemScriptGraph.NiagaraNodeAssignment_8.SetVariables_#4
  &SystemState
  &Timeline
  &self:SystemSpawnScript.SystemScriptSource.SystemScriptGraph.NiagaraNodeAssignment_5.SetVariables_#5
  &self:SystemSpawnScript.SystemScriptSource.SystemScriptGraph.NiagaraNodeAssignment_0.SetVariables_#6
  &self:SystemSpawnScript.SystemScriptSource.SystemScriptGraph.NiagaraNodeAssignment_6.SetVariables_#7
  &self:SystemSpawnScript.SystemScriptSource.SystemScriptGraph.NiagaraNodeAssignment_4.SetVariables_#8
  &self:SystemSpawnScript.SystemScriptSource.SystemScriptGraph.NiagaraNodeAssignment_3.SetVariables_#9
 ]
}
emitters {
 Airglow {
  emitterUpdateScript [
   &EmitterState()
   &SpawnBurst_Instantaneous(Use Spawn Probability=false, UseLoopCountLimit=false)
   &SpawnBurst_Instantaneous(Spawn Group=1, Use Spawn Probability=false, UseLoopCountLimit=false) {module=SpawnBurst_Instantaneous001}
  ]
  particleSpawnScript [&InitializeParticle(
    Color Mode=::NewEnumerator0
    Lifetime=$Emitter.CurrentLoopDuration
    Position Offset=&SelectVectorFromArray(Direct Array Index=&SpawnGroup(), Vector Selection Array=$System.SpawnLocations)
    Sprite Size Mode=::NewEnumerator3
    Uniform Sprite Size=300
    UsePositionOffset=true
    Write Material Random=true
   )]
  particleUpdateScript [
   &ParticleState()
   &SolveForcesAndVelocity(Clamp Velocity=false, Limit Acceleration=false)
   &Color(Color=&HueShiftLinearColor(
     Hue=0
     Saturation=1
     Value=0.5
     Linear Color=&MultiplyLinearColor(Linear Color=$System.BlendedHighlightColor, Scale Factor=<1 1 1 0.35>)
    ))
  ]
  rendererClasses [&NiagaraSpriteRendererProperties]
  renderers [{rendererClass=&NiagaraSpriteRendererProperties}]
 }
 FN_Debug_Distance {
  bEnabled=false
  emitterUpdateScript [
   &EmitterState(Life Cycle Mode=::NewEnumerator1, Loop Duration=10, UseLoopDelay=false)
   &SpawnBurst_Instantaneous(Spawn Count=1, Use Spawn Probability=false)
  ]
  particleSpawnScript [/Niagara/Modules/Spawn/Initialization/InitializeParticle(
    Lifetime=5
    Module.Write Color=true
    Module.Write Lifetime=true
    Module.Write Mass=true
    Module.Write Material Random=true
    Module.Write Position=true
    Module.Write Scale=false
    Module.Write SpriteRotation=false
    Module.Write SpriteSize=true
    Position=&Add_Vector(A=<0 0 100>, B=<0 0 0>)
    Sprite Size=<25 25>
   )]
  particleUpdateScript [
   &ParticleState(Kill Particles When Lifetime Has Elapsed=false, Let Infinitely Lived Particles Die When Emitter Deactivates=true)
   /SetVariables(Particles.Debug:NiagaraFloat=&Divide_Float(A=$Engine.Owner.LODDistance, B=10)) {isSetParameters=true}
   &DynamicMaterialParameters(
    Index 0 Param 1=$Particles.Debug
    Module.Param0WriteEnabled=true
    Module.Param1WriteEnabled=true
    Module.Param2WriteEnabled=true
    Module.Param3WriteEnabled=true
   )
   &CameraOffset(Camera Offset Amount=100)
  ]
  rendererClasses [&NiagaraSpriteRendererProperties]
  renderers [{rendererClass=&NiagaraSpriteRendererProperties}]
 }
 LightAndGlow {
  emitterUpdateScript [
   &EmitterState()
   &SpawnBurst_Instantaneous(Use Spawn Probability=false, UseLoopCountLimit=false)
  ]
  particleSpawnScript [/Niagara/Modules/Spawn/Initialization/V2/InitializeParticle(
    Color=<rgba 1.25 0.8781 0.1497 2>
    Lifetime=8
    Sprite Size=<69.6964 371.1906>
    Sprite Size Mode=::NewEnumerator4
    UsePositionOffset=true
    Write Material Random=true
   )]
  particleUpdateScript [
   &ParticleState()
   &ScaleColor(
    Scale Alpha=&Lerp_Float(
     A=1
     B=&Add_Float(A=1, B=$System.CurrentLoopDuration)
     Alpha=&FloatFromCurve(
      FloatCurve=DataInterface() {propertyValues {Curve {keys [
          {interpMode=RCIM_Cubic, tangentMode=RCTM_User, time=0, value=0}
          {interpMode=RCIM_Cubic, tangentMode=RCTM_User, time=0.5, value=1}
          {tangentMode=RCTM_User, time=1, value=0}
         ]}}}
      CurveIndex=$System.NormalizedLoopAge
      Scale Curve=&Multiply_Float(A=0.333, B=$User.Reveal)
     )
    )
    Scale RGB=&VectorFromFloat(Value=&Multiply_Float(
      A=$User.Reveal
      B=&Waveform(
       Waveform Count=::NewEnumerator1
       "Waveform [1]"=::NewEnumerator12
       "[1] Frequency"=10
       "[1] Amplitude Min/Max"=<0 1>
       "[1] Bias"=0.2
       "[2] Average"=false
       "[2] Input Position"=0
       "[2] Phase"=$Emitter.Age
       "[2] Phase Scale"=1
       "[2] Compound Frequency"=2
       "[2] Frequency"=2
       "[2] Amplitude Min/Max"=<0 1>
       "[2] Scale"=1
       "[2] Bias"=0
       "[2] Exponent"=1
       "[2] Pulse Duration"=0.1
       "[2] Clip Triangle Min/Max"=<0 0>
       "[2] Random Seed"=0
      )
     ))
    ScaleA=true
    ScaleRGB=true
   ) {enabled=false}
   &DynamicMaterialParameters(
    Index 0 Param 1=0.35
    Index 0 Param 2=2
    Module.Param0WriteEnabled=true
    Module.Param1WriteEnabled=true
    Module.Param2WriteEnabled=true
    Module.Param3WriteEnabled=true
   )
   &DynamicMaterialParameters(
    Index 0 Param 4=190.7661
    Module.Param0WriteEnabled=true
    Module.Param1WriteEnabled=true
    Module.Param2WriteEnabled=true
    Module.Param3WriteEnabled=true
   ) {enabled=false, module=DynamicMaterialParameters001}
   &Color(Color=&MultiplyLinearColor(Linear Color=$System.BlendedHighlightColor, Scale Factor=&MakeVector4_FromFloat(Float=1.5)))
   &FN_Light_Attributes(
    Color=$System.BlendedHighlightColor
    Create Color Attribute=true
    Create Exponent Attribute=true
    Create Position Attribute=true
    Create Radius Attribute=true
    Enable Particle Light Enable=true
    Exponent=4
    Light Enabled=true
    Position=<-60 0 0>
    Radius=&Lerp_Float(A=512, B=512, Alpha=$System.LightBrightness)
    Scale Color=&FloatFromCurve(FloatCurve=DataInterface() {propertyValues {Curve {keys [<at 0 0.45>,<at 1 1>]}}}, CurveIndex=$System.LightBrightness)
   )
   &FN_Gen5_Light() {enabled=false}
   &SpriteFacingAndAlignment(Sprite Alignment=<1 0 0>, WriteAlignment=true, WriteFacing=true)
   &Color(Color=<rgba 5 0 0 8>) {enabled=false, module=Color001}
  ]
  rendererClasses [&NiagaraLightRendererProperties, &NiagaraSpriteRendererProperties]
  renderers [
   {rendererClass=&NiagaraLightRendererProperties, rendererIndex=0}
   {rendererClass=&NiagaraSpriteRendererProperties, rendererIndex=1}
  ]
 }
 WarningWrapper {
  emitterUpdateScript [
   &EmitterState(Life Cycle Mode=::NewEnumerator1, Loop Duration=0.5, UseLoopDelay=false)
   &SpawnBurst_Instantaneous(Spawn Group=1, Use Spawn Probability=false, UseLoopCountLimit=false) {module=SpawnBurst_Instantaneous001}
   &SpawnBurst_Instantaneous(Use Spawn Probability=false, UseLoopCountLimit=false)
   /SetVariables(Emitter.Blink:NiagaraFloat=&MakeCustomFloatFromBool(
     Bool=true
     True Float=&Waveform("Waveform [1]"=::NewEnumerator0, "[1] Frequency"=5, "[1] Amplitude Min/Max"=<0.1 1>)
     False Float=1
    )) {isSetParameters=true, module=SetVariables_#10}
  ]
  particleSpawnScript [
   &InitializeParticle(
    Color Mode=::NewEnumerator0
    Mesh Scale Mode=::NewEnumerator3
    Mesh Uniform Scale=1.1
    Position Offset=&SelectVectorFromArray(Direct Array Index=&SpawnGroup(), Vector Selection Array=$System.SpawnLocations)
    Sprite Size=<125 64>
    Sprite Size Mode=::NewEnumerator4
    UsePositionOffset=true
    Write Material Random=true
   )
   /SetVariables(Particles.FlareSize:Vector2f=<100 100>) {isSetParameters=true, module=SetVariables_#11}
  ]
  particleUpdateScript [
   &ParticleState()
   &ScaleMeshSize(Scale Factor=&VectorFromFloat(Value=&FloatFromCurve(
      FloatCurve=DataInterface() {propertyValues {Curve {keys [
          {arriveTangent=4.1785, interpMode=RCIM_Cubic, leaveTangent=4.1785, tangentMode=RCTM_User, time=0, value=0}
          {arriveTangent=1.1179, interpMode=RCIM_Cubic, leaveTangent=1.1179, tangentMode=RCTM_User, time=0.2042, value=0.6228}
          {arriveTangent=0.0224, interpMode=RCIM_Cubic, leaveTangent=0.0224, tangentMode=RCTM_User, time=1, value=1}
         ]}}}
      Scale Curve=0.7
     )))
   &ScaleSpriteSize(
    Scale Sprite Size Mode=::NewEnumerator0
    Uniform Scale Factor=&FN_LOD_Distance_Remap(Start Distance=100, End Distance=8000, Minimum Output=1, Maximum Output=4)
   )
   &ScaleSpriteSize(Initial Sprite Size:Vector2f=$StackContext.SpriteSize, Scale Sprite Size Mode=::NewEnumerator0, Uniform Scale Factor=$Emitter.Blink) {module=ScaleSpriteSize001}
   &SolveForcesAndVelocity(Clamp Velocity=false, Limit Acceleration=false)
   &Color(Color=<rgba 33 24.0267 7.3382 0.35>)
   /SetVariables(Particles.FlareSize:Vector2f=&Multiply_Vector2D_ByFloat(Float=$Emitter.Blink, Vector2D=$Particles.Initial.FlareSize)) {isSetParameters=true, module=SetVariables_#12}
   /SetVariables(Particles.FlareColor=&MultiplyLinearColor(
     Linear Color=<rgba 1 0.8333 0.4177 0.2>
     Scale Factor=&MakeVector4_FromFloat(Float=&Multiply_Float(A=$Emitter.Blink, B=$System.WarningBrightness))
    )) {isSetParameters=true, module=SetVariables_#13}
   &CameraOffset(Camera Offset Amount=&FN_LOD_Distance_Remap(Start Distance=0, End Distance=8000, Minimum Output=1, Maximum Output=250))
   &ScaleColor(
    Color Value To Scale=$StackContext.Color
    Scale Alpha=&FloatFromCurve(FloatCurve=DataInterface() {propertyValues {Curve {keys [<at 0 0 RCIM_Cubic>,<at 0.1 1 RCIM_Cubic>,<at 1 0 RCIM_Cubic>]}}})
    ScaleA=true
    ScaleRGB=true
   )
   &DynamicMaterialParameters(Write Parameter Index 0=false)
  ]
  rendererClasses [&NiagaraMeshRendererProperties]
  renderers [{rendererClass=&NiagaraMeshRendererProperties, rendererIndex=0}]
 }
 WarningWrapper001 {
  emitterUpdateScript [
   &EmitterState(
    Enable Distance Culling=true
    MaxDistance=10000
    Module.bUseMaxDistance=true
    Module.bUseMinDistance=false
    Scalability Mode=::NewEnumerator1
   )
   &SpawnBurst_Instantaneous(Spawn Group=1, Use Spawn Probability=false, UseLoopCountLimit=false) {module=SpawnBurst_Instantaneous001}
   &SpawnBurst_Instantaneous(Use Spawn Probability=false, UseLoopCountLimit=false)
   /SetVariables(Emitter.Blink:NiagaraFloat=$System.LightBrightness) {isSetParameters=true, module=SetVariables_#10}
   &SpawnBurst_Instantaneous(Spawn Time=6.0138, Use Spawn Probability=false, UseLoopCountLimit=false) {enabled=false, module=SpawnBurst_Instantaneous002}
  ]
  particleSpawnScript [
   &InitializeParticle(
    Color Mode=::NewEnumerator0
    Lifetime=$Emitter.CurrentLoopDuration
    Position Offset=&SelectVectorFromArray(Direct Array Index=&SpawnGroup(), Vector Selection Array=$System.SpawnLocations)
    Sprite Rotation Mode=::NewEnumerator2
    Sprite Size=<150 150>
    Sprite Size Mode=::NewEnumerator4
    UsePositionOffset=true
    Write Material Random=true
   )
   /SetVariables(Particles.FlareSize:Vector2f=<45 45>) {isSetParameters=true, module=SetVariables_#11}
   /SetVariables(Particles.SpikesRot=45) {isSetParameters=true}
   /SetVariables() {isSetParameters=true}
  ]
  particleUpdateScript [
   &ParticleState()
   &ScaleSpriteSize(
    Scale Sprite Size Mode=::NewEnumerator0
    Uniform Scale Factor=&FN_LOD_Distance_Remap(Start Distance=500, End Distance=8000, Minimum Output=1.5, Maximum Output=3)
   )
   &ScaleSpriteSize(Initial Sprite Size:Vector2f=$StackContext.SpriteSize, Scale Sprite Size Mode=::NewEnumerator0, Uniform Scale Factor=$Emitter.Blink) {module=ScaleSpriteSize001}
   &SolveForcesAndVelocity(Clamp Velocity=false, Limit Acceleration=false)
   &Color(
    Color=&MultiplyLinearColor(
     Linear Color=&MultiplyLinearColor(Linear Color=$System.BlendedHighlightColor, Scale Factor=&MakeVector4_FromFloat(Float=$Emitter.Blink))
     Scale Factor=&MakeVector4_FromFloat(Float=1)
    )
    Scale Alpha=&FNTimeOfDayScalar(Night Scale=0.3, Morning Scale=0.7, Evening Scale=0.7)
   )
   /SetVariables(Particles.AirglowColor:LinearColor=&MultiplyLinearColor(
     Linear Color=&MultiplyLinearColor(Linear Color=$System.BlendedHighlightColor, Scale Factor=&MakeVector4_FromFloat(Float=$Emitter.Blink))
     Scale Factor=&MakeVector4_FromFloat(Float=&FNTimeOfDayScalar(Night Scale=0.3, Morning Scale=0.5, Day Scale=0.5, Evening Scale=0.5))
    )) {isSetParameters=true}
   /SetVariables(Particles.FlareSize:Vector2f=&Multiply_Vector2D_ByFloat(Float=$Emitter.Blink, Vector2D=$Particles.Initial.FlareSize)) {isSetParameters=true, module=SetVariables_#12}
   /SetVariables(Particles.FlareColor=&MultiplyLinearColor(
     Linear Color=<rgba 12 10.0921 7.4225 0.25>
     Scale Factor=&MakeVector4_FromFloat(Float=&Multiply_Float(
       A=&Multiply_Float(A=$Emitter.Blink, B=$System.RampUp)
       B=&FNTimeOfDayScalar(Night Scale=0.3, Morning Scale=0.7, Evening Scale=0.7)
      ))
    )) {isSetParameters=true, module=SetVariables_#13}
   &CameraOffset(Camera Offset Amount=&FN_LOD_Distance_Remap(Start Distance=100, End Distance=8000, Minimum Output=40, Maximum Output=125))
   &DynamicMaterialParameters(
    Index 0 Param 1=&Waveform(
     "Waveform [1]"=::NewEnumerator12
     "[1] Input Position"=&RandomRangeFloat(Minimum=0, Maximum=1)
     "[1] Phase Scale"=3
     "[1] Frequency"=8
     "[1] Amplitude Min/Max"=<0.65 2>
    )
    Index 0 Param 2=3
    Module.Param0WriteEnabled=true
    Module.Param1WriteEnabled=true
    Module.Param2WriteEnabled=true
    Module.Param3WriteEnabled=true
   )
   /SetVariables(Particles.DistanceScaler:NiagaraFloat=&ClampFloat(
     Float=&FN_LOD_Distance_Remap(Start Distance=$System.MinDistance, End Distance=$System.MaxDistance, Minimum Output=0, Maximum Output=1)
     Min=0.35
     Max=0.75
    )) {isSetParameters=true}
   /Niagara/Modules/Update/Camera/MaintainInCameraParticleScale() {enabled=false}
   /SetVariables(
    Particles.FlareSize:Vector2f=&Multiply_Vector2D_ByFloat(Float=$Particles.DistanceScaler, Vector2D=$StackContext.FlareSize)
    Particles.SpriteSize=&Multiply_Vector2D_ByFloat(
     Float=/Niagara/DynamicInputs/Scalability/ScalabilityDistanceBasedFloat(
      Curve:NiagaraDataInterfaceCurve=DataInterface() {propertyValues {Curve {keys [<at 0.15 0.5>,<at 0.35 1>,<at 0.45 1.5>,<at 1 1.5>]}}}
      CurveIndex:NiagaraFloat=$Emitter.Scalability.DistanceFraction
      Bias:NiagaraFloat=0
      Scale:NiagaraFloat=1
     )
     Vector2D=&Multiply_Vector2D_ByFloat(Float=$Particles.DistanceScaler, Vector2D=$StackContext.SpriteSize)
    )
   ) {isSetParameters=true}
   /SetVariables(
    Particles.FlareSize:Vector2f=&Multiply_Vector2D_ByFloat(Float=$Particles.DistanceScaler, Vector2D=$StackContext.FlareSize)
    Particles.SpriteSize=&Multiply_Vector2D_ByFloat(Float=$Particles.DistanceScaler, Vector2D=$StackContext.SpriteSize)
   ) {enabled=false, isSetParameters=true}
  ]
  rendererClasses [&NiagaraSpriteRendererProperties]
  renderers [
   {rendererClass=&NiagaraSpriteRendererProperties}
   {rendererClass=&NiagaraSpriteRendererProperties, rendererIndex=1}
   {rendererClass=&NiagaraSpriteRendererProperties, rendererIndex=2}
  ]
 }
 WrapMesh {
  emitterUpdateScript [
   &EmitterState()
   &SpawnBurst_Instantaneous(Use Spawn Probability=false, UseLoopCountLimit=false)
  ]
  particleSpawnScript [
   &InitializeParticle(
    Color Mode=::NewEnumerator0
    Lifetime=6
    Mesh Scale=&VectorFromFloat(Value=&MakeCustomFloatFromBool(Bool=true, True Float=1.2453, False Float=1))
    Mesh Scale Mode=::NewEnumerator4
    Sprite Size=<250 125>
    Sprite Size Mode=::NewEnumerator4
    UsePositionOffset=false
    Write Material Random=true
   )
   /SetVariables(Particles.ShowFresnel=1) {enabled=false, isSetParameters=true, module=SetVariables_#14}
   &InitialMeshOrientation(Mesh Orientation Mode=::NewEnumerator2, Orientation Vector=<0 0 1>, Use Rotation Vector=false)
   &DynamicMaterialParameters(
    Index 0 Param 1=2
    Index 0 Param 2=2
    Index 0 Param 3=2
    Index 0 Param 4=0.5
    Module.Param0WriteEnabled=true
    Module.Param1WriteEnabled=true
    Module.Param2WriteEnabled=true
    Module.Param3WriteEnabled=true
   ) {module=DynamicMaterialParameters001}
   &CameraOffset(Camera Offset Amount=5)
  ]
  particleUpdateScript [
   &ParticleState()
   &SolveForcesAndVelocity(Clamp Velocity=false, Limit Acceleration=false)
   &Color(Color=&MultiplyLinearColor(
     Linear Color=$System.BlendedHighlightColor
     Scale Factor=&MakeFloat4FromVectorAndFloat("Vector (RGB)"=&VectorFromFloat(Value=30), "Float (Alpha)"=1)
    ))
   /SetVariables(Particles.ShowFresnel:NiagaraFloat=/Niagara/DynamicInputs/Lerp/SmoothLerpOverTime_Float(
     Convergence Rate:NiagaraFloat=4
     Smooth Value:NiagaraFloat=$Particles.ShowFresnel
     Target Value:NiagaraFloat=&MakeCustomFloatFromBool(Bool=&SetBoolByIntComparison(A=$User.State, B=1, Comparison Type=::NewEnumerator4), True Float=0, False Float=1)
    )) {enabled=false, isSetParameters=true, module=SetVariables_#15}
  ]
  rendererClasses [&NiagaraMeshRendererProperties]
  renderers [{rendererClass=&NiagaraMeshRendererProperties, rendererIndex=0}]
 }
 WrapMesh_Trans {
  bEnabled=false
  emitterUpdateScript [
   &EmitterState()
   &SpawnBurst_Instantaneous(Use Spawn Probability=false, UseLoopCountLimit=false)
  ]
  particleSpawnScript [
   &InitializeParticle(
    Color Mode=::NewEnumerator0
    Lifetime=6
    Mesh Scale=&VectorFromFloat(Value=&MakeCustomFloatFromBool(Bool=true, True Float=1.2453, False Float=1))
    Mesh Scale Mode=::NewEnumerator4
    Sprite Size=<250 125>
    Sprite Size Mode=::NewEnumerator4
    UsePositionOffset=false
    Write Material Random=true
   )
   /SetVariables(Particles.ShowFresnel=1) {isSetParameters=true, module=SetVariables_#14}
   &InitialMeshOrientation(Mesh Orientation Mode=::NewEnumerator2, Orientation Vector=<0 0 1>, Use Rotation Vector=false)
  ]
  particleUpdateScript [
   &ParticleState()
   &SolveForcesAndVelocity(Clamp Velocity=false, Limit Acceleration=false)
   &Color(Color=$System.BlendedHighlightColor)
   /SetVariables(Particles.ShowFresnel:NiagaraFloat=&SmoothLerpOverTime_Float(
     Convergence Rate:NiagaraFloat=4
     Smooth Value:NiagaraFloat=$Particles.ShowFresnel
     Target Value:NiagaraFloat=&MakeCustomFloatFromBool(Bool=&SetBoolByIntComparison(A=$User.State, B=1, Comparison Type=::NewEnumerator4), True Float=0, False Float=1)
    )) {isSetParameters=true, module=SetVariables_#15}
   &DynamicMaterialParameters(
    Index 0 Param 1=$Particles.ShowFresnel
    Module.Param0WriteEnabled=true
    Module.Param1WriteEnabled=true
    Module.Param2WriteEnabled=true
    Module.Param3WriteEnabled=true
   )
  ]
  rendererClasses [&NiagaraSpriteRendererProperties, &NiagaraMeshRendererProperties]
  renderers [
   {rendererClass=&NiagaraSpriteRendererProperties}
   {rendererClass=&NiagaraMeshRendererProperties, rendererIndex=1}
  ]
 }
}
summary {
 systemName=NS_FeycrabLauncher_Prj_GlowWarn
 emitters [
  {emitterName=Airglow, rendererClasses [&NiagaraSpriteRendererProperties]}
  {emitterName=WarningWrapper001, rendererClasses [&NiagaraSpriteRendererProperties]}
  {emitterName=LightAndGlow, rendererClasses [&NiagaraLightRendererProperties, &NiagaraSpriteRendererProperties]}
  {
   bEnabled=false
   emitterName=WrapMesh_Trans
   rendererClasses [&NiagaraSpriteRendererProperties, &NiagaraMeshRendererProperties]
  }
  {emitterName=WarningWrapper, rendererClasses [&NiagaraMeshRendererProperties]}
  {bEnabled=false, emitterName=FN_Debug_Distance, rendererClasses [&NiagaraSpriteRendererProperties]}
  {emitterName=WrapMesh, rendererClasses [&NiagaraMeshRendererProperties]}
 ]
 userVariables [
  User.DamageAmount {type=NiagaraFloat, defaultValue=0, description=""}
  User.FuelAmount {type=NiagaraFloat, defaultValue=0, description=""}
  User.Global Scale {type=NiagaraFloat, defaultValue=0, description=""}
  User.Reveal {type=NiagaraFloat, defaultValue=0, description=""}
  User.State {type=NiagaraInt32}
  User.WarnDuration {type=NiagaraFloat, defaultValue=0, description=""}
 ]
}
