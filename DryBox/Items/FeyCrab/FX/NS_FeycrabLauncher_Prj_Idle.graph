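# Niagara idle effect for the FeyCrab launcher projectile, drawing a thruster exhaust with glow and wind.
# BrightEdge and Flame emitters layer short-lived, randomly scaled meshes with panning materials at Thruster_Local_Pos, beneath a waveform-pulsed, camera-offset glow sprite.
# User.FuelAmount drives edge spawn rate, brightness and mesh scale through curves, with User.DamageAmount scaling some of those curves.
# A separate Wind mesh sits forward on X, while the Left_Jet sprite emitter is disabled.
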NiagaraSystem /DryBox/Items/FeyCrab/FX/NS_FeycrabLauncher_Prj_Idle
dependencies {
 usedDataInterfaces [
  &NiagaraDataInterfaceVectorCurve
  &NiagaraDataInterfaceCurve
  &NiagaraDataInterfaceVector2DCurve
 ]
 usedDynamicInputs []
 usedModules [
  &self:SystemSpawnScript.SystemScriptSource.SystemScriptGraph.NiagaraNodeAssignment_2.SetVariables_#1
  &self:SystemSpawnScript.SystemScriptSource.SystemScriptGraph.NiagaraNodeAssignment_3.SetVariables_#2
  &SystemState
 ]
}
emitters {
 BrightEdge_Left_v2 {
  emitterUpdateScript [
   &EmitterState()
   &SpawnRate(
    SpawnRate=&Lerp_Float(
     A=10
     B=22
     Alpha=&FloatFromCurve(
      FloatCurve=DataInterface() {propertyValues {Curve {keys [
          {interpMode=RCIM_Cubic, tangentMode=RCTM_Break, time=0, value=0}
          {arriveTangent=1.4927, interpMode=RCIM_Cubic, tangentMode=RCTM_Break, time=1, value=1}
         ]}}}
      CurveIndex=$User.FuelAmount
      Scale Curve=$User.DamageAmount
     )
    )
    Use Spawn Probability=false
   )
   &SpawnBurst_Instantaneous(Spawn Count=2, Use Spawn Probability=false)
  ]
  particleSpawnScript [
   &FN_SetSpawnGroup(Number of Spawn Groups=2)
   &InitializeParticle(
    Color=<rgba 24 17.1643 3.4178 1>
    Lifetime Max=0.25
    Lifetime Min=0.1
    Lifetime Mode=::NewEnumerator1
    Mesh Scale=&Multiply_VectorByFloat(
     Vector=&MakeVector(
      X=&Multiply_Float(
       A=&Lerp_Float(
        A=0.15
        B=0.3
        Alpha=&FloatFromCurve(
         FloatCurve=DataInterface() {propertyValues {Curve {keys [
             {interpMode=RCIM_Cubic, tangentMode=RCTM_Break, time=0.3, value=0}
             {arriveTangent=3.8108, interpMode=RCIM_Cubic, tangentMode=RCTM_Break, time=1, value=1}
            ]}}}
         CurveIndex=&RandomRangeFloat(Minimum=0, Maximum=1)
        )
       )
       B=1
      )
      Y=&RandomRangeFloat(Minimum=0.12, Maximum=0.2)
      Z=&RandomRangeFloat(Minimum=0.12, Maximum=0.2)
     )
     Float=1.2
    )
    Mesh Scale Mode=::NewEnumerator4
    Position Offset=$System.Thruster_Local_Pos
    UsePositionOffset=true
    Write Material Random=true
   )
   /SetVariables(Particles.ConeOrientation:Quat4f=<0 0 1 0>) {isSetParameters=true}
   /SetVariables(
    Particles.ConeColor:LinearColor=<rgba 1.5 0.9091 0.1997 5>
    Particles.ConePos:Vector3f=<-5 0 0>
    Particles.ConeScale:Vector3f=&Multiply_VectorByFloat(
     Vector=<-0.02 0.025 0.025>
     Float=&Lerp_Float(
      A=0.7
      B=1
      Alpha=&FloatFromCurve(
       FloatCurve=DataInterface() {propertyValues {Curve {keys [
           {interpMode=RCIM_Cubic, tangentMode=RCTM_Break, time=0, value=0}
           {arriveTangent=1.4927, interpMode=RCIM_Cubic, tangentMode=RCTM_Break, time=1, value=1}
          ]}}}
       CurveIndex=$User.FuelAmount
      )
     )
    )
   ) {isSetParameters=true}
   &InitialMeshOrientation(
    Mesh Orientation Mode=::NewEnumerator2
    Rotation=&RandomRangeVector(Minimum=<0 0 0>, Maximum=<1 0 0>)
    Use Rotation Vector=true
   )
  ]
  particleUpdateScript [
   &ParticleState()
   &ScaleMeshSize(Scale Factor=&VectorFromCurve(
     VectorCurve=DataInterface() {propertyValues {
       XCurve {keys [{arriveTangent=6.3957, interpMode=RCIM_Cubic, time=0, value=0.47}, <at 0.9748 2>]}
       YCurve {keys [{arriveTangent=6.1645, interpMode=RCIM_Cubic, time=0.0035, value=0.4903}, <at 0.9786 2>]}
       ZCurve {keys [
         <at -0.0011 1.0067 RCIM_Cubic>
         {arriveTangent=0.0529, interpMode=RCIM_Cubic, leaveTangent=0.0529, tangentMode=RCTM_User, time=1.0016, value=1}
        ]}
      }}
     Scale Curve=&Multiply_VectorByFloat(
      Vector=<1 1 1>
      Float=&Lerp_Float(
       A=0.8
       B=1.2
       Alpha=&FloatFromCurve(
        FloatCurve=DataInterface() {propertyValues {Curve {keys [
            {interpMode=RCIM_Cubic, tangentMode=RCTM_Break, time=0, value=0}
            {arriveTangent=1.4927, interpMode=RCIM_Cubic, tangentMode=RCTM_Break, time=1, value=1}
           ]}}}
        CurveIndex=$User.FuelAmount
       )
      )
     )
    ))
   &DynamicMaterialParameters(
    Index 0 Param 2=2
    Index 0 Param 4=1
    Module.Param0WriteEnabled=false
    Module.Param1WriteEnabled=true
    Module.Param2WriteEnabled=false
    Module.Param3WriteEnabled=true
   )
   /Game/Effects/Niagara/Modules/Materials/MaterialPanner(Offset:NiagaraFloat=&Add_Float(A=$Emitter.Age, B=$Particles.Initial.MaterialRandom), Speed=-5)
   /SetVariables(Particles.ConeScale:Vector3f=&Multiply_Vector(
     A=$Particles.Initial.ConeScale
     B=&MakeVector(
      X=&FloatFromCurve(FloatCurve=DataInterface() {propertyValues {Curve {keys [<at 0 0.75 RCIM_Cubic>, {arriveTangent=-0.25, interpMode=RCIM_Cubic, leaveTangent=-0.25, time=0.5, value=2.5}, <at 1 0.5 RCIM_Cubic>]}}})
      Y=&FloatFromCurve(FloatCurve=DataInterface() {propertyValues {Curve {keys [<at 0.5 1 RCIM_Cubic>,<at 1 1.75 RCIM_Cubic>]}}})
      Z=&FloatFromCurve(FloatCurve=DataInterface() {propertyValues {Curve {keys [<at 0.5 1 RCIM_Cubic>,<at 1 1.75 RCIM_Cubic>]}}})
     )
    )) {isSetParameters=true}
   &ScaleColor(
    Scale RGB=&VectorFromFloat(Value=&Lerp_Float(
      A=0.7
      B=1
      Alpha=&FloatFromCurve(
       FloatCurve=DataInterface() {propertyValues {Curve {keys [
           {interpMode=RCIM_Cubic, tangentMode=RCTM_Break, time=0, value=0}
           {arriveTangent=1.4927, interpMode=RCIM_Cubic, tangentMode=RCTM_Break, time=1, value=1}
          ]}}}
       CurveIndex=$User.FuelAmount
      )
     ))
    ScaleA=false
    ScaleRGB=true
   )
  ]
  rendererClasses [&NiagaraMeshRendererProperties]
  renderers [
   {rendererClass=&NiagaraMeshRendererProperties, rendererIndex=0}
   {rendererClass=&NiagaraMeshRendererProperties, rendererIndex=1}
  ]
 }
 Flame_Left_v2 {
  emitterUpdateScript [
   &EmitterState()
   &SpawnRate(SpawnRate=9, Use Spawn Probability=false)
   &SpawnBurst_Instantaneous(Spawn Count=1, Spawn Time=0.05, Use Spawn Probability=false) {enabled=false}
  ]
  particleSpawnScript [
   &InitializeParticle(
    Color=<rgba 3 1.1878 0.1797 1>
    Lifetime Max=0.15
    Lifetime Min=0.07
    Lifetime Mode=::NewEnumerator1
    Mesh Scale=&Multiply_VectorByFloat(
     Vector=&Lerp_Vector(
      A=<0.7 0.5 0.5>
      B=<1.25 0.8 0.8>
      Alpha=&FloatFromCurve(
       FloatCurve=DataInterface() {propertyValues {Curve {keys [
           {interpMode=RCIM_Cubic, tangentMode=RCTM_Break, time=0, value=0}
           {arriveTangent=1.393, interpMode=RCIM_Cubic, tangentMode=RCTM_Break, time=1, value=1}
          ]}}}
       CurveIndex=&RandomRangeFloat(Minimum=0, Maximum=1)
      )
     )
     Float=1.2
    )
    Mesh Scale Mode=::NewEnumerator4
    Position Offset=$System.Thruster_Local_Pos
    UsePositionOffset=true
    Write Material Random=true
   )
   /SetVariables(Particles.MeshOrientation=&MakeQuatFromAxisAngle(Axis=<0 0 1>, AngleInDegrees=&UniformRangedFloat(Minimum=180, Maximum=180))) {isSetParameters=true}
  ]
  particleUpdateScript [
   &ParticleState()
   &ScaleColor(
    Scale Alpha=&FloatFromCurve(FloatCurve=DataInterface() {propertyValues {Curve {keys [
         {interpMode=RCIM_Cubic, leaveTangent=9.8521, tangentMode=RCTM_Break, time=0, value=0.1}
         {interpMode=RCIM_Cubic, tangentMode=RCTM_User, time=0.2, value=1}
         {interpMode=RCIM_Cubic, tangentMode=RCTM_Break, time=1, value=0}
        ]}}})
    Scale RGB=&VectorFromFloat(Value=&RandomRangeFloat(Minimum=1, Maximum=10))
    ScaleA=true
    ScaleRGB=true
   )
   &ScaleMeshSize(Scale Factor=&VectorFromCurve(
     VectorCurve=DataInterface() {propertyValues {
       bOptimizeLUT=false
       XCurve {keys [{arriveTangent=-0.4338, interpMode=RCIM_Cubic, leaveTangent=-0.4338, tangentMode=RCTM_User, time=0, value=0.75}, <at 1 2>]}
       YCurve {keys [{arriveTangent=0.1422, interpMode=RCIM_Cubic, leaveTangent=0.1422, tangentMode=RCTM_User, time=0, value=0.75}, <at 1 2>]}
       ZCurve {keys [
         <at -0.0011 1.0067 RCIM_Cubic>
         {arriveTangent=-0.105, interpMode=RCIM_Cubic, leaveTangent=-0.105, tangentMode=RCTM_User, time=1.0016, value=1}
        ]}
      }}
     Scale Curve=&Multiply_VectorByFloat(
      Vector=<1 1 1>
      Float=&Lerp_Float(
       A=0.6
       B=1
       Alpha=&FloatFromCurve(
        FloatCurve=DataInterface() {propertyValues {Curve {keys [
            {interpMode=RCIM_Cubic, tangentMode=RCTM_Break, time=0, value=0}
            {arriveTangent=1.4927, interpMode=RCIM_Cubic, tangentMode=RCTM_Break, time=1, value=1}
           ]}}}
        CurveIndex=$User.FuelAmount
        Scale Curve=$User.DamageAmount
       )
      )
     )
    ))
   &MaterialPanner(Offset:NiagaraFloat=&Add_Float(A=$Emitter.Age, B=$Particles.Initial.MaterialRandom), Speed=-5)
  ]
  rendererClasses [&NiagaraMeshRendererProperties]
  renderers [{rendererClass=&NiagaraMeshRendererProperties, rendererIndex=0}]
 }
 Glow {
  emitterUpdateScript [
   &EmitterState(Life Cycle Mode=::NewEnumerator1, Loop Behavior=::NewEnumerator1, Loop Duration Mode=::NewEnumerator1, UseLoopDelay=false)
   &SpawnBurst_Instantaneous(Use Spawn Probability=false, UseLoopCountLimit=false)
   &SpawnRate(SpawnRate=10, Use Spawn Probability=false) {enabled=false}
  ]
  particleSpawnScript [&InitializeParticle(
    Color=<rgba 1 0.3898 0.0328 0.5>
    Lifetime=0.1
    Position Offset=$System.Thruster_Local_Pos
    Sprite Size Mode=::NewEnumerator3
    Uniform Sprite Size=&Multiply_Float(A=45, B=1)
    UsePositionOffset=true
    Write Material Random=true
   )]
  particleUpdateScript [
   &ParticleState(Kill Particles When Lifetime Has Elapsed=false)
   &ScaleSpriteSize(
    Scale Sprite Size Mode=::NewEnumerator0
    Uniform Scale Factor=&Waveform(
     Waveform Count=::NewEnumerator1
     "Waveform [1]"=::NewEnumerator12
     "[1] Frequency"=5
     "[1] Amplitude Min/Max"=<0.5 1.3>
     "[2] Average"=false
     "Waveform [2]"=::NewEnumerator12
     "[2] Input Position"=0
     "[2] Phase"=$Emitter.Age
     "[2] Phase Scale"=3
     "[2] Compound Frequency"=2
     "[2] Frequency"=6
     "[2] Amplitude Min/Max"=<0.25 1.1>
     "[2] Scale"=1
     "[2] Bias"=0
     "[2] Exponent"=1
     "[2] Pulse Duration"=0.1
     "[2] Clip Triangle Min/Max"=<0 0>
     "[2] Random Seed"=0
    )
   ) {module=ScaleSpriteSize001}
   &SpriteRotationRate(Rotation Rate=50) {enabled=false}
   &ScaleSpriteSize(
    Scale Sprite Size Mode=::NewEnumerator0
    Uniform Scale Factor=&FN_LOD_Distance_Remap(Start Distance=500, End Distance=2500, Minimum Output=1, Maximum Output=1.1)
   )
   &ScaleColor(Scale RGB=&VectorFromFloat(Value=2), ScaleA=false, ScaleRGB=true)
   &CameraOffset(Camera Offset Amount=35)
   /Niagara/Modules/Update/Camera/MaintainInCameraParticleScale() {enabled=false}
  ]
  rendererClasses [&NiagaraSpriteRendererProperties]
  renderers [
   {rendererClass=&NiagaraSpriteRendererProperties}
   {rendererClass=&NiagaraSpriteRendererProperties, rendererIndex=1}
  ]
 }
 Left_Jet_v2 {
  bEnabled=false
  emitterUpdateScript [
   &EmitterState(Life Cycle Mode=::NewEnumerator1, Loop Duration=0.1, UseLoopDelay=false)
   &SpawnRate(SpawnRate=250, Use Spawn Probability=false)
   &SpawnBurst_Instantaneous(Spawn Count=40, Use Spawn Probability=false) {enabled=false}
  ]
  particleSpawnScript [
   /Niagara/Modules/Spawn/Initialization/InitializeParticle(
    Color=<rgba 5 3.1271 0.4771 1>
    Lifetime=&UniformRangedFloat(Minimum=0.1, Maximum=0.1)
    Module.Write Color=true
    Module.Write Lifetime=true
    Module.Write Mass=true
    Module.Write Material Random=true
    Module.Write Position=true
    Module.Write Scale=false
    Module.Write SpriteRotation=false
    Module.Write SpriteSize=true
    Sprite Size=<12 25>
   ) {enabled=false}
   /Niagara/Modules/Spawn/Initialization/V2/InitializeParticle(
    Color=<rgba 5 3.2364 1.3557 0.1>
    Lifetime=&UniformRangedFloat(Minimum=0.1, Maximum=0.1)
    Position Offset:NiagaraPosition=$System.Thruster_Local_Pos
    Sprite Size=<12 40>
    Sprite Size Mode=::NewEnumerator4
    UsePositionOffset=true
    Write Material Random=true
   ) {module=InitializeParticle001}
   &AddVelocityInCone(
    Cone Angle=0
    Module.Use Velocity Falloff On Cone Axis=false
    Velocity Strength=&UniformRangedFloat(Minimum=-50, Maximum=-350)
   )
  ]
  particleUpdateScript [
   &ParticleState()
   &SolveForcesAndVelocity(Clamp Velocity=false, Limit Acceleration=false)
   &Color(
    Color=$Particles.Initial.Color
    Scale Alpha=&FloatFromCurve(FloatCurve=DataInterface() {propertyValues {Curve {keys [
         <at -0.0046 0 RCIM_Cubic>
         {arriveTangent=0.0019, interpMode=RCIM_Cubic, leaveTangent=0.0019, time=0.25, value=1}
         <at 1.0011 0.002 RCIM_Cubic>
        ]}}})
    Scale Color=&VectorFromCurve(VectorCurve=DataInterface() {propertyValues {XCurve {keys [<at 0 5>,<at 0.2 2>]}, YCurve {keys [<at 0 5>,<at 0.2 0.21>]}, ZCurve {keys [<at 0 5>,<at 0.2 0>]}}})
   )
   &ScaleSpriteSize(Scale Factor=&Vector2DFromCurve(
     Vector2Curve=DataInterface() {propertyValues {XCurve {keys [<at 0 1 RCIM_Cubic>,<at 0.9989 0.25>]}, YCurve {keys [<at -0.0011 1.0067 RCIM_Cubic>,<at 0.9989 2>]}}}
     ScaleCurve=&MakeVector2D(
      X=1
      Y=&Multiply_Float(
       A=1
       B=&Lerp_Float(
        A=0.3
        B=1
        Alpha=&FloatFromCurve(
         FloatCurve=DataInterface() {propertyValues {Curve {keys [
             {interpMode=RCIM_Cubic, tangentMode=RCTM_Break, time=0, value=0}
             {arriveTangent=1.4927, interpMode=RCIM_Cubic, tangentMode=RCTM_Break, time=1, value=1}
            ]}}}
         CurveIndex=$User.FuelAmount
         Scale Curve=$User.DamageAmount
        )
       )
      )
     )
    ))
  ]
  rendererClasses [&NiagaraSpriteRendererProperties]
  renderers [{rendererClass=&NiagaraSpriteRendererProperties}]
 }
 Wind {
  emitterUpdateScript [
   &EmitterState()
   &SpawnRate(SpawnRate=5, Use Spawn Probability=false)
  ]
  particleSpawnScript [
   &InitializeParticle(
    Color=<rgba 1 1 1 0.25>
    Lifetime Max=0.6
    Lifetime Min=0.3
    Lifetime Mode=::NewEnumerator1
    Mesh Scale=&Multiply_VectorByFloat(Vector=<1.5 0.5 0.5>, Float=0.5)
    Mesh Scale Mode=::NewEnumerator4
    Position Offset=<90 0 0>
    UsePositionOffset=true
    Write Material Random=true
   )
   &InitialMeshOrientation(Mesh Orientation Mode=::NewEnumerator2, Use Rotation Vector=true)
  ]
  particleUpdateScript [&ParticleState()]
  rendererClasses [&NiagaraMeshRendererProperties]
  renderers [{rendererClass=&NiagaraMeshRendererProperties, rendererIndex=0}]
 }
}
summary {
 systemName=NS_FeycrabLauncher_Prj_Idle
 emitters [
  {emitterName=Wind, rendererClasses [&NiagaraMeshRendererProperties]}
  {bEnabled=false, emitterName=Left_Jet_v2, rendererClasses [&NiagaraSpriteRendererProperties]}
  {emitterName=Flame_Left_v2, rendererClasses [&NiagaraMeshRendererProperties]}
  {emitterName=BrightEdge_Left_v2, rendererClasses [&NiagaraMeshRendererProperties]}
  {emitterName=Glow, rendererClasses [&NiagaraSpriteRendererProperties]}
 ]
 userVariables [
  User.DamageAmount {type=NiagaraFloat, defaultValue=0, description=""}
  User.FuelAmount {type=NiagaraFloat, defaultValue=0, description=""}
  User.RibbonIDOffset {type=NiagaraInt32}
  User.SkeletalMesh {
   type=NiagaraDataInterfaceSkeletalMesh
   defaultValue {
    struct=NiagaraExt_VariableValue_DataInterface
    value {dataInterface=&self:NiagaraDataInterfaceSkeletalMesh_0}
   }
  }
  User.ThrusterOffset {type=NiagaraFloat, defaultValue=0, description=""}
 ]
}
