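# Niagara system for the FeyCrab projectile explosion, layering flash, energy wave with light, mesh distortion, dark chunks, embers and fluid smoke-fire.
# System-level blast source and direction vectors place and aim emitters, with user parameters for scale and color.
# Smoke-fire is a camera-facing Grid2D gas simulation sourced from particles, limited by instance count, quality level and distance.
# A combined flash/wave emitter stands in for the disabled separate EnergyWave and Flash_wide emitters.
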NiagaraSystem /DryBox/Items/FeyCrab/FX/NS_Feycrab_Prj_Explosion
dependencies {
 usedDataInterfaces [
  &NiagaraDataInterfaceCurve
  &NiagaraDataInterfaceParticleRead
  &NiagaraDataInterfaceMeshRendererInfo
  &NiagaraDataInterfaceVector2DCurve
  &NiagaraDataInterfaceColorCurve
  /Script/Niagara.NiagaraDataInterfaceGrid2DCollection
  /Script/Niagara.NiagaraDataInterfaceRenderTarget2D
  &NiagaraDataInterfaceCamera
  &NiagaraDataInterfacePlatformSet
  /Script/Niagara.NiagaraDataInterfaceEmitterProperties
  &NiagaraDataInterfaceVectorCurve
  &NiagaraDataInterfaceArrayFloat
 ]
 usedDynamicInputs []
 usedModules [
  &SystemState
  &self:SystemSpawnScript.SystemScriptSource.SystemScriptGraph.NiagaraNodeAssignment_0.SetVariables_#1
  &self:SystemSpawnScript.SystemScriptSource.SystemScriptGraph.NiagaraNodeAssignment_1.SetVariables_#2
  &self:SystemSpawnScript.SystemScriptSource.SystemScriptGraph.NiagaraNodeAssignment_2.SetVariables_#3
 ]
}
emitters {
 Blast_Stylized {
  bEnabled=false
  emitterUpdateScript [
   &EmitterState(Life Cycle Mode=::NewEnumerator1, Loop Behavior=::NewEnumerator1, Loop Duration=0.4, UseLoopDelay=true)
   &SpawnBurst_Instantaneous(Spawn Count=3, Use Spawn Probability=false)
   &SpawnBurst_Instantaneous(Spawn Count=3, Spawn Group=1, Use Spawn Probability=false) {module=SpawnBurst_Instantaneous001}
  ]
  particleSpawnScript [
   &InitializeParticle(
    Color=<rgba 2 0.8649 0.1068 1>
    Lifetime=&Multiply_Float(A=&RandomRangeFloat(Minimum=0.14, Maximum=0.17), B=0.5)
    Position Offset=&MakeCustomVectorFromBool(
     Bool=&SetBoolByIntComparison(A=&SpawnGroup(), B=0, Comparison Type=::NewEnumerator4)
     True Vector=$System.BlastSrcA
     False Vector=$System.BlastSrcB
    )
    Sprite Size=&Multiply_Vector2D_ByFloat(
     Float=&Multiply_Float(A=$User.Global Scale, B=0.75)
     Vector2D=&UniformRangedVector2D(Minimum=<100 100>, Maximum=<160 200>)
    )
    Sprite Size Mode=::NewEnumerator4
    Sprite UV Mode=::NewEnumerator1
    UsePositionOffset=true
    Write Material Random=true
   )
   /Niagara/Modules/Spawn/Location/SphereLocation(
    MaskSpawnGroup=false
    Module.Override Randomness=false
    Sphere Radius=&Multiply_Float(A=70, B=$User.Global Scale)
    Surface Only=true
   ) {enabled=false}
   &AddVelocity(
    Cone Angle=65
    Cone Axis=&MakeCustomVectorFromBool(
     Bool=&SetBoolByIntComparison(A=&SpawnGroup(), B=0, Comparison Type=::NewEnumerator4)
     True Vector=$System.BlastDir
     False Vector=$System.BlastDir2
    )
    Module.UseConeFalloff=false
    Module.UseVelDistribution=false
    Override Seed=false
    Velocity Mode=::NewEnumerator2
    Velocity Speed=20
   )
   &AddVelocityFromPoint(
    Module.Use Falloff Distance=false
    Velocity Strength=&Multiply_Float(A=&UniformRangedFloat(Minimum=400, Maximum=500), B=$User.Global Scale)
   ) {enabled=false}
   &SubUVAnimation(End Frame=3, SubUV Animation Mode:ENiagara_SubUVLookupMode=::NewEnumerator2)
   &CameraOffset(Camera Offset Amount=&Multiply_Float(A=-10, B=$User.Global Scale))
  ]
  particleUpdateScript [
   &ParticleState()
   &Drag(
    Drag=&FloatFromCurve(FloatCurve=DataInterface() {propertyValues {Curve {keys [<at 0.2 1>,<at 0.3 0>]}}}, Scale Curve=15)
    Use Linear Drag=true
    Use Rotational Drag=false
   )
   &GravityForce(Gravity=&Multiply_VectorByFloat(
     Vector=<0 0 -2.2>
     Float=&FloatFromCurve(
      FloatCurve=DataInterface() {propertyValues {Curve {keys [{arriveTangent=3.1357, interpMode=RCIM_Cubic, leaveTangent=3.1357, tangentMode=RCTM_User, time=0.1, value=0}, <at 0.6 1>]}}}
      CurveIndex=$Emitter.Age
     )
    )) {enabled=false}
   &SolveForcesAndVelocity(Clamp Velocity=false, Limit Acceleration=false)
   &ScaleSpriteSize(Scale Factor=&Vector2DFromCurve(
     Vector2Curve=DataInterface() {propertyValues {
       XCurve {keys [
         <at 0 0.625 RCIM_Cubic>
         {arriveTangent=0.875, interpMode=RCIM_Cubic, leaveTangent=0.875, time=0.4, value=0.625}
         {arriveTangent=1.2531, interpMode=RCIM_Cubic, leaveTangent=1.2531, tangentMode=RCTM_User, time=1, value=1.3}
        ]}
       YCurve {keys [
         {arriveTangent=9.4573, interpMode=RCIM_Cubic, leaveTangent=9.4573, tangentMode=RCTM_User, time=0, value=0}
         <at 0.18 1.625>
         <at 1 2 RCIM_Cubic>
        ]}
      }}
     ScaleCurve=<1.7 2.1>
    ))
   &ScaleColor(
    Linear Color Curve=DataInterface() {propertyValues {
      AlphaCurve {keys [
        {arriveTangent=-0.3577, interpMode=RCIM_Cubic, leaveTangent=-0.3577, tangentMode=RCTM_User, time=0.08, value=1}
        {arriveTangent=-1.087, interpMode=RCIM_Cubic, leaveTangent=-1.0164, tangentMode=RCTM_Break, time=0.2616, value=1}
        <at 1 0>
       ]}
      BlueCurve {keys [<at 0.1976 0>]}
      GreenCurve {keys [<at 0.1976 0.1604>]}
      RedCurve {keys [<at 0.1976 4>]}
     }}
    Scale Mode=::NewEnumerator2
   )
   &DynamicMaterialParameters(
    Index 0 Param 1=&Multiply_Float(A=100, B=$User.Global Scale)
    Index 0 Param 4=&FloatFromCurve(FloatCurve=DataInterface() {propertyValues {Curve {keys [{interpMode=RCIM_Cubic, tangentMode=RCTM_User, time=0, value=0}, {interpMode=RCIM_Cubic, tangentMode=RCTM_User, time=0.4, value=1}]}}})
    Index 1 Param 1=&FloatFromCurve(
     FloatCurve=DataInterface() {propertyValues {Curve {keys [
         {interpMode=RCIM_Cubic, tangentMode=RCTM_User, time=0.2, value=0}
         {arriveTangent=0.6295, interpMode=RCIM_Cubic, leaveTangent=0.6295, tangentMode=RCTM_User, time=1, value=1}
        ]}}}
     Scale Curve=0.03
    )
    Index 1 Param 2=&FloatFromCurve(
     FloatCurve=DataInterface() {propertyValues {Curve {keys [
         {arriveTangent=-9.7084, interpMode=RCIM_Cubic, leaveTangent=-9.7084, tangentMode=RCTM_User, time=0.1, value=50}
         {arriveTangent=-0.3869, interpMode=RCIM_Cubic, leaveTangent=-0.3869, tangentMode=RCTM_User, time=0.35, value=0}
        ]}}}
     CurveIndex=$Emitter.Age
     Scale Curve=&MaskFloatBySpawnGroup(Masked Float=0, Passthrough Float=0.9, Spawn Group Mask=1)
    )
    Index 1 Param 3=&FloatFromCurve(
     FloatCurve=DataInterface() {propertyValues {Curve {keys [{interpMode=RCIM_Cubic, tangentMode=RCTM_User, time=0.2, value=0.5}, {interpMode=RCIM_Cubic, tangentMode=RCTM_User, time=1, value=1}]}}}
     Scale Curve=2
    )
    Index 1 Param 4=&FloatFromCurve(FloatCurve=DataInterface() {propertyValues {Curve {keys [<at 0 1>,<at 1 0>]}}}, Scale Curve=5)
    Module.Param0WriteEnabled=true
    Module.Param0WriteEnabled1=true
    Module.Param1WriteEnabled=false
    Module.Param1WriteEnabled1=true
    Module.Param2WriteEnabled=false
    Module.Param2WriteEnabled1=true
    Module.Param3WriteEnabled=true
    Module.Param3WriteEnabled1=true
    Write Parameter Index 1=true
   )
  ]
  rendererClasses [&NiagaraSpriteRendererProperties]
  renderers [{rendererClass=&NiagaraSpriteRendererProperties}]
 }
 Dark_Chunks_6 {
  simTarget=GPUComputeSim
  emitterUpdateScript [
   &EmitterState(Life Cycle Mode=::NewEnumerator1, Loop Behavior=::NewEnumerator1, Loop Delay=0.1669, Loop Duration=0.1, UseLoopDelay=true)
   &SpawnBurst_Instantaneous(Spawn Count=10, Use Spawn Probability=false)
   &SpawnBurst_Instantaneous(Spawn Count=10, Spawn Group=1, Use Spawn Probability=false) {module=SpawnBurst_Instantaneous001}
  ]
  particleSpawnScript [
   &InitializeParticle(
    Color=<rgba 0 0 0 10>
    Lifetime Max=0.4
    Lifetime Min=0.1
    Lifetime Mode=::NewEnumerator1
    Sprite Rotation Mode=::NewEnumerator2
    Sprite Size Mode=::NewEnumerator1
    Sprite UV Mode=::NewEnumerator3
    Uniform Sprite Size Max=60
    Uniform Sprite Size Min=35
    UsePositionOffset=false
    Write Material Random=true
   )
   &SphereLocation(
    MaskSpawnGroup=false
    Module.Override Randomness=false
    Offset=&MakeCustomVectorFromBool(
     Bool=&SetBoolByIntComparison(A=&SpawnGroup(), B=0, Comparison Type=::NewEnumerator4)
     True Vector=$System.BlastSrcA
     False Vector=$System.BlastSrcB
    )
    Sphere Radius=&Multiply_Float(A=135, B=$User.Global Scale)
    Surface Only=false
   )
   &AddVelocity(
    Cone Angle=100
    Cone Axis=&MakeCustomVectorFromBool(
     Bool=&SetBoolByIntComparison(A=&SpawnGroup(), B=0, Comparison Type=::NewEnumerator4)
     True Vector=$System.BlastDir
     False Vector=$System.BlastDir2
    )
    Module.UseConeFalloff=false
    Module.UseVelDistribution=false
    Override Seed=false
    Velocity Mode=::NewEnumerator2
    Velocity Speed=&Multiply_Float(A=&RandomRangeFloat(Minimum=1500, Maximum=2000), B=$User.Global Scale)
   )
   &ApplyInitialForces(Apply Rotational Force to Rotational Velocity=false)
   /SetVariables(Particles.SpriteSize=&Multiply_Vector2D_ByFloat(Float=$User.Global Scale, Vector2D=$StackContext.SpriteSize)) {isSetParameters=true, module=SetVariables_#4}
   &SubUVAnimation(SubUV Animation Mode:ENiagara_SubUVLookupMode=::NewEnumerator2)
   &CameraOffset(Camera Offset Amount=&Multiply_Float(A=100, B=$User.Global Scale))
  ]
  particleUpdateScript [
   &ParticleState()
   &Drag(Drag=&RandomRangeFloat(Minimum=2, Maximum=6), Use Linear Drag=true, Use Rotational Drag=false)
   &CurlNoiseForce(Noise Frequency=7, Noise Quality / Cost=::NewEnumerator0, Noise Strength=2000, Pan Noise=false) {enabled=false}
   &GravityForce(Gravity=<0 0 -2000>)
   &SolveForcesAndVelocity(Clamp Velocity=false, Limit Acceleration=false)
   &ScaleColor(
    Scale Alpha=&FloatFromCurve(FloatCurve=DataInterface() {propertyValues {Curve {keys [
         {arriveTangent=26.7273, interpMode=RCIM_Cubic, leaveTangent=26.7273, tangentMode=RCTM_User, time=-0.0092, value=0.0144}
         {interpMode=RCIM_Cubic, tangentMode=RCTM_User, time=0.1, value=1}
         {interpMode=RCIM_Cubic, tangentMode=RCTM_User, time=0.8, value=1}
         <at 1.0011 0 RCIM_Cubic>
        ]}}})
    ScaleA=true
    ScaleRGB=false
   )
   &ScaleSpriteSize(Scale Factor=&Vector2DFromFloat(Value=&FloatFromCurve(FloatCurve=DataInterface() {propertyValues {Curve {keys [<at -0.0046 0>,<at 0.025 1>,<at 1.0011 0>]}}})))
  ]
  rendererClasses [&NiagaraSpriteRendererProperties]
  renderers [{rendererClass=&NiagaraSpriteRendererProperties}]
 }
 EnergyWave {
  bEnabled=false
  emitterUpdateScript [
   &EmitterState(Life Cycle Mode=::NewEnumerator1, Loop Behavior=::NewEnumerator1, Loop Duration=1, UseLoopDelay=true)
   &SpawnBurst_Instantaneous(Spawn Count=1, Use Spawn Probability=false)
  ]
  particleSpawnScript [
   &InitializeParticle(
    Color Mode=::NewEnumerator0
    Lifetime=0.4
    Mesh Scale Mode=::NewEnumerator3
    Mesh Uniform Scale=2.75
    UsePositionOffset=false
    Write Material Random=true
   )
   &InitialMeshOrientation(Orientation Axis=<0 0 1>, Orientation Vector=<0 0 1>, Rotation=<0 0.25 0>, Use Orientation Vector=true, Use Rotation Vector=true)
   &InitialMeshOrientation(
    Orientation Axis=<0 0 0>
    Orientation Vector=<0 0 0>
    Rotation=&RandomRangeVector(Minimum=<0 0 0>, Maximum=<1 0 0>)
    Use Orientation Vector=false
    Use Rotation Vector=true
   ) {module=InitialMeshOrientation001}
   &CameraOffset(Camera Offset Amount=30)
   &ShapeLocation(MaskSpawnGroup=false, Offset=<0 0 25>, Offset Mode=::NewEnumerator0, Override Seed=false, Sphere Radius=1, UseOwnerScale=true)
   /SetVariables(Particles.LightPosOffset:NiagaraPosition=<0 0 50>) {isSetParameters=true, module=SetVariables_#5}
  ]
  particleUpdateScript [
   &ParticleState()
   &ScaleColor(
    Linear Color Curve=DataInterface() {propertyValues {
      AlphaCurve {keys [<at 0 1>,<at 0.3576 1>,<at 1 0>]}
      BlueCurve {keys [<at 0 0>,<at 0.1684 0>,<at 0.3654 0>,<at 1 0>]}
      GreenCurve {keys [<at 0 0.4051>,<at 0.1684 0.325>,<at 0.3654 0.2679>,<at 1 0.1355>]}
      RedCurve {keys [<at 0 1>,<at 0.1684 1>,<at 0.3654 1>,<at 1 1>]}
     }}
    Scale Mode=::NewEnumerator2
   )
   &DynamicMaterialParameters(
    Index 0 Param 1=&FloatFromCurve(FloatCurve=DataInterface() {propertyValues {Curve {keys [{arriveTangent=-2.1015, interpMode=RCIM_Cubic, leaveTangent=-2.1015, tangentMode=RCTM_User, time=0, value=0.5}, <at 1 -0.15>]}}})
    Index 0 Param 2=&FloatFromCurve(FloatCurve=DataInterface() {propertyValues {Curve {keys [{arriveTangent=-0.3464, interpMode=RCIM_Cubic, leaveTangent=-0.3464, tangentMode=RCTM_User, time=0, value=1}, <at 1 0.9>]}}})
    Index 0 Param 3=&FloatFromCurve(FloatCurve=DataInterface() {propertyValues {Curve {keys [
         {interpMode=RCIM_Cubic, tangentMode=RCTM_Break, time=0, value=0}
         {arriveTangent=0.4478, interpMode=RCIM_Cubic, leaveTangent=0.4478, tangentMode=RCTM_User, time=0.15, value=1}
         {interpMode=RCIM_Cubic, tangentMode=RCTM_Break, time=1, value=1.2}
        ]}}})
    Module.Param0WriteEnabled=true
    Module.Param1WriteEnabled=true
    Module.Param2WriteEnabled=true
    Module.Param3WriteEnabled=false
   )
   &ScaleColor(
    Scale RGB=&VectorFromFloat(Value=&FloatFromCurve(FloatCurve=DataInterface() {propertyValues {Curve {keys [
          {interpMode=RCIM_Cubic, tangentMode=RCTM_Break, time=0.1, value=10}
          {arriveTangent=-6.6072, interpMode=RCIM_Cubic, leaveTangent=-6.6072, tangentMode=RCTM_User, time=0.25, value=3}
          {interpMode=RCIM_Cubic, tangentMode=RCTM_Break, time=1, value=1}
         ]}}}))
    ScaleA=true
    ScaleRGB=true
   ) {module=ScaleColor001}
   &ScaleMeshSize(Scale Factor=&VectorFromFloat(Value=&FloatFromCurve(FloatCurve=DataInterface() {propertyValues {Curve {keys [
          {interpMode=RCIM_Cubic, tangentMode=RCTM_Break, time=0, value=0.2}
          {arriveTangent=2.9365, interpMode=RCIM_Cubic, leaveTangent=2.9365, tangentMode=RCTM_User, time=0.1, value=1}
          {interpMode=RCIM_Cubic, tangentMode=RCTM_Break, time=1, value=2}
         ]}}})))
   &FN_Light_Attributes(
    Color=<rgba 1 0.1939 0 1>
    Create Color Attribute=true
    Create Exponent Attribute=true
    Create Position Attribute=true
    Create Radius Attribute=true
    Exponent=15
    Position=&Add_Vector(A=Linked() {linkedVariable {name=Particles.Initial.Position, type=NiagaraPosition}}, B=<0 0 25>)
    Radius=1000
    Scale Color=&FloatFromCurve(
     FloatCurve=DataInterface() {propertyValues {Curve {keys [
         {interpMode=RCIM_Cubic, leaveTangentWeight=0.9776, tangentMode=RCTM_Break, time=0.1, value=1}
         {arriveTangent=-2.4571, interpMode=RCIM_Cubic, tangentMode=RCTM_Break, time=0.75, value=0}
        ]}}}
     Scale Curve=10
    )
   )
   /SetVariables(Particles.LightPos:NiagaraPosition=&Add_Vector(
     A=Linked() {linkedVariable {name=Particles.Position, type=NiagaraPosition}}
     B=Linked() {linkedVariable {name=Particles.LightPosOffset, type=NiagaraPosition}}
    )) {isSetParameters=true, module=SetVariables_#6}
  ]
  rendererClasses [&NiagaraMeshRendererProperties, &NiagaraLightRendererProperties]
  renderers [
   {rendererClass=&NiagaraMeshRendererProperties, rendererIndex=0}
   {rendererClass=&NiagaraLightRendererProperties, rendererIndex=1}
  ]
 }
 Fireworks_Mesh_AndDistortion {
  emitterUpdateScript [
   &EmitterState(Life Cycle Mode=::NewEnumerator1, Loop Behavior=::NewEnumerator1, Loop Duration=0.25, UseLoopDelay=true)
   &SpawnBurst_Instantaneous(Spawn Count=1, Use Spawn Probability=false)
  ]
  particleSpawnScript [
   /Niagara/Modules/Spawn/Initialization/InitializeParticle(
    Color=&UniformRangedLinearColor(Minimum=<rgba 1 1 1 0.7>, Maximum=<rgba 1 0.5 0 0.7>)
    Lifetime=&UniformRangedFloat(Minimum=0.75, Maximum=1)
    Mesh Scale=&Multiply_VectorByFloat(Vector=<0.3 0.1 0.2>, Float=1)
    Module.Write Color=true
    Module.Write Lifetime=true
    Module.Write Mass=true
    Module.Write Material Random=true
    Module.Write Position=true
    Module.Write Scale=true
    Module.Write SpriteRotation=false
    Module.Write SpriteSize=false
    Position=&SimulationPosition()
    Sprite Size=<25 25>
   )
   &InitialMeshOrientation(Orientation Vector=$System.BlastDir, Rotation Coordinate Space=::NewEnumerator2, Use Orientation Vector=true, Use Rotation Vector=false) {module=InitialMeshOrientation001}
   /SetVariables(Particles.DistortionSize:Vector3f=&VectorFromFloat(Value=5)) {isSetParameters=true}
   /SetVariables(Particles.Color2=<rgba 5 2.6921 0.6188 0.5>) {isSetParameters=true}
   /SetVariables(Particles.ShowDistortion=true) {isSetParameters=true}
  ]
  particleUpdateScript [
   &ParticleState()
   &ScaleColor(
    Scale Alpha=&FloatFromCurve(FloatCurve=DataInterface() {propertyValues {Curve {keys [{arriveTangent=-3.2421, interpMode=RCIM_Cubic, leaveTangent=-3.2421, tangentMode=RCTM_User, time=0.1015, value=1.0322}, <at 1 0>]}}})
    Scale RGB=&VectorFromCurve(
     VectorCurve=DataInterface() {propertyValues {
       XCurve {keys [<at -0.0011 0.01>,<at 0.2555 0>,<at 0.4988 0>,<at 1 0.5>]}
       YCurve {keys [<at -0.0011 15>,<at 0.2492 9.0413>,<at 0.4988 5>,<at 1 1.649>]}
       ZCurve {keys [<at -0.0011 1.0067>,<at 0.2555 1>,<at 0.4988 1>,<at 0.9989 1>]}
      }}
     Scale Curve=<0.7 0.7 0.7>
    )
    ScaleA=true
    ScaleRGB=true
   )
   &ScaleMeshSize(Scale Factor=&VectorFromFloat(Value=&FloatFromCurve(FloatCurve=DataInterface() {propertyValues {Curve {keys [
          {arriveTangent=9.455, interpMode=RCIM_Cubic, leaveTangent=9.455, tangentMode=RCTM_User, time=0, value=1}
          {arriveTangent=1.6472, interpMode=RCIM_Cubic, leaveTangent=1.6472, tangentMode=RCTM_User, time=0.2427, value=2.1701}
          <at 1.0062 2.5692>
         ]}}})))
   &DynamicMaterialParameters(
    Index 0 Param 1=1
    Index 0 Param 2=1
    Index 0 Param 3=&FloatFromCurve(FloatCurve=DataInterface() {propertyValues {Curve {keys [
         {arriveTangent=2.6532, interpMode=RCIM_Cubic, leaveTangent=2.6532, tangentMode=RCTM_User, time=0.0008, value=0.1517}
         {arriveTangent=0.491, interpMode=RCIM_Cubic, leaveTangent=0.491, tangentMode=RCTM_User, time=1, value=1.9}
        ]}}})
    Index 0 Param 4=&FloatFromCurve(FloatCurve=DataInterface() {propertyValues {Curve {keys [<at 0.0842 1.8976>,<at 1 1.2>]}}})
    Module.Param0WriteEnabled=true
    Module.Param1WriteEnabled=true
    Module.Param2WriteEnabled=true
    Module.Param3WriteEnabled=true
   )
   /SetVariables(Particles.DistortionParams:Vector4f=&MakeVector4(
     X=&FloatFromCurve(FloatCurve=DataInterface() {propertyValues {Curve {keys [<at 0 1>,<at 0.3 0.96>,<at 1 1>]}}})
     Y=&FloatFromCurve(FloatCurve=DataInterface() {propertyValues {Curve {keys [<at 0 15>,<at 1 0>]}}})
     Z=0
     W=0
    )) {isSetParameters=true}
   /SetVariables(Particles.ShowDistortion:NiagaraBool=&InvertBool(Boolean:NiagaraBool=&SetBoolByFloatComparison(A=$Particles.Age, B=0.2))) {isSetParameters=true}
   /SetVariables(Particles.DistortionSize:Vector3f=&Multiply_VectorByFloat(
     Vector=/Niagara/DynamicInputs/ScaleByCurve/ScaleVectorByCurve(
      InitialValue:Vector3f=$Particles.Initial.DistortionSize
      ScaleCurve:NiagaraDataInterfaceVectorCurve=DataInterface() {propertyValues {
        XCurve {keys [
          {arriveTangent=4.2859, interpMode=RCIM_Cubic, leaveTangent=4.2859, tangentMode=RCTM_User, time=0, value=0}
          {arriveTangent=2, interpMode=RCIM_Cubic, leaveTangent=2, tangentMode=RCTM_User, time=0.25, value=1}
          {interpMode=RCIM_Cubic, tangentMode=RCTM_User, time=1, value=2}
         ]}
        YCurve {keys [
          {arriveTangent=5.8758, interpMode=RCIM_Cubic, leaveTangent=5.8759, tangentMode=RCTM_User, time=0, value=0}
          {arriveTangent=2, interpMode=RCIM_Cubic, leaveTangent=2, tangentMode=RCTM_User, time=0.25, value=1}
          {interpMode=RCIM_Cubic, tangentMode=RCTM_User, time=1, value=2}
         ]}
        ZCurve {keys [
          {arriveTangent=4.7007, interpMode=RCIM_Cubic, leaveTangent=4.7007, tangentMode=RCTM_User, time=0, value=0}
          {arriveTangent=2, interpMode=RCIM_Cubic, leaveTangent=2, tangentMode=RCTM_User, time=0.25, value=1}
          {interpMode=RCIM_Cubic, tangentMode=RCTM_User, time=1, value=2}
         ]}
       }}
      CurveIndex:NiagaraFloat=&Multiply_Float(A=$Particles.NormalizedAge, B=4)
     )
     Float=&MakeCustomFloatFromBool(Bool=$Particles.ShowDistortion, True Float=1, False Float=0)
    )) {isSetParameters=true}
   /SetVariables(Particles.Color2=&MultiplyLinearColor(
     Linear Color=$Particles.Initial.Color2
     Scale Factor=&MakeVector4_FromLinearColor(LinearColor=&MakeLinearColorFromVectorAndFloat(
       "Vector (RGB)"=&VectorFromCurve(
        VectorCurve=DataInterface() {propertyValues {XCurve {keys [<at -0.0011 0.25>,<at 0.5 0>]}, YCurve {keys [<at -0.0011 0.25>,<at 0.5 0>]}, ZCurve {keys [<at -0.0011 0.25>,<at 0.5 0>]}}}
        CurveIndex=&Multiply_Float(A=$Particles.NormalizedAge, B=2)
       )
       "Float (Alpha)"=&FloatFromCurve(
        FloatCurve=DataInterface() {propertyValues {Curve {keys [
            {interpMode=RCIM_Cubic, tangentMode=RCTM_User, time=0.0045, value=0.9397}
            {arriveTangent=-1.0036, interpMode=RCIM_Cubic, leaveTangent=-1.0036, tangentMode=RCTM_User, time=0.4647, value=0.5233}
            {arriveTangent=-2.4196, interpMode=RCIM_Cubic, leaveTangent=-2.4196, tangentMode=RCTM_User, time=1.0011, value=0.002}
           ]}}}
        CurveIndex=&Multiply_Float(A=$Particles.NormalizedAge, B=2)
       )
      ))
    )) {isSetParameters=true}
  ]
  rendererClasses [&NiagaraMeshRendererProperties]
  renderers [
   {rendererClass=&NiagaraMeshRendererProperties, rendererIndex=0}
   {rendererClass=&NiagaraMeshRendererProperties, rendererIndex=1}
  ]
 }
 FlashWide_EnergyWave_Combined {
  emitterUpdateScript [
   &EmitterState(Life Cycle Mode=::NewEnumerator1, Loop Behavior=::NewEnumerator1, Loop Duration=1.0662, UseLoopDelay=true)
   &SpawnBurst_Instantaneous(Spawn Count=1, Use Spawn Probability=false)
   &SpawnBurst_Instantaneous(Spawn Group=1, Use Spawn Probability=false, UseLoopCountLimit=false) {module=SpawnBurst_Instantaneous001}
   &SpawnBurst_Instantaneous(Spawn Group=2, Use Spawn Probability=false, UseLoopCountLimit=false) {module=SpawnBurst_Instantaneous002}
  ]
  particleSpawnScript [
   /SetVariables(Particles.IsFlash:NiagaraBool=&SetBoolByIntComparison(A=&SpawnGroup(), B=0, Comparison Type=::NewEnumerator0)) {isSetParameters=true}
   /Niagara/Modules/Spawn/Initialization/V2/InitializeParticle(
    Color Mode=::NewEnumerator0
    Lifetime=&MaskFloatBySpawnGroup(Masked Float=0.4, Passthrough Float=0.1)
    Mesh Scale=&MakeCustomVectorFromBool(
     Bool=true
     True Vector=&Multiply_VectorByFloat(
      Vector=&RandomRangeVector(Minimum=<0.012 0.5 2>, Maximum=<0.012 3 2>)
      Float=&Multiply_Float(A=$User.Global Scale, B=1)
     )
     False Vector=&VectorFromFloat(Value=2.75)
    )
    Mesh Scale Mode=::NewEnumerator4
    Position Offset=&MakeCustomVectorFromBool(
     Bool=$Particles.IsFlash
     True Vector=&MakeCustomVectorFromBool(
      Bool=&SetBoolByIntComparison(A=&SpawnGroup(), B=1, Comparison Type=::NewEnumerator4)
      True Vector=$System.BlastSrcA
      False Vector=$System.BlastSrcB
     )
     False Vector=<0 0 0>
    )
    UsePositionOffset=true
    Write Material Random=true
   )
   /SetVariables(Particles.VisibilityTag=&MakeCustomIntFromBool(Bool=$Particles.IsFlash, True INT=1, False INT=0)) {isSetParameters=true}
   &InitialMeshOrientation(Orientation Axis=<0 0 1>, Orientation Vector=<0 0 1>, Rotation=<0 0.25 0>, Use Orientation Vector=true, Use Rotation Vector=true)
   &InitialMeshOrientation(
    Orientation Axis=<0 0 0>
    Orientation Vector=<0 0 0>
    Rotation=&RandomRangeVector(Minimum=<0 0 0>, Maximum=<1 0 0>)
    Use Orientation Vector=false
    Use Rotation Vector=true
   ) {module=InitialMeshOrientation001}
   /SetVariables() {isSetParameters=true}
   &CameraOffset(Camera Offset Amount=30)
   /SetVariables(Particles.Color2=<rgba 8 1.9658 0 0.2>) {isSetParameters=true}
   &ShapeLocation(MaskSpawnGroup=false, Offset=<0 0 25>, Offset Mode=::NewEnumerator0, Override Seed=false, Sphere Radius=1, UseOwnerScale=true) {enabled=false}
   /SetVariables(Particles.LightPosOffset:NiagaraPosition=<0 0 50>) {isSetParameters=true, module=SetVariables_#5}
  ]
  particleUpdateScript [
   &ParticleState()
   &ScaleColor(
    Linear Color Curve=DataInterface() {propertyValues {
      AlphaCurve {keys [<at 0 1>,<at 0.3576 1>,<at 1 0>]}
      BlueCurve {keys [<at 0 0>,<at 0.1684 0>,<at 0.3654 0>,<at 1 0>]}
      GreenCurve {keys [<at 0 0.4051>,<at 0.1684 0.325>,<at 0.3654 0.2679>,<at 1 0.1355>]}
      RedCurve {keys [<at 0 1>,<at 0.1684 1>,<at 0.3654 1>,<at 1 1>]}
     }}
    Scale Mode=::NewEnumerator2
   )
   &DynamicMaterialParameters(
    Index 0 Param 1=&FloatFromCurve(FloatCurve=DataInterface() {propertyValues {Curve {keys [{arriveTangent=-2.1015, interpMode=RCIM_Cubic, leaveTangent=-2.1015, tangentMode=RCTM_User, time=0, value=0.5}, <at 1 -0.15>]}}})
    Index 0 Param 2=&FloatFromCurve(FloatCurve=DataInterface() {propertyValues {Curve {keys [{arriveTangent=-0.3464, interpMode=RCIM_Cubic, leaveTangent=-0.3464, tangentMode=RCTM_User, time=0, value=1}, <at 1 0.9>]}}})
    Index 0 Param 3=&FloatFromCurve(FloatCurve=DataInterface() {propertyValues {Curve {keys [
         {interpMode=RCIM_Cubic, tangentMode=RCTM_Break, time=0, value=0}
         {arriveTangent=0.4478, interpMode=RCIM_Cubic, leaveTangent=0.4478, tangentMode=RCTM_User, time=0.15, value=1}
         {interpMode=RCIM_Cubic, tangentMode=RCTM_Break, time=1, value=1.2}
        ]}}})
    Module.Param0WriteEnabled=true
    Module.Param1WriteEnabled=true
    Module.Param2WriteEnabled=true
    Module.Param3WriteEnabled=false
   )
   &ScaleColor(
    Scale RGB=&VectorFromFloat(Value=&FloatFromCurve(FloatCurve=DataInterface() {propertyValues {Curve {keys [
          {interpMode=RCIM_Cubic, tangentMode=RCTM_Break, time=0.1, value=10}
          {arriveTangent=-6.6072, interpMode=RCIM_Cubic, leaveTangent=-6.6072, tangentMode=RCTM_User, time=0.25, value=3}
          {interpMode=RCIM_Cubic, tangentMode=RCTM_Break, time=1, value=1}
         ]}}}))
    ScaleA=true
    ScaleRGB=true
   ) {module=ScaleColor001}
   &ScaleMeshSize(Scale Factor=&VectorFromFloat(Value=&MakeCustomFloatFromBool(
      Bool=$Particles.IsFlash
      True Float=&FloatFromCurve(
       FloatCurve=DataInterface() {propertyValues {Curve {keys [
           {arriveTangent=9.901, interpMode=RCIM_Cubic, leaveTangent=9.901, tangentMode=RCTM_User, time=0, value=0.4}
           {arriveTangent=12.2238, interpMode=RCIM_Cubic, leaveTangent=12.2238, tangentMode=RCTM_User, time=1, value=6}
          ]}}}
       Scale Curve=1.2
      )
      False Float=&FloatFromCurve(FloatCurve=DataInterface() {propertyValues {Curve {keys [
           {interpMode=RCIM_Cubic, tangentMode=RCTM_Break, time=0, value=0.2}
           {arriveTangent=2.9365, interpMode=RCIM_Cubic, leaveTangent=2.9365, tangentMode=RCTM_User, time=0.1, value=1}
           {interpMode=RCIM_Cubic, tangentMode=RCTM_Break, time=1, value=2}
          ]}}})
     )))
   &FN_Light_Attributes(
    Color=<rgba 1 0.1939 0 1>
    Create Color Attribute=true
    Create Exponent Attribute=true
    Create Position Attribute=true
    Create Radius Attribute=true
    Exponent=15
    Position=&Add_Vector(A=Linked() {linkedVariable {name=Particles.Initial.Position, type=NiagaraPosition}}, B=<0 0 25>)
    Radius=1000
    Scale Color=&FloatFromCurve(
     FloatCurve=DataInterface() {propertyValues {Curve {keys [
         {interpMode=RCIM_Cubic, leaveTangentWeight=0.9776, tangentMode=RCTM_Break, time=0.1, value=1}
         {arriveTangent=-2.4571, interpMode=RCIM_Cubic, tangentMode=RCTM_Break, time=0.75, value=0}
        ]}}}
     Scale Curve=10
    )
   )
   /SetVariables(Particles.LightPos:NiagaraPosition=&Add_Vector(
     A=Linked() {linkedVariable {name=Particles.Position, type=NiagaraPosition}}
     B=Linked() {linkedVariable {name=Particles.LightPosOffset, type=NiagaraPosition}}
    )) {isSetParameters=true, module=SetVariables_#6}
   /SetVariables(Particles.Color2=&MultiplyLinearColor(
     Linear Color=$Particles.Initial.Color2
     Scale Factor=&MakeFloat4FromVectorAndFloat(
      "Vector (RGB)"=<1 1 1>
      "Float (Alpha)"=&FloatFromCurve(FloatCurve=DataInterface() {propertyValues {Curve {keys [<at 0.7 1 RCIM_Cubic>,<at 1 0>]}}})
     )
    )) {isSetParameters=true}
  ]
  rendererClasses [&NiagaraMeshRendererProperties, &NiagaraLightRendererProperties]
  renderers [
   {rendererClass=&NiagaraMeshRendererProperties, rendererIndex=0}
   {rendererClass=&NiagaraLightRendererProperties, rendererIndex=1}
   {rendererClass=&NiagaraMeshRendererProperties, rendererIndex=2}
  ]
 }
 Flash_wide {
  bEnabled=false
  emitterUpdateScript [
   &EmitterState(Life Cycle Mode=::NewEnumerator1, Loop Behavior=::NewEnumerator1, Loop Delay=0.0753, Loop Duration=0.1, UseLoopDelay=true)
   &SpawnBurst_Instantaneous(Spawn Count=1, Use Spawn Probability=false)
   &SpawnBurst_Instantaneous(Spawn Count=1, Spawn Group=1, Spawn Time=0.01, Use Spawn Probability=false) {module=SpawnBurst_Instantaneous001}
  ]
  particleSpawnScript [
   /Niagara/Modules/Spawn/Initialization/InitializeParticle(
    Color=<rgba 8 1.9658 0 0.2>
    Lifetime=0.1
    Mesh Scale=&RandomRangeVector(Minimum=<0.012 0.5 2>, Maximum=<0.012 3 2>)
    Module.Write Color=true
    Module.Write Lifetime=true
    Module.Write Mass=false
    Module.Write Material Random=true
    Module.Write Position=false
    Module.Write Scale=true
    Module.Write SpriteRotation=false
    Module.Write SpriteSize=false
    Position=&SimulationPosition()
    Sprite Size=<5 5>
   ) {enabled=false}
   /Niagara/Modules/Spawn/Initialization/V2/InitializeParticle(
    Color=<rgba 8 1.9658 0 0.2>
    Lifetime=0.1
    Mesh Scale=&RandomRangeVector(Minimum=<0.012 0.5 2>, Maximum=<0.012 3 2>)
    Mesh Scale Mode=::NewEnumerator4
    Position Offset=&MakeCustomVectorFromBool(
     Bool=&SetBoolByIntComparison(A=&SpawnGroup(), B=0, Comparison Type=::NewEnumerator4)
     True Vector=$System.BlastSrcA
     False Vector=$System.BlastSrcB
    )
    UsePositionOffset=true
    Write Material Random=true
   ) {module=InitializeParticle001}
   &CameraOffset(Camera Offset Amount=&Multiply_Float(A=50, B=$User.Global Scale))
   /SetVariables(Particles.Scale=&Multiply_VectorByFloat(Vector=$Particles.Scale, Float=&Multiply_Float(A=$User.Global Scale, B=1))) {isSetParameters=true}
  ]
  particleUpdateScript [
   &ParticleState()
   &SolveForcesAndVelocity(Clamp Velocity=false, Limit Acceleration=false)
   &ScaleMeshSize(Scale Factor=&VectorFromFloat(Value=&FloatFromCurve(
      FloatCurve=DataInterface() {propertyValues {Curve {keys [
          {arriveTangent=9.901, interpMode=RCIM_Cubic, leaveTangent=9.901, tangentMode=RCTM_User, time=0, value=0.4}
          {arriveTangent=12.2238, interpMode=RCIM_Cubic, leaveTangent=12.2238, tangentMode=RCTM_User, time=1, value=6}
         ]}}}
      Scale Curve=1.2
     )))
   &ScaleColor(
    Scale Alpha=&FloatFromCurve(FloatCurve=DataInterface() {propertyValues {Curve {keys [<at 0.7 1 RCIM_Cubic>,<at 1 0>]}}})
    ScaleA=true
    ScaleRGB=false
   )
  ]
  rendererClasses [&NiagaraMeshRendererProperties]
  renderers [{rendererClass=&NiagaraMeshRendererProperties, rendererIndex=0}]
 }
 Grid2D_Gas_SmokeFire {
  simTarget=GPUComputeSim
  emitterSpawnScript [
   /SetVariables(Emitter.ResX=512, Emitter.ResY=512) {isSetParameters=true}
   /SetVariables(Emitter.SimGrid=DataInterface() {propertyValues {
      ClearBeforeNonIterationStage=true
      NumAttributes=0
      NumCellsMaxAxis=3
      NumCellsX=1
      NumCellsY=1
      OverrideBufferFormat=Float
      PreviewAttribute=None
      SetGridFromMaxAxis=false
      bOverrideFormat=false
      bPreviewGrid=false
      RenderTargetUserParameter {parameter {name=None, typeDefHandle {registeredTypeIndex=1}, varData []}}
      WorldBBoxSize=<1200 1200>
     }}) {isSetParameters=true}
   /SetVariables(Emitter.LightingGrid=DataInterface() {propertyValues {
      ClearBeforeNonIterationStage=true
      NumAttributes=3
      NumCellsMaxAxis=3
      NumCellsX=512
      NumCellsY=512
      OverrideBufferFormat=Float
      PreviewAttribute=None
      SetGridFromMaxAxis=false
      bOverrideFormat=false
      bPreviewGrid=false
      RenderTargetUserParameter {parameter {name=None, typeDefHandle {registeredTypeIndex=4}, varData []}}
      WorldBBoxSize=<100 100>
     }}) {isSetParameters=true}
   /SetVariables(Emitter.LightingGridDownsample=DataInterface() {propertyValues {
      ClearBeforeNonIterationStage=true
      NumAttributes=1
      NumCellsMaxAxis=3
      NumCellsX=64
      NumCellsY=64
      OverrideBufferFormat=Float
      PreviewAttribute=None
      SetGridFromMaxAxis=false
      bOverrideFormat=false
      bPreviewGrid=false
      RenderTargetUserParameter {parameter {name=None, typeDefHandle {registeredTypeIndex=4}, varData []}}
      WorldBBoxSize=<100 100>
     }}) {isSetParameters=true}
   /SetVariables() {isSetParameters=true}
   /NiagaraFluids/Modules/Grid2D/Grid2D_SetResolution(
    Grid:NiagaraDataInterfaceGrid2DCollection=$Emitter.SimGrid
    NumCellsX=300
    NumCellsY=300
    WorldGridExtents:Vector2f=&Vector2DFromFloat(Value=&Multiply_Float(A=1280, B=$User.Grid2DScale))
   )
   &Grid2D_SetResolution(
    Grid:NiagaraDataInterfaceGrid2DCollection=$Emitter.PressureGrid
    OtherGrid:NiagaraDataInterfaceGrid2DCollection=$Emitter.SimGrid
    Resolution Method:ENiagaraGrid2DResolution=::NewEnumerator3
    WorldGridExtents:Vector2f=$Emitter.Grid2D_SetResolution.WorldGridExtents
   ) {module=Grid2D_SetResolution001}
   &Grid2D_SetResolution(
    Grid:NiagaraDataInterfaceGrid2DCollection=$Emitter.LightingGridDownsample
    OtherGrid:NiagaraDataInterfaceGrid2DCollection=$Emitter.SimGrid
    Resolution Method:ENiagaraGrid2DResolution=::NewEnumerator3
    Resolution Mult=0.2
    WorldGridExtents:Vector2f=$Emitter.Grid2D_SetResolution.WorldGridExtents
   ) {module=Grid2D_SetResolution002}
   &Grid2D_SetResolution(
    Grid:NiagaraDataInterfaceGrid2DCollection=$Emitter.LightingGrid
    OtherGrid:NiagaraDataInterfaceGrid2DCollection=$Emitter.SimGrid
    Resolution Method:ENiagaraGrid2DResolution=::NewEnumerator3
    WorldGridExtents:Vector2f=$Emitter.Grid2D_SetResolution.WorldGridExtents
   ) {module=Grid2D_SetResolution003}
   /SetVariables(Emitter.dt=0.0167) {isSetParameters=true}
   /SetVariables(Emitter.g:Vector2f=<0 -9.8>) {isSetParameters=true}
   /SetVariables() {isSetParameters=true}
   /SetVariables(Emitter.density=1) {isSetParameters=true}
   /SetVariables(Emitter.ParticleSource=DataInterface() {propertyValues {EmitterBinding {bindingMode=Other, emitterName=SourceParticle}}}) {isSetParameters=true}
   /SetVariables() {isSetParameters=true}
   &self:SetLightingRT_04(LightingGrid:NiagaraDataInterfaceGrid2DCollection=$Emitter.LightingGrid, LightingRT:NiagaraDataInterfaceRenderTarget2D=$Emitter.LightingRT) {module=SetLightingRT_04}
   /NiagaraFluids/Modules/Grid2D/Grid2D_UpdateSimAttributes(Grid:NiagaraDataInterfaceGrid2DCollection=$Emitter.SimGrid, WorldGridSize:Vector2f=$Emitter.Grid2D_SetResolution.WorldGridExtents) {module=UpdateSimAttributes}
   /SetVariables() {isSetParameters=true}
   /SetVariables(Emitter.testCurve=DataInterface() {propertyValues {
      bExposeCurve=true
      AlphaCurve {keys []}
      BlueCurve {keys [<at 0.3032 1>,<at 0.6671 0>]}
      GreenCurve {keys [<at 0.3032 1>,<at 0.6671 0.0785>]}
      RedCurve {keys [<at 0.3032 1>,<at 0.6671 1>]}
     }}) {isSetParameters=true}
   /NiagaraFluids/Functions/General/Fluids_Gas_Manager(
    Blackbody=true
    Density Bouyancy=-300
    Density Color Curve=DataInterface() {propertyValues {
      ExposedName=Curve Density
      bExposeCurve=true
      AlphaCurve {keys [<at 0 0 RCIM_Cubic>,<at 2.7618 1>]}
      BlueCurve {keys [<at 0.0245 0.0863>,<at 4.4467 0.0863>]}
      GreenCurve {keys [<at 0.0245 0.0588>,<at 4.4467 0.0588>]}
      RedCurve {keys [<at 0.0245 0.051>,<at 4.4467 0.051>]}
     }}
    Depth Fade=80
    Dissipation Density=0.4
    Dissipation Temperature=0.35
    Dissipation Velocity=0.5
    EnvironmentColor:LinearColor=<rgba 0.1427 0.2031 0.3542 1>
    Pressure Relaxation=1
    Radius Scale=0.8
    Shadowing=0.25
    SunColor:LinearColor=<rgba 1.2417 1.53 2 1>
    Temperature Color Curve=DataInterface() {propertyValues {
      ExposedName=Curve Temperature
      bExposeCurve=true
      AlphaCurve {keys [<at 0 0>,<at 0.4853 1>]}
      BlueCurve {keys [<at 0.0696 0>,<at 0.8304 0.0517>,<at 0.9917 0.0783>]}
      GreenCurve {keys [<at 0.0696 0>,<at 0.8304 0.729>,<at 0.9917 0.2349>]}
      RedCurve {keys [<at 0.0696 0>,<at 0.8304 8>,<at 0.9917 0.9401>]}
     }}
    Temperature Maximum=1245454
    Temperature Scale=2
    Vorticity Confinement=0.1
   )
   /NiagaraFluids/Modules/Grid2D/Grid2D_InitRT(Grid:NiagaraDataInterfaceGrid2DCollection=$Emitter.SimGrid, SimRT:NiagaraDataInterfaceRenderTarget2D=$Emitter.SimRT) {module=Grid2D_Gas_InitRT}
   &self:NiagaraEmitter_0.NiagaraScratchPadContainer_0.ScratchModule_0(Grid:NiagaraDataInterfaceGrid2DCollection=$Emitter.SimGrid, PressureGrid:NiagaraDataInterfaceGrid2DCollection=$Emitter.PressureGrid) {module=ScratchModule}
   /SetVariables(Emitter.GridsAllocated=true) {isSetParameters=true}
  ]
  emitterUpdateScript [
   &self:Emitter_Limitation_and_Culling__01(Max Num=3, Platform Set=DataInterface() {propertyValues {Platforms {qualityLevelMask=-17, cVarConditions [], deviceProfileStates []}}}) {module=Emitter_Limitation_and_Culling__01}
   &EmitterState(
    Enable Distance Culling=true
    Life Cycle Mode=::NewEnumerator1
    Loop Behavior=::NewEnumerator1
    Loop Duration=3
    MaxDistance=8000
    Module.bUseMaxDistance=true
    Module.bUseMinDistance=false
    Scalability Mode=::NewEnumerator1
    UseLoopDelay=false
   )
   /SetVariables(Emitter.dt:NiagaraFloat=$Engine.DeltaTime) {isSetParameters=true}
   /NiagaraFluids/Modules/Grid2D/Grid2D_CreateUnitToWorldTransform(
    Camera:NiagaraDataInterfaceCamera=$Emitter.CameraQuery
    Camera Facing=true
    Offset:NiagaraPosition=$Engine.Owner.Position
    Rotation:Quat4f=$Engine.Owner.Rotation
    WorldGridExtents:Vector2f=$Emitter.Grid2D_SetResolution.WorldGridExtents
   )
   /SetVariables(Emitter.Position:NiagaraPosition=&ConvertVectorToPosition(
     Input Position=$Emitter.Grid2D_CreateUnitToWorldTransform.LocalToWorldTranslation
     Conversion Operation:ENiagara_LWCConvertVecToPos=::NewEnumerator0
    )) {isSetParameters=true}
   /NiagaraFluids/Modules/Grid2D/Grid2D_SetBounds(
    CameraFacing:NiagaraBool=$Emitter.Grid2D_CreateUnitToWorldTransform.CameraFacing
    EmitterProperties=DataInterface() {propertyValues {EmitterBinding {bindingMode=Self, emitterName=None}}}
    LocalPivot:Vector2f=$Emitter.Grid2D_CreateUnitToWorldTransform.LocalPivot
    Rotate for Camera Facing=true
    WorldGridExtents:Vector2f=$Emitter.Grid2D_SetResolution.WorldGridExtents
   )
   /SetVariables(Emitter.ExecutionState:ENiagaraExecutionState=$Emitter.ENiagaraExecutionState) {isSetParameters=true}
   &self:ScratchModule_02(PerformDeallocation:NiagaraBool=/Niagara/DynamicInputs/Scalability/IsEmitterComplete()) {module=ScratchModule_02002}
  ]
  rendererClasses [&NiagaraSpriteRendererProperties]
  renderers [
   {rendererClass=&NiagaraSpriteRendererProperties}
   {rendererClass=&NiagaraSpriteRendererProperties, rendererIndex=1}
  ]
 }
 LingeringEmbers {
  simTarget=GPUComputeSim
  emitterUpdateScript [
   &EmitterState(Life Cycle Mode=::NewEnumerator1, Loop Behavior=::NewEnumerator1, Loop Duration=0.075, UseLoopDelay=true)
   &SpawnBurst_Instantaneous(Spawn Count=25, Use Spawn Probability=false, UseLoopCountLimit=false)
  ]
  particleSpawnScript [
   &InitializeParticle(
    Color=$User.Global Color
    Lifetime Max=2.25
    Lifetime Min=0.5
    Lifetime Mode=::NewEnumerator1
    Sprite Size Max=<4 6>
    Sprite Size Min=<3 3>
    Sprite Size Mode=::NewEnumerator2
    UsePositionOffset=false
    Write Material Random=true
   )
   &SphereLocation(
    MaskSpawnGroup=false
    Module.Override Randomness=false
    Sphere Radius=&Multiply_Float(A=0, B=$User.Global Scale)
    Surface Only=false
   ) {enabled=false}
   &AddVelocity(
    Cone Axis=$System.BlastDir
    Module.UseConeFalloff=false
    Module.UseVelDistribution=false
    Override Seed=false
    Velocity Mode=::NewEnumerator2
    Velocity Speed=&Multiply_Float(A=&RandomRangeFloat(Minimum=2500, Maximum=6000), B=$User.Global Scale)
   )
   &ApplyInitialForces(Apply Rotational Force to Rotational Velocity=false) {enabled=false}
   /SetVariables(Particles.SpriteSize=&Multiply_Vector2D_ByFloat(Float=$User.Global Scale, Vector2D=$StackContext.SpriteSize)) {isSetParameters=true, module=SetVariables_#4}
  ]
  particleUpdateScript [
   &ParticleState()
   &GravityForce(Gravity=&RandomRangeVector(Minimum=<0 0 -980>, Maximum=<0 0 -980>))
   &Drag(Drag=&RandomRangeFloat(Minimum=2, Maximum=10), Use Linear Drag=true, Use Rotational Drag=false)
   &CurlNoiseForce(Noise Frequency=15, Noise Quality / Cost=::NewEnumerator0, Noise Strength=500, Pan Noise=false)
   &SolveForcesAndVelocity(Clamp Velocity=false, Limit Acceleration=false)
   &ScaleColor(
    Scale Alpha=&FloatFromCurve(FloatCurve=DataInterface() {propertyValues {Curve {keys [
         {arriveTangent=26.7273, interpMode=RCIM_Cubic, leaveTangent=26.7273, tangentMode=RCTM_User, time=-0.0092, value=0.0144}
         {interpMode=RCIM_Cubic, tangentMode=RCTM_User, time=0.1, value=1}
         {interpMode=RCIM_Cubic, tangentMode=RCTM_User, time=0.85, value=1}
         {arriveTangent=-1.3701, interpMode=RCIM_Cubic, leaveTangent=-1.3701, tangentMode=RCTM_User, time=1.0011, value=0}
        ]}}})
    Scale RGB=&Lerp_Vector(
     A=<0 0 0>
     B=<12.5 4 0.8>
     Alpha=&FloatFromCurve(FloatCurve=DataInterface() {propertyValues {Curve {keys [
          {arriveTangent=15.3906, interpMode=RCIM_Cubic, leaveTangent=15.3906, tangentMode=RCTM_User, time=-0.0092, value=1}
          {interpMode=RCIM_Cubic, tangentMode=RCTM_User, time=0.0967, value=0.8}
          {interpMode=RCIM_Cubic, tangentMode=RCTM_User, time=0.7, value=0.8}
          <at 1 0 RCIM_Cubic>
         ]}}})
    )
    ScaleA=true
    ScaleRGB=true
   )
   &ScaleSpriteSize(Scale Factor=&Vector2DFromFloat(Value=&FloatFromCurve(FloatCurve=DataInterface() {propertyValues {Curve {keys [<at -0.0046 0>,<at 0.025 1>,<at 0.66 1>,<at 1.0011 0.5>]}}}, Scale Curve=1.5)))
   &ScaleSpriteSizeBySpeed(Max Scale Factor=<0.5 5>, Min Scale Factor=<1 0.125>, Velocity Threshold=1250)
  ]
  rendererClasses [&NiagaraSpriteRendererProperties]
  renderers [{rendererClass=&NiagaraSpriteRendererProperties}]
 }
 NE_Explosion_Main002 {
  emitterUpdateScript [
   &EmitterState(Life Cycle Mode=::NewEnumerator1, Loop Behavior=::NewEnumerator1, Loop Delay=0.05, Loop Duration=3, UseLoopDelay=true)
   &SpawnBurst_Instantaneous(Spawn Group=1, Use Spawn Probability=false, UseLoopCountLimit=false) {enabled=false, module=SpawnBurst_Instantaneous002}
   &SpawnBurst_Instantaneous(Spawn Count=9, Use Spawn Probability=false, UseLoopCountLimit=false) {module=SpawnBurst_Instantaneous001}
  ]
  particleSpawnScript [
   /SetVariables(Particles.IsLeftExpl:NiagaraBool=&RandomBool()) {isSetParameters=true}
   &InitializeParticle(
    Color=&MaskLinearColorBySpawnGroup(Masked Linear Color=<rgba 0.2 0.2 0.2 1.1>, Passthrough Linear Color=<rgba 0.2 0.2 0.2 1.1>, Spawn Group Mask=1)
    Lifetime=&MaskFloatBySpawnGroup(
     Masked Float=3
     Passthrough Float=&SelectFloatFromArray(
      Array Sampling Mode=::NewEnumerator0
      Direct Array Index=?
      Float Selection Array=DataInterface() {propertyValues {FloatData [1.9,3]}}
      Random Seed=0
     )
     Spawn Group Mask=1
    )
    Mesh Scale=&Multiply_VectorByFloat(
     Vector=&Multiply_VectorByFloat(
      Vector=&MaskVectorBySpawnGroup(
       Masked Vector=<3 3.5 2.6>
       Passthrough Vector=&RandomRangeVector(Minimum=<3 2.8 2.8>, Maximum=<3 3.5 3.5>)
       Spawn Group Mask=1
      )
      Float=&MaskFloatBySpawnGroup(
       Masked Float=2
       Passthrough Float=&FloatFromCurve(
        FloatCurve=DataInterface() {propertyValues {Curve {keys [<at 0 0.9 RCIM_Constant>,<at 0.5 1 RCIM_Constant>,<at 1 1>]}}}
        CurveIndex=&RandomRangeFloat(Minimum=0, Maximum=1)
        Scale Curve=1.35
       )
       Spawn Group Mask=1
      )
     )
     Float=0.7
    )
    Mesh Scale Mode=::NewEnumerator4
    Position Offset=&MakeCustomVectorFromBool(Bool=$Particles.IsLeftExpl, True Vector=$System.BlastSrcA, False Vector=$System.BlastSrcB)
    Position Offset Coordinate Space=::Local
    UsePositionOffset=true
    Write Material Random=true
   )
   &ShapeLocation(
    MaskSpawnGroup=false
    Override Seed=false
    Sphere Radius=&MaskFloatBySpawnGroup(Masked Float=0, Passthrough Float=50, Spawn Group Mask=1)
    UseOwnerScale=true
   )
   &AddVelocity(
    Module.Use Falloff Distance=false
    Origin Offset=&RandomRangeVector(Minimum=<0 0 0>, Maximum=<0 0 0>)
    Origin Offset Coordinate Space=::World
    Override Seed=false
    Velocity Mode=::NewEnumerator1
    Velocity Speed=&MaskFloatBySpawnGroup(Masked Float=0, Passthrough Float=&RandomRangeFloat(Minimum=1, Maximum=2500), Spawn Group Mask=1)
   )
   &AddVelocity(
    Cone Angle=35
    Cone Axis=&MakeCustomVectorFromBool(Bool=$Particles.IsLeftExpl, True Vector=$System.BlastDir, False Vector=$System.BlastDir2)
    Module.UseConeFalloff=false
    Module.UseVelDistribution=false
    Override Seed=false
    Velocity Mode=::NewEnumerator2
    Velocity Speed=&RandomRangeFloat(Minimum=500, Maximum=1250)
   ) {module=AddVelocity001}
   &CameraOffset(Camera Offset Amount=&MaskFloatBySpawnGroup(Masked Float=70, Passthrough Float=40, Spawn Group Mask=1))
  ]
  particleUpdateScript [
   &ParticleState()
   &GravityForce(
    Coordinate Space=::Local
    Gravity=&MaskVectorBySpawnGroup(
     Masked Vector=<0 0 1600>
     Passthrough Vector=&Multiply_VectorByFloat(
      Vector=&MakeCustomVectorFromBool(Bool=$Particles.IsLeftExpl, True Vector=$System.BlastDir, False Vector=$System.BlastDir2)
      Float=&RandomRangeFloat(Minimum=1200, Maximum=1600)
     )
     Spawn Group Mask=1
    )
   )
   &Drag(Drag=&RandomRangeFloat(Minimum=6, Maximum=20), Use Linear Drag=true, Use Rotational Drag=false)
   &DynamicMaterialParameters(
    Index 0 Param 1=&FloatFromCurve(
     FloatCurve=DataInterface() {propertyValues {Curve {keys [{arriveTangent=-2.2221, interpMode=RCIM_Cubic, leaveTangent=-2.2221, tangentMode=RCTM_User, time=0.1, value=1}, <at 0.95 0>]}}}
     Scale Curve=-80
    )
    Index 0 Param 2=&FloatFromCurve(
     FloatCurve=DataInterface() {propertyValues {Curve {keys [
         {tangentMode=RCTM_User, time=0, value=1}
         {arriveTangent=-0.0405, interpMode=RCIM_Cubic, leaveTangent=-0.0405, tangentMode=RCTM_User, time=0.2, value=1}
         {arriveTangent=-0.0002, interpMode=RCIM_Cubic, leaveTangent=-0.0002, tangentMode=RCTM_User, time=0.9, value=0.6}
        ]}}}
     Scale Curve=0.5
    )
    Index 0 Param 3=&FloatFromCurve(
     FloatCurve=DataInterface() {propertyValues {Curve {keys [
         {arriveTangent=1.4725, interpMode=RCIM_Cubic, leaveTangent=1.4725, tangentMode=RCTM_User, time=0, value=0}
         {tangentMode=RCTM_User, time=1, value=1}
        ]}}}
     Scale Curve=&MaskFloatBySpawnGroup(Masked Float=350, Passthrough Float=&RandomRangeFloat(Minimum=350, Maximum=400), Spawn Group Mask=1)
    )
    Index 0 Param 4=&FloatFromCurve(
     FloatCurve=DataInterface() {propertyValues {Curve {
        preInfinityExtrap=RCCE_Linear
        keys [
         {arriveTangent=0.1175, interpMode=RCIM_Cubic, leaveTangent=16.4114, tangentMode=RCTM_Break, time=0.03, value=-3}
         {arriveTangent=2.6068, interpMode=RCIM_Cubic, leaveTangent=2.6068, tangentMode=RCTM_User, time=0.3881, value=-0.5}
         {arriveTangent=32.7533, interpMode=RCIM_Cubic, leaveTangent=14.4841, tangentMode=RCTM_Break, time=0.6465, value=2.75}
         {interpMode=RCIM_Cubic, tangentMode=RCTM_User, time=0.8319, value=11.25}
        ]
       }}}
     Scale Curve=0.2
    )
    Index 1 Param 1=&FloatFromCurve(
     FloatCurve=DataInterface() {propertyValues {Curve {keys [
         {arriveTangent=0.5348, interpMode=RCIM_Cubic, leaveTangent=0.5348, tangentMode=RCTM_User, time=0, value=0.01}
         {arriveTangent=1.457, interpMode=RCIM_Cubic, leaveTangent=1.457, tangentMode=RCTM_User, time=0.04, value=1}
         {arriveTangent=85.7775, interpMode=RCIM_Cubic, leaveTangent=85.7775, tangentMode=RCTM_User, time=0.08, value=16}
        ]}}}
     Scale Curve=&MaskFloatBySpawnGroup(Masked Float=5, Passthrough Float=&RandomRangeFloat(Minimum=5, Maximum=5), Spawn Group Mask=1)
    )
    Index 1 Param 2=&MaskFloatBySpawnGroup(
     Masked Float=0.014
     Passthrough Float=&RandomRangeFloat(Minimum=0.012, Maximum=0.014)
     Spawn Group Mask=1
    )
    Index 1 Param 3=&FloatFromCurve(
     FloatCurve=DataInterface() {propertyValues {Curve {keys [{interpMode=RCIM_Cubic, tangentMode=RCTM_User, time=0.4, value=0.5}, <at 0.8 0.1>]}}}
     Scale Curve=5
    )
    Index 1 Param 4=&FloatFromCurve(FloatCurve=DataInterface() {propertyValues {Curve {keys [{interpMode=RCIM_Cubic, tangentMode=RCTM_User, time=0.05, value=1}, {interpMode=RCIM_Cubic, tangentMode=RCTM_User, time=1, value=6}]}}})
    Index 2 Param 1=&FloatFromCurve(
     FloatCurve=DataInterface() {propertyValues {Curve {keys [
         {arriveTangent=0.9553, interpMode=RCIM_Cubic, leaveTangent=0.9553, tangentMode=RCTM_User, time=0, value=0}
         {arriveTangent=1.5886, interpMode=RCIM_Cubic, leaveTangent=1.5886, tangentMode=RCTM_User, time=1, value=1.5}
        ]}}}
     Scale Curve=1.2
    )
    Index 2 Param 3=&MaskFloatBySpawnGroup(
     Masked Float=&FloatFromCurve(
      FloatCurve=DataInterface() {propertyValues {Curve {keys [
          {arriveTangent=-17.9351, interpMode=RCIM_Cubic, leaveTangent=-17.9351, tangentMode=RCTM_User, time=0, value=1}
          {arriveTangent=-0.7527, interpMode=RCIM_Cubic, leaveTangent=-0.7527, tangentMode=RCTM_User, time=0.02, value=0}
         ]}}}
      Scale Curve=0
     )
     Passthrough Float=&FloatFromCurve(
      FloatCurve=DataInterface() {propertyValues {Curve {keys [
          {arriveTangent=-36.0295, interpMode=RCIM_Cubic, leaveTangent=-36.0295, tangentMode=RCTM_User, time=0, value=1}
          {arriveTangent=-0.7527, interpMode=RCIM_Cubic, leaveTangent=-0.7527, tangentMode=RCTM_User, time=0.09, value=0}
         ]}}}
      Scale Curve=20
     )
     Spawn Group Mask=1
    )
    Index 2 Param 4=1
    Index 3 Param 1=&FloatFromCurve(
     FloatCurve=DataInterface() {propertyValues {Curve {keys [
         {arriveTangent=2.2353, interpMode=RCIM_Cubic, leaveTangent=2.2353, tangentMode=RCTM_User, time=0, value=0.2}
         {arriveTangent=1.4416, interpMode=RCIM_Cubic, leaveTangent=1.4416, tangentMode=RCTM_User, time=0.2, value=0.5}
         {interpMode=RCIM_Cubic, tangentMode=RCTM_User, time=0.8, value=1.25}
        ]}}}
     Scale Curve=-2.5
    )
    Index 3 Param 2=&FloatFromCurve(
     FloatCurve=DataInterface() {propertyValues {Curve {keys [
         {arriveTangent=5.1563, interpMode=RCIM_Cubic, leaveTangent=5.1563, tangentMode=RCTM_User, time=0, value=0}
         {arriveTangent=0.1309, interpMode=RCIM_Cubic, leaveTangent=0.1309, tangentMode=RCTM_User, time=0.1699, value=0.4}
         {arriveTangent=0.8545, interpMode=RCIM_Cubic, leaveTangent=0.8545, tangentMode=RCTM_User, time=1, value=1}
        ]}}}
     Scale Curve=-0.4
    )
    Index 3 Param 3=&FloatFromCurve(
     FloatCurve=DataInterface() {propertyValues {Curve {keys [
         {interpMode=RCIM_Cubic, tangentMode=RCTM_User, time=0.0014, value=8}
         {arriveTangent=-5.9757, interpMode=RCIM_Cubic, leaveTangent=-5.9757, tangentMode=RCTM_User, time=0.1, value=2.5}
         {arriveTangent=-5.9757, interpMode=RCIM_Cubic, leaveTangent=-5.9757, tangentMode=RCTM_User, time=0.15, value=1.25}
        ]}}}
     Scale Curve=&FNTimeOfDayScalar(Night Scale=0.2, Morning Scale=0.2, Day Scale=0.2, Evening Scale=0.2)
    )
    Index 3 Param 4=&FloatFromCurve(
     FloatCurve=DataInterface() {propertyValues {Curve {keys [{tangentMode=RCTM_User, time=0, value=0}, {tangentMode=RCTM_User, time=1, value=1}]}}}
     Scale Curve=-100
    )
    Module.Param0WriteEnabled=true
    Module.Param0WriteEnabled1=true
    Module.Param0WriteEnabled2=true
    Module.Param0WriteEnabled3=true
    Module.Param1WriteEnabled=true
    Module.Param1WriteEnabled1=true
    Module.Param1WriteEnabled2=true
    Module.Param1WriteEnabled3=true
    Module.Param2WriteEnabled=true
    Module.Param2WriteEnabled1=true
    Module.Param2WriteEnabled2=true
    Module.Param2WriteEnabled3=true
    Module.Param3WriteEnabled=true
    Module.Param3WriteEnabled1=true
    Module.Param3WriteEnabled2=true
    Module.Param3WriteEnabled3=true
    Write Parameter Index 1=true
    Write Parameter Index 2=true
    Write Parameter Index 3=true
   ) {enabled=false}
   &ScaleColor(
    Curve Index=&Lerp_Float(A=$Particles.NormalizedAge, B=$Particles.NormalizedAge, Alpha=0.125)
    Linear Color Curve=DataInterface() {propertyValues {
      AlphaCurve {keys [<at 0 1.2>,<at 0.073 1.26>,<at 0.4927 1>,<at 1 0>]}
      BlueCurve {keys [<at 0.0178 0>,<at 0.041 0.6042>,<at 0.4977 0.4497>]}
      GreenCurve {keys [<at 0.0178 0.0438>,<at 0.041 0.6042>,<at 0.4977 0.4497>]}
      RedCurve {keys [<at 0.0178 1>,<at 0.041 0.6042>,<at 0.4977 0.4497>]}
     }}
    Scale Mode=::NewEnumerator2
   )
   &DynamicMaterialParameters(
    Index 0 Param 1=&FloatFromCurve(
     FloatCurve=DataInterface() {propertyValues {Curve {keys [{arriveTangent=-2.2221, interpMode=RCIM_Cubic, leaveTangent=-2.2221, tangentMode=RCTM_User, time=0.1, value=1}, <at 0.95 0>]}}}
     Scale Curve=-80
    )
    Index 0 Param 2=&FloatFromCurve(
     FloatCurve=DataInterface() {propertyValues {Curve {keys [
         {tangentMode=RCTM_User, time=0, value=1}
         {arriveTangent=-0.0405, interpMode=RCIM_Cubic, leaveTangent=-0.0405, tangentMode=RCTM_User, time=0.2, value=1}
         {arriveTangent=-0.0002, interpMode=RCIM_Cubic, leaveTangent=-0.0002, tangentMode=RCTM_User, time=0.9, value=0.6}
        ]}}}
     Scale Curve=0.5
    )
    Index 0 Param 3=&FloatFromCurve(
     FloatCurve=DataInterface() {propertyValues {Curve {keys [
         {arriveTangent=1.4725, interpMode=RCIM_Cubic, leaveTangent=1.4725, tangentMode=RCTM_User, time=0, value=0}
         {tangentMode=RCTM_User, time=1, value=1}
        ]}}}
     Scale Curve=&MaskFloatBySpawnGroup(Masked Float=350, Passthrough Float=&RandomRangeFloat(Minimum=350, Maximum=400), Spawn Group Mask=1)
    )
    Index 0 Param 4=&FloatFromCurve(
     FloatCurve=DataInterface() {propertyValues {Curve {
        preInfinityExtrap=RCCE_Linear
        keys [
         {arriveTangent=0.1175, interpMode=RCIM_Cubic, leaveTangent=16.4114, tangentMode=RCTM_Break, time=0.03, value=-3}
         {arriveTangent=2.6068, interpMode=RCIM_Cubic, leaveTangent=2.6068, tangentMode=RCTM_User, time=0.3881, value=-0.5}
         {arriveTangent=32.7533, interpMode=RCIM_Cubic, leaveTangent=14.4841, tangentMode=RCTM_Break, time=0.6465, value=2.75}
         {interpMode=RCIM_Cubic, tangentMode=RCTM_User, time=0.8319, value=11.25}
        ]
       }}}
     Scale Curve=0.2
    )
    Index 1 Param 1=&FloatFromCurve(
     FloatCurve=DataInterface() {propertyValues {Curve {keys [
         {arriveTangent=0.5348, interpMode=RCIM_Cubic, leaveTangent=0.5348, tangentMode=RCTM_User, time=0, value=0.01}
         {arriveTangent=1.457, interpMode=RCIM_Cubic, leaveTangent=1.457, tangentMode=RCTM_User, time=0.04, value=1}
         {arriveTangent=85.7775, interpMode=RCIM_Cubic, leaveTangent=85.7775, tangentMode=RCTM_User, time=0.08, value=16}
        ]}}}
     Scale Curve=&MaskFloatBySpawnGroup(Masked Float=5, Passthrough Float=&RandomRangeFloat(Minimum=5, Maximum=5), Spawn Group Mask=1)
    )
    Index 1 Param 2=&MaskFloatBySpawnGroup(
     Masked Float=0.014
     Passthrough Float=&RandomRangeFloat(Minimum=0.012, Maximum=0.014)
     Spawn Group Mask=1
    )
    Index 1 Param 3=&FloatFromCurve(
     FloatCurve=DataInterface() {propertyValues {Curve {keys [{interpMode=RCIM_Cubic, tangentMode=RCTM_User, time=0.4, value=0.5}, <at 0.8 0.1>]}}}
     Scale Curve=5
    )
    Index 1 Param 4=&FloatFromCurve(FloatCurve=DataInterface() {propertyValues {Curve {keys [{interpMode=RCIM_Cubic, tangentMode=RCTM_User, time=0.05, value=1}, {interpMode=RCIM_Cubic, tangentMode=RCTM_User, time=1, value=6}]}}})
    Index 2 Param 1=&FloatFromCurve(
     FloatCurve=DataInterface() {propertyValues {Curve {keys [
         {arriveTangent=0.9553, interpMode=RCIM_Cubic, leaveTangent=0.9553, tangentMode=RCTM_User, time=0, value=0}
         {arriveTangent=1.5886, interpMode=RCIM_Cubic, leaveTangent=1.5886, tangentMode=RCTM_User, time=1, value=1.5}
        ]}}}
     Scale Curve=1.2
    )
    Index 2 Param 3=&MaskFloatBySpawnGroup(
     Masked Float=&FloatFromCurve(
      FloatCurve=DataInterface() {propertyValues {Curve {keys [
          {arriveTangent=-17.9351, interpMode=RCIM_Cubic, leaveTangent=-17.9351, tangentMode=RCTM_User, time=0, value=1}
          {arriveTangent=-0.7527, interpMode=RCIM_Cubic, leaveTangent=-0.7527, tangentMode=RCTM_User, time=0.02, value=0}
         ]}}}
      Scale Curve=100
     )
     Passthrough Float=&FloatFromCurve(
      FloatCurve=DataInterface() {propertyValues {Curve {keys [
          {arriveTangent=-36.0295, interpMode=RCIM_Cubic, leaveTangent=-36.0295, tangentMode=RCTM_User, time=0, value=1}
          {arriveTangent=-0.7527, interpMode=RCIM_Cubic, leaveTangent=-0.7527, tangentMode=RCTM_User, time=0.09, value=0}
         ]}}}
      Scale Curve=20
     )
     Spawn Group Mask=1
    )
    Index 2 Param 4=1
    Index 3 Param 1=&FloatFromCurve(
     FloatCurve=DataInterface() {propertyValues {Curve {keys [
         {arriveTangent=2.2353, interpMode=RCIM_Cubic, leaveTangent=2.2353, tangentMode=RCTM_User, time=0, value=0.2}
         {arriveTangent=1.4416, interpMode=RCIM_Cubic, leaveTangent=1.4416, tangentMode=RCTM_User, time=0.2, value=0.5}
         {interpMode=RCIM_Cubic, tangentMode=RCTM_User, time=0.8, value=1.25}
        ]}}}
     Scale Curve=-2.5
    )
    Index 3 Param 2=&FloatFromCurve(
     FloatCurve=DataInterface() {propertyValues {Curve {keys [
         {arriveTangent=5.1563, interpMode=RCIM_Cubic, leaveTangent=5.1563, tangentMode=RCTM_User, time=0, value=0}
         {arriveTangent=0.1309, interpMode=RCIM_Cubic, leaveTangent=0.1309, tangentMode=RCTM_User, time=0.1699, value=0.4}
         {arriveTangent=0.8545, interpMode=RCIM_Cubic, leaveTangent=0.8545, tangentMode=RCTM_User, time=1, value=1}
        ]}}}
     Scale Curve=-0.4
    )
    Index 3 Param 3=&FloatFromCurve(
     FloatCurve=DataInterface() {propertyValues {Curve {keys [
         {interpMode=RCIM_Cubic, tangentMode=RCTM_User, time=0.0014, value=8}
         {arriveTangent=-5.9757, interpMode=RCIM_Cubic, leaveTangent=-5.9757, tangentMode=RCTM_User, time=0.1, value=2.5}
         {arriveTangent=-5.9757, interpMode=RCIM_Cubic, leaveTangent=-5.9757, tangentMode=RCTM_User, time=0.15, value=1.25}
        ]}}}
     Scale Curve=&FNTimeOfDayScalar(Night Scale=0.6, Morning Scale=0.6, Day Scale=0.6, Evening Scale=0.6)
    )
    Index 3 Param 4=&FloatFromCurve(
     FloatCurve=DataInterface() {propertyValues {Curve {keys [{tangentMode=RCTM_User, time=0, value=0}, {tangentMode=RCTM_User, time=1, value=1}]}}}
     Scale Curve=-100
    )
    Module.Param0WriteEnabled=true
    Module.Param0WriteEnabled1=true
    Module.Param0WriteEnabled2=true
    Module.Param0WriteEnabled3=true
    Module.Param1WriteEnabled=true
    Module.Param1WriteEnabled1=true
    Module.Param1WriteEnabled2=true
    Module.Param1WriteEnabled3=true
    Module.Param2WriteEnabled=true
    Module.Param2WriteEnabled1=true
    Module.Param2WriteEnabled2=true
    Module.Param2WriteEnabled3=true
    Module.Param3WriteEnabled=true
    Module.Param3WriteEnabled1=true
    Module.Param3WriteEnabled2=true
    Module.Param3WriteEnabled3=true
    Write Parameter Index 1=true
    Write Parameter Index 2=true
    Write Parameter Index 3=true
   ) {module=DynamicMaterialParameters001}
   &CurlNoiseForce(
    Noise Frequency=1
    Noise Quality / Cost=::NewEnumerator0
    Noise Strength=&FloatFromCurve(
     FloatCurve=DataInterface() {propertyValues {Curve {keys [
         {arriveTangent=3.7261, interpMode=RCIM_Cubic, leaveTangent=3.7261, tangentMode=RCTM_User, time=0.05, value=0}
         {interpMode=RCIM_Cubic, tangentMode=RCTM_User, time=0.5, value=1}
         <at 0.8 0>
        ]}}}
     Scale Curve=1000
    )
    Pan Noise=true
    Pan Noise Field=&RandomRangeVector(Minimum=<0 0 0>, Maximum=<1 1 1>)
   )
   &AccelerationForce(Acceleration=&Multiply_VectorByFloat(Vector=<0 0 1>, Float=&RandomRangeFloat(Minimum=0, Maximum=1250)))
   &SolveForcesAndVelocity(Clamp Velocity=false, Limit Acceleration=false)
   &ScaleMeshSize(Scale Factor=&VectorFromCurve(
     VectorCurve=DataInterface() {propertyValues {
       XCurve {keys [
         {arriveTangent=4.1785, interpMode=RCIM_Cubic, leaveTangent=4.1785, tangentMode=RCTM_User, time=0, value=0.5}
         {arriveTangent=0.581, interpMode=RCIM_Cubic, leaveTangent=0.581, tangentMode=RCTM_User, time=0.1, value=0.775}
         {arriveTangent=0.0837, interpMode=RCIM_Cubic, leaveTangent=0.0837, tangentMode=RCTM_User, time=1, value=1.1}
        ]}
       YCurve {keys [
         {arriveTangent=4.1785, interpMode=RCIM_Cubic, leaveTangent=4.1785, tangentMode=RCTM_User, time=0, value=0.5}
         {arriveTangent=0.581, interpMode=RCIM_Cubic, leaveTangent=0.581, tangentMode=RCTM_User, time=0.1, value=0.775}
         {arriveTangent=0.0837, interpMode=RCIM_Cubic, leaveTangent=0.0837, tangentMode=RCTM_User, time=1, value=1.1}
        ]}
       ZCurve {keys [
         {arriveTangent=4.1785, interpMode=RCIM_Cubic, leaveTangent=4.1785, tangentMode=RCTM_User, time=0, value=0.5}
         {arriveTangent=0.8594, interpMode=RCIM_Cubic, leaveTangent=0.8594, tangentMode=RCTM_User, time=0.0803, value=0.7813}
         {arriveTangent=0.0837, interpMode=RCIM_Cubic, leaveTangent=0.0837, tangentMode=RCTM_User, time=1, value=1.1}
        ]}
      }}
     Scale Curve=&Multiply_VectorByFloat(
      Vector=<1 1 1>
      Float=&FloatFromCurve(
       FloatCurve=DataInterface() {propertyValues {Curve {keys [<at 0.5 0.65 RCIM_Constant>,<at 0.7 0.7 RCIM_Constant>]}}}
       CurveIndex=&RandomRangeFloat(Minimum=0.4, Maximum=1)
       Scale Curve=&MaskFloatBySpawnGroup(
        Masked Float=&FloatFromCurve(
         FloatCurve=DataInterface() {propertyValues {Curve {keys [
             {arriveTangent=0.8977, interpMode=RCIM_Cubic, leaveTangent=0.8977, tangentMode=RCTM_User, time=0, value=0.6}
             {arriveTangent=0.4659, interpMode=RCIM_Cubic, leaveTangent=0.4659, tangentMode=RCTM_User, time=0.7, value=1}
            ]}}}
         Scale Curve=0.88
        )
        Passthrough Float=1
        Spawn Group Mask=1
       )
      )
     )
    )) {enabled=false}
   &ScaleMeshSize(Scale Factor=&VectorFromCurve(
     VectorCurve=DataInterface() {propertyValues {
       XCurve {keys [
         {arriveTangent=1.1946, interpMode=RCIM_Cubic, leaveTangent=1.1946, tangentMode=RCTM_User, time=0, value=0.7}
         {arriveTangent=0.581, interpMode=RCIM_Cubic, leaveTangent=0.581, tangentMode=RCTM_User, time=0.2, value=0.8875}
         {arriveTangent=0.0837, interpMode=RCIM_Cubic, leaveTangent=0.0837, tangentMode=RCTM_User, time=1, value=1.1}
        ]}
       YCurve {keys [
         {arriveTangent=1.1946, interpMode=RCIM_Cubic, leaveTangent=1.1946, tangentMode=RCTM_User, time=0, value=0.7}
         {arriveTangent=0.581, interpMode=RCIM_Cubic, leaveTangent=0.581, tangentMode=RCTM_User, time=0.2, value=0.8875}
         {arriveTangent=0.0837, interpMode=RCIM_Cubic, leaveTangent=0.0837, tangentMode=RCTM_User, time=1, value=1.1}
        ]}
       ZCurve {keys [
         {arriveTangent=0.9745, interpMode=RCIM_Cubic, leaveTangent=0.9745, tangentMode=RCTM_User, time=0, value=0.7}
         {arriveTangent=0.8594, interpMode=RCIM_Cubic, leaveTangent=0.8594, tangentMode=RCTM_User, time=0.2, value=0.8875}
         {arriveTangent=0.0837, interpMode=RCIM_Cubic, leaveTangent=0.0837, tangentMode=RCTM_User, time=1, value=1.1}
        ]}
      }}
     Scale Curve=&Multiply_VectorByFloat(
      Vector=<1 1 1>
      Float=&FloatFromCurve(
       FloatCurve=DataInterface() {propertyValues {Curve {keys [<at 0.5 0.8 RCIM_Constant>,<at 0.7 0.9 RCIM_Constant>]}}}
       CurveIndex=&RandomRangeFloat(Minimum=0.4, Maximum=1)
       Scale Curve=&MaskFloatBySpawnGroup(
        Masked Float=&FloatFromCurve(
         FloatCurve=DataInterface() {propertyValues {Curve {keys [
             {arriveTangent=0.8977, interpMode=RCIM_Cubic, leaveTangent=0.8977, tangentMode=RCTM_User, time=0, value=0.6}
             {arriveTangent=0.4659, interpMode=RCIM_Cubic, leaveTangent=0.4659, tangentMode=RCTM_User, time=0.7, value=1}
            ]}}}
         Scale Curve=0.88
        )
        Passthrough Float=1
        Spawn Group Mask=1
       )
      )
     )
    )) {module=ScaleMeshSize001}
   &KillParticles(Kill Particles=&SetBoolByFloatComparison(
     A=1
     B=&FloatFromCurve(FloatCurve=DataInterface() {propertyValues {Curve {keys [<at 1.5 0>,<at 2.25 1.01>]}}}, CurveIndex=$Emitter.Age)
     Comparison Type=::NewEnumerator6
    ))
  ]
  rendererClasses [&NiagaraMeshRendererProperties]
  renderers [{rendererClass=&NiagaraMeshRendererProperties, rendererIndex=0}]
 }
 NE_FN_PhysMatDebris001 {
  simTarget=GPUComputeSim
  emitterUpdateScript [
   &EmitterState(Life Cycle Mode=::NewEnumerator1, Loop Behavior=::NewEnumerator1, Loop Duration=0.2, UseLoopDelay=false)
   &SpawnBurst_Instantaneous(
    Spawn Count=&RandomRangeInt(Minimum=10, Maximum=15)
    Use Spawn Probability=false
    UseLoopCountLimit=false
   ) {enabled=false}
   &SpawnParticlesFromOtherEmitter(
    Attribute Reader=DataInterface() {propertyValues {EmitterBinding {bindingMode=Other, emitterName=Smoke_Trail_Leader}}}
    Minimum Age Spawn Threshold=0
    Spawn Rate=25
    Spawn Rate Per Particle Cap=1250
    Use Maximum Age=true
    Use Minimum Age=true
   )
  ]
  particleSpawnScript [
   /SetVariables(Particles.BiasedRandom=&FloatFromCurve(
     FloatCurve=DataInterface() {propertyValues {Curve {keys [
         <at 0.2 1 RCIM_Cubic>
         {arriveTangent=-3.3394, interpMode=RCIM_Cubic, leaveTangent=-3.3394, tangentMode=RCTM_User, time=0.5, value=0.5}
         <at 0.8 0 RCIM_Cubic>
        ]}}}
     CurveIndex=&RandomRangeFloat(Minimum=0, Maximum=1)
    )) {isSetParameters=true}
   &InitializeParticle(
    Color Mode=::NewEnumerator0
    Lifetime=&RandomRangeFloat(Minimum=0.2, Maximum=1.5)
    Mass=&Lerp_Float(A=0.6, B=2.2, Alpha=$Particles.BiasedRandom)
    Mass Mode=::NewEnumerator1
    Mesh Scale=&Multiply_VectorByFloat(
     Vector=&RandomRangeVector(Minimum=<0.9 0.4 0.5>, Maximum=<1 1 1>)
     Float=&Lerp_Float(A=1.4, B=3, Alpha=$Particles.BiasedRandom)
    )
    Mesh Scale Mode=::NewEnumerator4
    UsePositionOffset=false
    Write Material Random=true
   )
   &FN_PhysicalSurface(
    Dirt Mesh Count=2
    Glass Mesh Count=3
    Ice Mesh Count=2
    Metal Gravity=&RandomRangeFloat(Minimum=-350, Maximum=-1800)
    Metal Mesh Count=2
    Metal Velocity=&RandomRangeFloat(Minimum=80, Maximum=375)
    Metal Velocity Cone Angle=130
    Physical Surface=$User.PhysicalSurface
    Stone Gravity=&RandomRangeFloat(Minimum=-700, Maximum=-1250)
    Stone Mesh Count=2
    Stone Velocity=&RandomRangeFloat(Minimum=80, Maximum=550)
    Stone Velocity Cone Angle=120
    Wood Gravity=&RandomRangeFloat(Minimum=-450, Maximum=-900)
    Wood Mesh Count=2
    Wood Scale=1.5
    Wood Velocity=&RandomRangeFloat(Minimum=80, Maximum=600)
    Wood Velocity Cone Angle=130
   )
   &AddVelocity(
    Cone Axis=$System.BlastDir
    Module.UseConeFalloff=false
    Module.UseVelDistribution=false
    Override Seed=false
    Velocity Mode=::NewEnumerator2
    Velocity Speed=&RandomRangeFloat(Minimum=100, Maximum=2500)
   ) {enabled=false}
   &AddVelocityInCone(
    Cone Angle=$Output.FN_PhysicalSurface.SurfaceResponseConeAngle
    Module.Use Velocity Falloff On Cone Axis=false
    Velocity Strength=$Output.FN_PhysicalSurface.SurfaceResponseVelocity
   ) {enabled=false}
   &InitialMeshOrientation(Use Rotation Vector=false)
   &AddRotationalVelocity(Rotation Rate=&RandomVector(Vector Scale=&RandomRangeFloat(Minimum=1, Maximum=8))) {enabled=false}
   &ApplyInitialForces()
   &SampleParticlesFromOtherEmitter(
    Applied Velocity Scale:NiagaraFloat=&RandomRangeFloat(Minimum=0.12, Maximum=0.5)
    Velocity Application:ENiagara_DirectReadApplicationMode=::NewEnumerator0
    Velocity Sampling=::NewEnumerator0
   )
   /SetVariables(Particles.GravityOverrideValue=$Output.FN_PhysicalSurface.GravityOverride) {isSetParameters=true}
   &ShapeLocation(MaskSpawnGroup=false, Override Seed=false, Sphere Radius=30, UseOwnerScale=true)
   &AddVelocity(
    Module.Use Falloff Distance=false
    Override Seed=false
    Velocity Mode=::NewEnumerator1
    Velocity Origin=$Particles.ShapeLocation.ShapePosition
    Velocity Speed=&RandomRangeFloat(Minimum=100, Maximum=500)
   ) {module=AddVelocity001}
   &AddRotationalVelocity(Rotation Rate=<200 0 0>) {module=AddRotationalVelocity001}
  ]
  particleUpdateScript [
   &ParticleState()
   &GravityForce(Gravity=$Particles.GravityOverrideValue)
   &Collision(
    Advanced Aging Rate=1
    Age Colliding Particles=false
    Collision Enabled=&Boolean_AndOperation(
     Bool1=&SetBoolByFloatComparison(A=$Particles.Initial.Lifetime, B=1)
     Bool2=&SetBoolByFloatComparison(A=$Particles.Age, B=0.12)
    )
    Control Roll On Collision=true
    Friction=1
    Friction During a Bounce=0.25
    Kill Offscreen Particles=true
    Module.Control Rest State Via Surface Normal=false
    Module.EnableMaxCollisionCount=false
    Module.Randomize Collision Normal Vector=true
    Randomize Collision Normal=0.25
    ScaleRateOfRotation=false
   ) {enabled=false}
   &SolveForcesAndVelocity(Clamp Velocity=false, Limit Acceleration=false)
   &ScaleMeshSize(Scale Factor=&VectorFromFloat(Value=&FloatFromCurve(FloatCurve=DataInterface() {propertyValues {Curve {keys [<at 0.6 1 RCIM_Cubic>, {arriveTangent=-6.308, interpMode=RCIM_Cubic, leaveTangent=-6.308, tangentMode=RCTM_User, time=1, value=0}]}}})))
   &DynamicMaterialParameters(Module.Param0WriteEnabled=true, Module.Param1WriteEnabled=true, Module.Param2WriteEnabled=false, Module.Param3WriteEnabled=false)
  ]
  rendererClasses [&NiagaraSpriteRendererProperties, &NiagaraMeshRendererProperties]
  renderers [
   {rendererClass=&NiagaraSpriteRendererProperties}
   {rendererClass=&NiagaraMeshRendererProperties, rendererIndex=1}
   {rendererClass=&NiagaraMeshRendererProperties, rendererIndex=2}
   {rendererClass=&NiagaraMeshRendererProperties, rendererIndex=3}
   {rendererClass=&NiagaraMeshRendererProperties, rendererIndex=4}
   {rendererClass=&NiagaraMeshRendererProperties, rendererIndex=5}
   {rendererClass=&NiagaraMeshRendererProperties, rendererIndex=6}
   {rendererClass=&NiagaraMeshRendererProperties, rendererIndex=7}
   {rendererClass=&NiagaraMeshRendererProperties, rendererIndex=8}
  ]
 }
 Smoke_Trail_Leader {
  simTarget=GPUComputeSim
  emitterUpdateScript [
   &EmitterState(Life Cycle Mode=::NewEnumerator1, Loop Behavior=::NewEnumerator1, Loop Duration=0.5836, UseLoopDelay=true)
   &SpawnBurst_Instantaneous(Spawn Count=4, Use Spawn Probability=false, UseLoopCountLimit=false)
   &SpawnBurst_Instantaneous(Spawn Count=4, Spawn Group=1, Use Spawn Probability=false, UseLoopCountLimit=false) {module=SpawnBurst_Instantaneous002}
  ]
  particleSpawnScript [
   &InitializeParticle(
    Lifetime=0.2
    Mass Max=1.4
    Mass Min=1
    Mass Mode=::NewEnumerator3
    Sprite Rotation Mode=::NewEnumerator2
    Sprite Size Mode=::NewEnumerator1
    Uniform Sprite Size Max=12
    Uniform Sprite Size Min=6
    UsePositionOffset=true
    Write Material Random=true
   )
   /Niagara/Modules/Spawn/Location/V2/SphereLocation(
    MaskSpawnGroup=false
    Override Seed=false
    Sphere Radius=15
    Surface Only=false
    UseNonUniformScale=true
    UseOffset=false
    UseSphereOrientation=false
   )
   &AddVelocity(
    Cone Angle=35
    Cone Axis=&MakeCustomVectorFromBool(
     Bool=&SetBoolByIntComparison(A=&SpawnGroup(), B=0, Comparison Type=::NewEnumerator4)
     True Vector=$System.BlastDir
     False Vector=$System.BlastDir2
    )
    Module.UseConeFalloff=false
    Module.UseVelDistribution=false
    Override Seed=false
    Velocity Mode=::NewEnumerator2
    Velocity Speed=&RandomRangeFloat(Minimum=1500, Maximum=2500)
   )
  ]
  particleUpdateScript [
   &ParticleState()
   &GravityForce(Gravity=<0 0 -800>)
   &Drag(Drag=&RandomRangeFloat(Minimum=2, Maximum=4), Use Linear Drag=true, Use Rotational Drag=false)
   &ScaleColor(
    Scale Alpha=&FloatFromCurve(FloatCurve=DataInterface() {propertyValues {Curve {keys [<at -0.0046 1.0113 RCIM_Cubic>,<at 1.0011 0.002 RCIM_Cubic>]}}})
    ScaleA=true
    ScaleRGB=true
   )
   &SolveForcesAndVelocity(Clamp Velocity=false, Limit Acceleration=false)
  ]
  rendererClasses [&NiagaraSpriteRendererProperties, &NiagaraMeshRendererProperties]
  renderers [
   {rendererClass=&NiagaraSpriteRendererProperties}
   {rendererClass=&NiagaraMeshRendererProperties, rendererIndex=1}
  ]
 }
 SourceParticle {
  bEnabled=false
  simTarget=GPUComputeSim
  emitterSpawnScript [
   /SetVariables(Emitter.RasterizationGrid=DataInterface() {propertyValues {
      ClearBeforeNonIterationStage=true
      NumAttributes=4
      NumCellsMaxAxis=3
      NumCellsX=3
      NumCellsY=3
      OverrideBufferFormat=Float
      PreviewAttribute=None
      SetGridFromMaxAxis=false
      bOverrideFormat=false
      bPreviewGrid=false
      RenderTargetUserParameter {parameter {name=None, typeDefHandle {registeredTypeIndex=4}, varData []}}
      WorldBBoxSize=<100 100>
     }}) {isSetParameters=true}
   &Grid2D_SetResolution(Grid:NiagaraDataInterfaceGrid2DCollection=$Emitter.RasterizationGrid, NumCellsX=256, NumCellsY=256, WorldGridExtents:Vector2f=<1200 1200>)
   &Grid2D_UpdateSimAttributes(Grid:NiagaraDataInterfaceGrid2DCollection=$Emitter.RasterizationGrid, WorldGridSize:Vector2f=$Emitter.Grid2D_SetResolution.WorldGridExtents) {module=UpdateSimAttributes}
  ]
  emitterUpdateScript [
   /NiagaraFN/Modules/Scalability/SignificantInstancesCull(Cull Response:ENiagaraExecutionStateManagement=::KillImmediately)
   &EmitterState(
    Enable Distance Culling=true
    Life Cycle Mode=::NewEnumerator1
    Loop Behavior=::NewEnumerator1
    Loop Duration=0.2
    MaxDistance=8000
    Module.bUseMaxDistance=true
    Module.bUseMinDistance=false
    Scalability Mode=::NewEnumerator1
    UseLoopDelay=false
   )
   &SpawnBurst_Instantaneous(
    Spawn Count=&RandomRangeInt(Minimum=6, Maximum=10, Evaluation Type=::NewEnumerator1)
    Use Spawn Probability=false
   )
   &SpawnRate(SpawnRate=1300, Use Spawn Probability=false) {enabled=false}
   &Grid2D_CreateUnitToWorldTransform(
    Camera Facing=true
    Local Pivot:Vector2f=<0.5 0.1>
    Offset:NiagaraPosition=$Engine.Owner.Position
    Rotation:Quat4f=$Engine.Owner.Rotation
    WorldGridExtents:Vector2f=$Emitter.Grid2D_SetResolution.WorldGridExtents
   )
  ]
  particleSpawnScript [
   /Niagara/Modules/Spawn/Initialization/InitializeParticle(
    Color=<rgba 1111 13.6009 0 1>
    Lifetime=&RandomRangeFloat(Minimum=0.85, Maximum=1.5)
    Mass=&RandomRangeFloat(Minimum=0.75, Maximum=1)
    Module.Write Color=true
    Module.Write Lifetime=true
    Module.Write Mass=true
    Module.Write Material Random=true
    Module.Write Position=true
    Module.Write Scale=false
    Module.Write SpriteRotation=false
    Module.Write SpriteSize=true
    Sprite Size=&Vector2DFromFloat(Value=&Multiply_Float(A=&RandomRangeFloat(Minimum=12, Maximum=250), B=$User.Grid2DScale))
   )
   &AddVelocityInCone(
    Cone Angle=80
    Cone Axis=<0 0 1>
    Cone Axis Coordinate Space=::World
    Module.Use Velocity Falloff On Cone Axis=true
    Velocity Distribution Along Cone Axis=0
    Velocity Falloff Away From Cone Axis=0
    Velocity Strength=&Multiply_Float(A=&RandomRangeFloat(Minimum=1075, Maximum=1756), B=$User.Grid2DScale)
   )
   &SphereLocation(
    MaskSpawnGroup=false
    Override Seed=false
    Sphere Radius=30
    Surface Only=false
    UseNonUniformScale=true
    UseOffset=true
    UseSphereOrientation=false
   ) {enabled=false}
   /Niagara/Modules/Update/Forces/CurlNoiseForce(Noise Frequency=0.04, Noise Strength=0, Pan Noise=true, Pan Noise Field=<0.3 0.5 0.3>) {enabled=false}
   &ApplyInitialForces()
   &TorusLocation(
    Align Torus With Axis=true
    Large Radius:NiagaraFloat=&Multiply_Float(A=60, B=$User.Grid2DScale)
    MaskSpawnGroup=false
    Torus Axis=<1 0 0>
    Torus Distribution Mode:ENiagaraTorusDistributionMode=::NewEnumerator1
    Torus Mode=::NewEnumerator2
    U Position:NiagaraFloat=&Subtract_Float(A=&ReturnNormalizedExecIndex(), B=2)
   )
   &AddVelocityFromPoint(
    Module.Use Falloff Distance=false
    Origin Offset=&Multiply_VectorByFloat(Vector=$Engine.Owner.SystemXAxis, Float=&RandomRangeFloat(Minimum=0, Maximum=-60))
    Origin Offset Coordinate Space=::Simulation
    Velocity Origin:Vector3f=&SimulationPosition()
    Velocity Strength=&UniformRangedFloat(Minimum=5, Maximum=5)
   )
  ]
  particleUpdateScript [
   &ParticleState()
   &CurlNoiseForce(Noise Strength=300, Pan Noise=true, Pan Noise Field=<0.1 0.5 0.2>) {module=CurlNoiseForce001}
   &GravityForce(Gravity=&Multiply_VectorByFloat(Vector=<0 0 -1500>, Float=$User.Grid2DScale))
   &Drag(Drag=1.4, Use Linear Drag=true, Use Rotational Drag=false)
   &ScaleColor(
    Scale Alpha=&FloatFromCurve(FloatCurve=DataInterface() {propertyValues {Curve {keys [<at -0.0046 1.0113 RCIM_Cubic>,<at 1.0011 0.002 RCIM_Cubic>]}}})
    ScaleA=true
    ScaleRGB=true
   )
   &Collision(
    Age Colliding Particles=false
    Collision Enabled=&SetBoolByFloatComparison(A=$Particles.Age, B=0.4)
    Enable Rest State=false
    Friction=0.2
    Friction During a Bounce=0.3
    Kill Occluded Particles=false
    Kill Offscreen Particles=true
    Maximum time offscreen=99
    Module.EnableMaxCollisionCount=false
    Module.Randomize Collision Normal Vector=false
    Restitution=0.2
   ) {module=Collision001}
   &SolveForcesAndVelocity(Clamp Velocity=false, Limit Acceleration=false)
   /NiagaraFluids/Modules/Particles/Fluids_Gas_Source(
    Falloff=0.001
    NoiseIntensity=0
    NoiseScale=3
    Radius:NiagaraFloat=&FloatFromCurve(
     FloatCurve=DataInterface() {propertyValues {Curve {keys [<at 0 0>, {interpMode=RCIM_Cubic, leaveTangent=-2.282, tangentMode=RCTM_Break, time=0.0361, value=0.984}, <at 1.0022 0.25>]}}}
     Scale Curve=&Multiply_Float(A=&RandomRangeFloat(Minimum=45, Maximum=70), B=$User.Grid2DScale)
    )
    Temperature:NiagaraFloat=&FloatFromCurve(
     FloatCurve=DataInterface() {propertyValues {Curve {keys [<at 0 1>,<at 0.228 0.772>,<at 1 0>]}}}
     Scale Curve=&RandomRangeFloat(Minimum=500, Maximum=3000)
    )
    VelocityScale:NiagaraFloat=&RandomRangeFloat(Minimum=0.5, Maximum=1)
   )
   /SetVariables(Particles.SpriteSize=&Vector2DFromFloat(Value=&Multiply_Float(A=$Particles.fluids_source_radius, B=2))) {isSetParameters=true}
  ]
  rendererClasses [&NiagaraSpriteRendererProperties]
  renderers [{rendererClass=&NiagaraSpriteRendererProperties}]
 }
 Sparkles001 {
  simTarget=GPUComputeSim
  emitterUpdateScript [
   &EmitterState(Life Cycle Mode=::NewEnumerator1, Loop Behavior=::NewEnumerator1, Loop Duration=0.2, UseLoopDelay=true)
   &SpawnParticlesFromOtherEmitter(
    Attribute Reader=DataInterface() {propertyValues {EmitterBinding {bindingMode=Other, emitterName=Smoke_Trail_Leader}}}
    Minimum Age Spawn Threshold=0
    Spawn Rate=200
    Spawn Rate Per Particle Cap=1250
    Use Maximum Age=true
    Use Minimum Age=true
   )
  ]
  particleSpawnScript [
   /Niagara/Modules/Spawn/Initialization/V2/InitializeParticle(
    Color=$User.Global Color
    Lifetime Max=1.5
    Lifetime Min=0.1
    Lifetime Mode=::NewEnumerator1
    Sprite Size Max=<3 5>
    Sprite Size Min=<4 2.5>
    Sprite Size Mode=::NewEnumerator2
    UsePositionOffset=false
    Write Material Random=true
   )
   &ApplyInitialForces(Apply Rotational Force to Rotational Velocity=false) {enabled=false}
   /SetVariables(Particles.SpriteSize=&Multiply_Vector2D_ByFloat(Float=$User.Global Scale, Vector2D=$StackContext.SpriteSize)) {isSetParameters=true, module=SetVariables_#4}
   &SampleParticlesFromOtherEmitter(
    Applied Velocity Scale:NiagaraFloat=&RandomRangeFloat(Minimum=0.5, Maximum=1.5)
    Velocity Sampling=::NewEnumerator0
   )
   /Niagara/Modules/Spawn/Location/SphereLocation(
    MaskSpawnGroup=false
    Module.Override Randomness=false
    Sphere Radius=&Multiply_Float(A=5, B=$User.Global Scale)
    Surface Only=false
   ) {enabled=false}
  ]
  particleUpdateScript [
   &ParticleState()
   &GravityForce(Gravity=&RandomRangeVector(Minimum=<0 0 -400>, Maximum=<0 0 -980>))
   &Drag(Drag=&RandomRangeFloat(Minimum=1, Maximum=6), Use Linear Drag=true, Use Rotational Drag=false)
   /Niagara/Modules/Update/Forces/V2/CurlNoiseForce(Noise Frequency=5, Noise Quality / Cost=::NewEnumerator0, Noise Strength=500, Pan Noise=false) {enabled=false}
   &SolveForcesAndVelocity(Clamp Velocity=false, Limit Acceleration=false)
   &ScaleColor(
    Scale Alpha=&FloatFromCurve(FloatCurve=DataInterface() {propertyValues {Curve {keys [
         {arriveTangent=26.7273, interpMode=RCIM_Cubic, leaveTangent=26.7273, tangentMode=RCTM_User, time=-0.0092, value=0.0144}
         {interpMode=RCIM_Cubic, tangentMode=RCTM_User, time=0.1, value=1}
         {interpMode=RCIM_Cubic, tangentMode=RCTM_User, time=0.85, value=1}
         {arriveTangent=-1.3701, interpMode=RCIM_Cubic, leaveTangent=-1.3701, tangentMode=RCTM_User, time=1.0011, value=0}
        ]}}})
    Scale RGB=&Lerp_Vector(
     A=<0 0 0>
     B=<12.5 4 0.8>
     Alpha=&FloatFromCurve(FloatCurve=DataInterface() {propertyValues {Curve {keys [
          {arriveTangent=15.3906, interpMode=RCIM_Cubic, leaveTangent=15.3906, tangentMode=RCTM_User, time=-0.0092, value=1}
          {interpMode=RCIM_Cubic, tangentMode=RCTM_User, time=0.0967, value=0.8}
          {interpMode=RCIM_Cubic, tangentMode=RCTM_User, time=0.7, value=0.8}
          <at 1 0 RCIM_Cubic>
         ]}}})
    )
    ScaleA=true
    ScaleRGB=true
   )
   &ScaleSpriteSize(Scale Factor=&Vector2DFromFloat(Value=&FloatFromCurve(FloatCurve=DataInterface() {propertyValues {Curve {keys [<at -0.0046 0>,<at 0.025 1>,<at 0.66 1>,<at 1.0011 0.5>]}}}, Scale Curve=1.5)))
   &ScaleSpriteSizeBySpeed(Max Scale Factor=<0.5 5>, Min Scale Factor=<1 0.6>, Velocity Threshold=1250)
  ]
  rendererClasses [&NiagaraSpriteRendererProperties]
  renderers [{rendererClass=&NiagaraSpriteRendererProperties}]
 }
 Streaks_Fat001 {
  emitterUpdateScript [
   &EmitterState(Life Cycle Mode=::NewEnumerator1, Loop Behavior=::NewEnumerator1, Loop Delay=0.0353, Loop Duration=0.4, UseLoopDelay=true)
   &SpawnBurst_Instantaneous(Spawn Count=&Multiply_Int(A=4, B=1), Use Spawn Probability=false, UseLoopCountLimit=false)
   &SpawnBurst_Instantaneous(Spawn Count=4, Use Spawn Probability=false, UseLoopCountLimit=false) {module=SpawnBurst_Instantaneous001}
   &SpawnBurst_Instantaneous(Spawn Count=&Multiply_Int(A=4, B=1), Spawn Group=3, Use Spawn Probability=false, UseLoopCountLimit=false) {module=SpawnBurst_Instantaneous002}
  ]
  particleSpawnScript [
   &InitializeParticle(
    Color=<rgba 0.9635 0.9372 0.8852 1>
    Lifetime=&MakeCustomFloatFromBool(
     Bool=/Niagara/DynamicInputs/Helpers/MaskBoolBySpawnGroup(Masked Bool:NiagaraBool=false, Passthrough Bool:NiagaraBool=true, Spawn Group Mask:NiagaraInt32=3)
     True Float=&RandomRangeFloat(Minimum=0.5, Maximum=2)
     False Float=&Multiply_Float(A=&RandomRangeFloat(Minimum=0.14, Maximum=0.17), B=0.5)
    )
    Sprite Size=&Multiply_Vector2D_ByFloat(Float=$User.Global Scale, Vector2D=&UniformRangedVector2D(Minimum=<85 100>, Maximum=<100 220>))
    Sprite Size Mode=::NewEnumerator4
    UsePositionOffset=false
    Write Material Random=true
   )
   /SetVariables(Particles.VisibilityTag=&SpawnGroup()) {isSetParameters=true}
   &SphereLocation(
    Coordinate Space:ENiagaraCoordinateSpace=::Simulation
    MaskSpawnGroup=false
    Module.Override Randomness=false
    Offset=&MakeCustomVectorFromBool(
     Bool=&SetBoolByIntComparison(A=&SpawnGroup(Max Spawn Group=1), B=0, Comparison Type=::NewEnumerator4)
     True Vector=$System.BlastSrcA
     False Vector=$System.BlastSrcB
    )
    Sphere Radius=&Multiply_Float(A=30, B=$User.Global Scale)
    Surface Only=true
   )
   &AddVelocityFromPoint(
    Module.Use Falloff Distance=false
    Origin Offset=<0 0 -3>
    Velocity Strength=&UniformRangedFloat(Minimum=400, Maximum=500)
   ) {enabled=false}
   &AddVelocity(
    Cone Angle=70
    Cone Axis=&MakeCustomVectorFromBool(
     Bool=&SetBoolByFloatComparison(A=&ReturnNormalizedExecIndex(), B=0.5)
     True Vector=$System.BlastDir
     False Vector=$System.BlastDir2
    )
    Module.UseConeFalloff=false
    Module.UseVelDistribution=false
    Override Seed=false
    Velocity Mode=::NewEnumerator2
    Velocity Speed=388.9138
   )
   &SubUVAnimation(End Frame=3, SubUV Animation Mode:ENiagara_SubUVLookupMode=::NewEnumerator2)
   &CameraOffset(Camera Offset Amount=&Multiply_Float(A=&MaskFloatBySpawnGroup(Masked Float=10.01, Passthrough Float=0, Spawn Group Mask=1), B=$User.Global Scale))
   /SetVariables(Particles.SpikesSize:Vector2f=&Multiply_Vector2D_ByFloat(
     Float=&Multiply_Float(A=$User.Global Scale, B=0.7)
     Vector2D=&UniformRangedVector2D(Minimum=<85 100>, Maximum=<100 220>)
    )) {isSetParameters=true}
  ]
  particleUpdateScript [
   &ParticleState()
   &Drag(
    Drag=&FloatFromCurve(FloatCurve=DataInterface() {propertyValues {Curve {keys [<at 0.2 1>,<at 0.3 0>]}}}, Scale Curve=15)
    Use Linear Drag=true
    Use Rotational Drag=false
   )
   &GravityForce(Gravity=&Multiply_VectorByFloat(
     Vector=<0 0 -2.2>
     Float=&FloatFromCurve(
      FloatCurve=DataInterface() {propertyValues {Curve {keys [{arriveTangent=3.1357, interpMode=RCIM_Cubic, leaveTangent=3.1357, tangentMode=RCTM_User, time=0.1, value=0}, <at 0.6 1>]}}}
      CurveIndex=$Emitter.Age
     )
    )) {enabled=false}
   &SolveForcesAndVelocity(Clamp Velocity=false, Limit Acceleration=false)
   &ScaleSpriteSize(Scale Factor=&Vector2DFromCurve(
     Vector2Curve=DataInterface() {propertyValues {
       XCurve {keys [
         <at 0 0.625 RCIM_Cubic>
         {arriveTangent=1.125, interpMode=RCIM_Cubic, leaveTangent=1.125, time=0.4, value=0.75}
         {arriveTangent=2.6138, interpMode=RCIM_Cubic, leaveTangent=2.6138, tangentMode=RCTM_User, time=1, value=1.75}
        ]}
       YCurve {keys [
         {arriveTangent=45.128, interpMode=RCIM_Cubic, leaveTangent=45.128, tangentMode=RCTM_User, time=0, value=0.1}
         <at 0.1 1.625>
         <at 1 1.875 RCIM_Cubic>
        ]}
      }}
     ScaleCurve=<1.5 2>
    ))
   &ScaleColor(
    Linear Color Curve=DataInterface() {propertyValues {
      AlphaCurve {keys [
        {arriveTangent=-0.3577, interpMode=RCIM_Cubic, leaveTangent=-0.3577, tangentMode=RCTM_User, time=0.08, value=1}
        {arriveTangent=-1.087, interpMode=RCIM_Cubic, leaveTangent=-1.0164, tangentMode=RCTM_Break, time=0.3702, value=1}
        <at 1 0>
       ]}
      BlueCurve {keys [<at 0.6491 0.3>,<at 1 0.45>]}
      GreenCurve {keys [<at 0.6491 0.3>,<at 1 0.45>]}
      RedCurve {keys [<at 0.6491 0.3>,<at 1 0.45>]}
     }}
    Scale Mode=::NewEnumerator2
   )
   &DynamicMaterialParameters(
    Index 0 Param 1=&Multiply_Float(A=100.01, B=$User.Global Scale)
    Index 0 Param 4=&FloatFromCurve(FloatCurve=DataInterface() {propertyValues {Curve {keys [{interpMode=RCIM_Cubic, tangentMode=RCTM_User, time=0, value=0}, {interpMode=RCIM_Cubic, tangentMode=RCTM_User, time=0.4, value=1}]}}})
    Index 1 Param 1=&FloatFromCurve(
     FloatCurve=DataInterface() {propertyValues {Curve {keys [
         {interpMode=RCIM_Cubic, tangentMode=RCTM_User, time=0, value=0}
         {arriveTangent=0.6295, interpMode=RCIM_Cubic, leaveTangent=0.6295, tangentMode=RCTM_User, time=1, value=1}
        ]}}}
     Scale Curve=0.42
    )
    Index 1 Param 2=&FloatFromCurve(
     FloatCurve=DataInterface() {propertyValues {Curve {keys [
         {arriveTangent=-713.2054, interpMode=RCIM_Cubic, leaveTangent=-713.2054, tangentMode=RCTM_User, time=0.05, value=50}
         {arriveTangent=-0.3869, interpMode=RCIM_Cubic, leaveTangent=-0.3869, tangentMode=RCTM_User, time=0.25, value=0}
        ]}}}
     CurveIndex=$Emitter.Age
     Scale Curve=&MaskFloatBySpawnGroup(Masked Float=0.125, Passthrough Float=0.125, Spawn Group Mask=1)
    )
    Index 1 Param 3=&FloatFromCurve(
     FloatCurve=DataInterface() {propertyValues {Curve {keys [{interpMode=RCIM_Cubic, tangentMode=RCTM_User, time=0.2, value=0.5}, {interpMode=RCIM_Cubic, tangentMode=RCTM_User, time=1, value=1}]}}}
     Scale Curve=2
    )
    Index 1 Param 4=&FloatFromCurve(FloatCurve=DataInterface() {propertyValues {Curve {keys [<at 0 1>,<at 1 0>]}}}, Scale Curve=-0.3)
    Module.Param0WriteEnabled=true
    Module.Param0WriteEnabled1=true
    Module.Param1WriteEnabled=false
    Module.Param1WriteEnabled1=true
    Module.Param2WriteEnabled=false
    Module.Param2WriteEnabled1=true
    Module.Param3WriteEnabled=true
    Module.Param3WriteEnabled1=true
    Write Parameter Index 1=true
   )
   /SetVariables(
    Particles.Color2=&ColorFromCurve(ColorCurve=DataInterface() {propertyValues {
       AlphaCurve {keys [
         {arriveTangent=-0.3577, interpMode=RCIM_Cubic, leaveTangent=-0.3577, tangentMode=RCTM_User, time=0.08, value=1}
         {arriveTangent=-1.087, interpMode=RCIM_Cubic, leaveTangent=-1.0164, tangentMode=RCTM_Break, time=0.2616, value=1}
         <at 1 0>
        ]}
       BlueCurve {keys [<at 0.1976 0>]}
       GreenCurve {keys [<at 0.1976 0.1604>]}
       RedCurve {keys [<at 0.1976 4>]}
      }})
    Particles.SpikesParams:Vector4f=&MakeVector4(
     X=&Multiply_Float(A=100, B=$User.Global Scale)
     Y=0
     Z=0
     W=&FloatFromCurve(FloatCurve=DataInterface() {propertyValues {Curve {keys [{interpMode=RCIM_Cubic, tangentMode=RCTM_User, time=0, value=0}, {interpMode=RCIM_Cubic, tangentMode=RCTM_User, time=0.4, value=1}]}}})
    )
    Particles.SpikesParams2:Vector4f=&MakeVector4(
     X=&FloatFromCurve(
      FloatCurve=DataInterface() {propertyValues {Curve {keys [
          {interpMode=RCIM_Cubic, tangentMode=RCTM_User, time=0.2, value=0}
          {arriveTangent=0.6295, interpMode=RCIM_Cubic, leaveTangent=0.6295, tangentMode=RCTM_User, time=1, value=1}
         ]}}}
      Scale Curve=0.03
     )
     Y=&FloatFromCurve(
      FloatCurve=DataInterface() {propertyValues {Curve {keys [
          {arriveTangent=-9.7084, interpMode=RCIM_Cubic, leaveTangent=-9.7084, tangentMode=RCTM_User, time=0.1, value=50}
          {arriveTangent=-0.3869, interpMode=RCIM_Cubic, leaveTangent=-0.3869, tangentMode=RCTM_User, time=0.35, value=0}
         ]}}}
      CurveIndex=$Emitter.Age
      Scale Curve=&MaskFloatBySpawnGroup(Masked Float=0, Passthrough Float=0.9, Spawn Group Mask=1)
     )
     Z=&FloatFromCurve(
      FloatCurve=DataInterface() {propertyValues {Curve {keys [{interpMode=RCIM_Cubic, tangentMode=RCTM_User, time=0.2, value=0.5}, {interpMode=RCIM_Cubic, tangentMode=RCTM_User, time=1, value=1}]}}}
      Scale Curve=2
     )
     W=&FloatFromCurve(FloatCurve=DataInterface() {propertyValues {Curve {keys [<at 0 1>,<at 1 0>]}}}, Scale Curve=5)
    )
    Particles.SpikesSize:Vector2f=&Multiply_Vector2D(
     A=$Particles.Initial.SpikesSize
     B=&Vector2DFromCurve(
      Vector2Curve=DataInterface() {propertyValues {
        XCurve {keys [
          <at 0 0.625 RCIM_Cubic>
          {arriveTangent=0.875, interpMode=RCIM_Cubic, leaveTangent=0.875, time=0.4, value=0.625}
          {arriveTangent=1.2531, interpMode=RCIM_Cubic, leaveTangent=1.2531, tangentMode=RCTM_User, time=1, value=1.3}
         ]}
        YCurve {keys [
          {arriveTangent=9.4573, interpMode=RCIM_Cubic, leaveTangent=9.4573, tangentMode=RCTM_User, time=0, value=0}
          <at 0.18 1.625>
          <at 1 2 RCIM_Cubic>
         ]}
       }}
      ScaleCurve=<1.7 2.1>
     )
    )
   ) {isSetParameters=true}
  ]
  rendererClasses [&NiagaraSpriteRendererProperties]
  renderers [
   {rendererClass=&NiagaraSpriteRendererProperties}
   {rendererClass=&NiagaraSpriteRendererProperties, rendererIndex=1}
  ]
 }
 Streaks_Fat_3 {
  bEnabled=false
  emitterUpdateScript [
   &EmitterState(Life Cycle Mode=::NewEnumerator1, Loop Behavior=::NewEnumerator1, Loop Delay=0.0353, Loop Duration=0.4, UseLoopDelay=true)
   &SpawnBurst_Instantaneous(Spawn Count=&Multiply_Int(A=4, B=1), Use Spawn Probability=false, UseLoopCountLimit=false)
   &SpawnBurst_Instantaneous(Spawn Count=4, Spawn Group=1, Use Spawn Probability=false, UseLoopCountLimit=false) {module=SpawnBurst_Instantaneous001}
  ]
  particleSpawnScript [
   &InitializeParticle(
    Color=<rgba 0.9635 0.9372 0.8852 1>
    Lifetime Min=1.5
    Lifetime Mode=::NewEnumerator1
    Sprite Size=&Multiply_Vector2D_ByFloat(Float=$User.Global Scale, Vector2D=&UniformRangedVector2D(Minimum=<85 100>, Maximum=<100 220>))
    Sprite Size Mode=::NewEnumerator4
    UsePositionOffset=false
    Write Material Random=true
   )
   &SphereLocation(
    Coordinate Space:ENiagaraCoordinateSpace=::Simulation
    MaskSpawnGroup=false
    Module.Override Randomness=false
    Offset=&MakeCustomVectorFromBool(
     Bool=&SetBoolByIntComparison(A=&SpawnGroup(Max Spawn Group=1), B=0, Comparison Type=::NewEnumerator4)
     True Vector=$System.BlastSrcA
     False Vector=$System.BlastSrcB
    )
    Sphere Radius=&Multiply_Float(A=30, B=$User.Global Scale)
    Surface Only=true
   )
   &AddVelocityFromPoint(
    Module.Use Falloff Distance=false
    Origin Offset=<0 0 -3>
    Velocity Strength=&UniformRangedFloat(Minimum=400, Maximum=500)
   ) {enabled=false}
   &AddVelocity(
    Cone Angle=70
    Cone Axis=&MakeCustomVectorFromBool(
     Bool=&SetBoolByIntComparison(A=&SpawnGroup(), B=0, Comparison Type=::NewEnumerator4)
     True Vector=$System.BlastDir
     False Vector=$System.BlastDir2
    )
    Module.UseConeFalloff=false
    Module.UseVelDistribution=false
    Override Seed=false
    Velocity Mode=::NewEnumerator2
    Velocity Speed=388.9138
   )
   &SubUVAnimation(End Frame=3, SubUV Animation Mode:ENiagara_SubUVLookupMode=::NewEnumerator2)
   &CameraOffset(Camera Offset Amount=&Multiply_Float(A=&MaskFloatBySpawnGroup(Masked Float=10.01, Passthrough Float=0, Spawn Group Mask=1), B=$User.Global Scale))
  ]
  particleUpdateScript [
   &ParticleState()
   &Drag(
    Drag=&FloatFromCurve(FloatCurve=DataInterface() {propertyValues {Curve {keys [<at 0.2 1>,<at 0.3 0>]}}}, Scale Curve=15)
    Use Linear Drag=true
    Use Rotational Drag=false
   )
   &GravityForce(Gravity=&Multiply_VectorByFloat(
     Vector=<0 0 -2.2>
     Float=&FloatFromCurve(
      FloatCurve=DataInterface() {propertyValues {Curve {keys [{arriveTangent=3.1357, interpMode=RCIM_Cubic, leaveTangent=3.1357, tangentMode=RCTM_User, time=0.1, value=0}, <at 0.6 1>]}}}
      CurveIndex=$Emitter.Age
     )
    )) {enabled=false}
   &SolveForcesAndVelocity(Clamp Velocity=false, Limit Acceleration=false)
   &ScaleSpriteSize(Scale Factor=&Vector2DFromCurve(
     Vector2Curve=DataInterface() {propertyValues {
       XCurve {keys [
         <at 0 0.625 RCIM_Cubic>
         {arriveTangent=1.125, interpMode=RCIM_Cubic, leaveTangent=1.125, time=0.4, value=0.75}
         {arriveTangent=2.6138, interpMode=RCIM_Cubic, leaveTangent=2.6138, tangentMode=RCTM_User, time=1, value=1.75}
        ]}
       YCurve {keys [
         {arriveTangent=45.128, interpMode=RCIM_Cubic, leaveTangent=45.128, tangentMode=RCTM_User, time=0, value=0.1}
         <at 0.1 1.625>
         <at 1 1.875 RCIM_Cubic>
        ]}
      }}
     ScaleCurve=<1.5 2>
    ))
   &ScaleColor(
    Linear Color Curve=DataInterface() {propertyValues {
      AlphaCurve {keys [
        {arriveTangent=-0.3577, interpMode=RCIM_Cubic, leaveTangent=-0.3577, tangentMode=RCTM_User, time=0.08, value=1}
        {arriveTangent=-1.087, interpMode=RCIM_Cubic, leaveTangent=-1.0164, tangentMode=RCTM_Break, time=0.3702, value=1}
        <at 1 0>
       ]}
      BlueCurve {keys [<at 0.6491 0.3>,<at 1 0.45>]}
      GreenCurve {keys [<at 0.6491 0.3>,<at 1 0.45>]}
      RedCurve {keys [<at 0.6491 0.3>,<at 1 0.45>]}
     }}
    Scale Mode=::NewEnumerator2
   )
   &DynamicMaterialParameters(
    Index 0 Param 1=&Multiply_Float(A=100.01, B=$User.Global Scale)
    Index 0 Param 4=&FloatFromCurve(FloatCurve=DataInterface() {propertyValues {Curve {keys [{interpMode=RCIM_Cubic, tangentMode=RCTM_User, time=0, value=0}, {interpMode=RCIM_Cubic, tangentMode=RCTM_User, time=0.4, value=1}]}}})
    Index 1 Param 1=&FloatFromCurve(
     FloatCurve=DataInterface() {propertyValues {Curve {keys [
         {interpMode=RCIM_Cubic, tangentMode=RCTM_User, time=0, value=0}
         {arriveTangent=0.6295, interpMode=RCIM_Cubic, leaveTangent=0.6295, tangentMode=RCTM_User, time=1, value=1}
        ]}}}
     Scale Curve=0.42
    )
    Index 1 Param 2=&FloatFromCurve(
     FloatCurve=DataInterface() {propertyValues {Curve {keys [
         {arriveTangent=-713.2054, interpMode=RCIM_Cubic, leaveTangent=-713.2054, tangentMode=RCTM_User, time=0.05, value=50}
         {arriveTangent=-0.3869, interpMode=RCIM_Cubic, leaveTangent=-0.3869, tangentMode=RCTM_User, time=0.25, value=0}
        ]}}}
     CurveIndex=$Emitter.Age
     Scale Curve=&MaskFloatBySpawnGroup(Masked Float=0.125, Passthrough Float=0.125, Spawn Group Mask=1)
    )
    Index 1 Param 3=&FloatFromCurve(
     FloatCurve=DataInterface() {propertyValues {Curve {keys [{interpMode=RCIM_Cubic, tangentMode=RCTM_User, time=0.2, value=0.5}, {interpMode=RCIM_Cubic, tangentMode=RCTM_User, time=1, value=1}]}}}
     Scale Curve=2
    )
    Index 1 Param 4=&FloatFromCurve(FloatCurve=DataInterface() {propertyValues {Curve {keys [<at 0 1>,<at 1 0>]}}}, Scale Curve=-0.3)
    Module.Param0WriteEnabled=true
    Module.Param0WriteEnabled1=true
    Module.Param1WriteEnabled=false
    Module.Param1WriteEnabled1=true
    Module.Param2WriteEnabled=false
    Module.Param2WriteEnabled1=true
    Module.Param3WriteEnabled=true
    Module.Param3WriteEnabled1=true
    Write Parameter Index 1=true
   )
  ]
  rendererClasses [&NiagaraSpriteRendererProperties]
  renderers [{rendererClass=&NiagaraSpriteRendererProperties}]
 }
}
summary {
 systemName=NS_Feycrab_Prj_Explosion
 emitters [
  {
   emitterName=Smoke_Trail_Leader
   simTarget=GPUComputeSim
   rendererClasses [&NiagaraSpriteRendererProperties, &NiagaraMeshRendererProperties]
  }
  {bEnabled=false, emitterName=Flash_wide, rendererClasses [&NiagaraMeshRendererProperties]}
  {emitterName=Sparkles001, simTarget=GPUComputeSim, rendererClasses [&NiagaraSpriteRendererProperties]}
  {bEnabled=false, emitterName=Blast_Stylized, rendererClasses [&NiagaraSpriteRendererProperties]}
  {bEnabled=false, emitterName=Streaks_Fat_3, rendererClasses [&NiagaraSpriteRendererProperties]}
  {emitterName=Dark_Chunks_6, simTarget=GPUComputeSim, rendererClasses [&NiagaraSpriteRendererProperties]}
  {emitterName=Grid2D_Gas_SmokeFire, simTarget=GPUComputeSim, rendererClasses [&NiagaraSpriteRendererProperties]}
  {bEnabled=false, emitterName=SourceParticle, simTarget=GPUComputeSim, rendererClasses [&NiagaraSpriteRendererProperties]}
  {emitterName=Fireworks_Mesh_AndDistortion, rendererClasses [&NiagaraMeshRendererProperties]}
  {emitterName=LingeringEmbers, simTarget=GPUComputeSim, rendererClasses [&NiagaraSpriteRendererProperties]}
  {emitterName=NE_Explosion_Main002, rendererClasses [&NiagaraMeshRendererProperties]}
  {
   emitterName=NE_FN_PhysMatDebris001
   simTarget=GPUComputeSim
   rendererClasses [&NiagaraSpriteRendererProperties, &NiagaraMeshRendererProperties]
  }
  {
   bEnabled=false
   emitterName=EnergyWave
   rendererClasses [&NiagaraMeshRendererProperties, &NiagaraLightRendererProperties]
  }
  {emitterName=Streaks_Fat001, rendererClasses [&NiagaraSpriteRendererProperties]}
  {
   emitterName=FlashWide_EnergyWave_Combined
   rendererClasses [&NiagaraMeshRendererProperties, &NiagaraLightRendererProperties]
  }
 ]
 userVariables [
  User.ExplosionPointBack {type=NiagaraPosition, description="", defaultValue {}}
  User.ExplosionPointFront {type=NiagaraPosition, description="", defaultValue {}}
  User.Global Color {type=LinearColor, description="", defaultValue {struct=LinearColor, value=<rgba 4 2.156 0.6271 1>}}
  User.Global Scale {type=NiagaraFloat, defaultValue=1, description=""}
  User.Grid2DScale {type=NiagaraFloat, defaultValue=1, description=""}
  User.IsUnderwater {type=NiagaraBool}
  User.PhysicalSurface {type=NiagaraInt32}
  User.ResolutionMaxAxis {type=NiagaraInt32}
  User.ShowOverlays {type=NiagaraBool}
  User.WaterHeight {type=NiagaraFloat, defaultValue=0, description=""}
  User.WorldSpaceSize {type=Vector3f, description="", defaultValue {struct=Vector3f, value=<0 0 0>}}
 ]
}
