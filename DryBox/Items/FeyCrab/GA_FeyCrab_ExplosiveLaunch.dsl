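; Gameplay ability that fires the FeyCrab launcher's explosive projectile, ending immediately if the pawn has a hiding-prop tag.
; Otherwise, behind a DoOnce, it sends a Fired event, caches the current weapon as the launcher and, if ammo remains, calls its OnFired.
; It then applies a prevent-actions effect while waiting for the projectile Explode event, which removes the projectile from the world and ends the ability.
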(event Ability|EventActivateAbility
  (bind _returnvalue (Ability|GetActivatingPawn))
  (bind _returnvalue_2 (Ability|GetActivatingPawn))
  ;[1 1 1] On Fire send over events for firing and preventing actions
  (if (GameplayTags|HasAnyMatchingGameplayTags (Ability|GetActivatingPawn) (Variables|Default|GetHidingPropTag))
    (Ability|EndAbility)
    (else
      (Utilities|FlowControl|DoOnce false
        (:Completed
          (Ability|SendGameplayEventtoActor _returnvalue_2 "(TagName=\"Event.Ability.FeyCrab.Fired\")")
          (bind _asb_fey_crab_launcher (Utilities|Casting|CastToB_FeyCrab_Launcher (Variables|/Script/FortniteGame.FortPawn|GetCurrentWeapon _returnvalue_2)))
          (Variables|Default|SetFey%20Crab%20Launcher%20Ref _asb_fey_crab_launcher)
          (if (> (FortWeapon|GetRemainingAmmo _asb_fey_crab_launcher) 0)
            (Class|BFeyCrabLauncher|OnFired (Variables|Default|GetFey%20Crab%20Launcher%20Ref) self)
            (Ability|Async|WaitGameplayEventToActor _returnvalue "(TagName=\"Event.Ability.FeyCrab.Proj.Explode\")"
              (:then
                (GameplayEffects|ApplyGameplayEffectToSelf :self (Ability|GetAbilitySystemComponent _returnvalue) :GameplayEffectClass "/DryBox/Items/FeyCrab/GE_Grant_PreventActions.GE_Grant_PreventActions_C" :Level 0.0))
              (:EventReceived
                (Class|BFeyCrabLauncher|RemoveProjectilefromWorld (Variables|Default|GetFey%20Crab%20Launcher%20Ref))
                (Ability|EndAbility)))))))))

(event Ability|EventOnEndAbility
  (params (bWasCancelled :type Boolean)))
