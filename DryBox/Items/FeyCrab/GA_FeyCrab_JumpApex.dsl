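; Gameplay ability for the FeyCrab item that handles the apex of a jump when triggered by a gameplay event.
; On activation it applies fall immunity to the pawn, then uses a knockback tag check to gate the in-air velocity scaling and apex gravity logic.
; The apex logic applies a low-gravity GameplayEffect and uses a half-second timer to switch character movement to Falling.
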(event Ability|EventOnEndAbility
  (params (bWasCancelled :type Boolean)))

(event Ability|EventActivateAbilityFromEvent
  (params (EventData :type Gameplay%20Event%20Data%20Structure%20%28by%20ref%29))
  (bind _returnvalue (Ability|GetActivatingPawn))
  ;[1 1 1] Jump Apex
  (|CallParentFunction EventData)
  (GameplayEffects|ApplyGameplayEffectToSelf :self (Ability|GetFortAbilitySystemComponent _returnvalue) :GameplayEffectClass (Variables|Default|GetGE_FallImmunity) :Level 0.0)
  (if (not (GameplayTags|HasMatchingGameplayTag _returnvalue (Variables|Default|GetKnockbagTag)))
    (Variables|/Script/Engine.MovementComponent|SetVelocity (* (Variables|/Script/Engine.MovementComponent|GetVelocity (Variables|/Script/Engine.Character|GetCharacterMovement (Ability|GetActivatingPawn))) (ScalableFloat|GetValueatLevel (Variables|Default|GetInAirVelocityMulitiplier))) (Variables|/Script/Engine.Character|GetCharacterMovement (Ability|GetActivatingPawn)))
    (CallFunction|ApexGravityMulti)))

(event Custom|ApexGravityMulti
  ;[1 1 1] Change gravity at the apex
  (Utilities|IsValid (Ability|GetActivatingPawn)
    (:Is%20Valid
      (Utilities|FlowControl|Sequence
        (:then_0
          (CallFunction|ApplyLowGrav))
        (:then_1
          (Utilities|Time|SetTimerbyEvent (delegate Custom|SetFalling) 0.5))))))

(event Custom|SetFalling
  ;[1 1 1] Change gravity at the apex
  (Pawn|Components|CharacterMovement|SetMovementMode (Variables|/Script/Engine.Character|GetCharacterMovement (Ability|GetActivatingPawn)) "MOVE_Falling" 0))

(fn ApplyLowGrav
  (GameplayEffects|ApplyGameplayEffectToSelf :self (Ability|GetAbilitySystemComponent (Ability|GetActivatingPawn)) :GameplayEffectClass "/DryBox/Items/FeyCrab/GE_Feycrab_LowGrav.GE_Feycrab_LowGrav_C" :Level 0.0))
