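; FeyCrab item gameplay ability that, on a gameplay event, knocks the activating character backward, apparently as in-air firing recoil.
; Launch opposes camera forward, or pawn forward when the camera points upward, so recoil never pushes downward.
; An aim-axis line trace selects indoor-scaled strength; otherwise a swinging tag selects its own multiplier over base strength.
; Launch and local knockback cue are skipped when the pawn has upper-body-only tags.
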(event Ability|EventOnEndAbility
  (params (bWasCancelled :type Boolean)))

(event Ability|EventActivateAbilityFromEvent
  (params (EventData :type Gameplay%20Event%20Data%20Structure%20%28by%20ref%29))
  (bind _returnvalue (Ability|GetActivatingPawn))
  (|CallParentFunction EventData)
  (bind (_outhit _returnvalue_2) (Collision|LineTraceForObjects (Transformation|GetWorldLocation (Variables|/Script/Engine.Character|GetCapsuleComponent (Ability|GetActivatingPawn))) (+ (Transformation|GetWorldLocation (Variables|/Script/Engine.Character|GetCapsuleComponent (Ability|GetActivatingPawn))) (* (Math|Vector|GetForwardVector (Pawn|GetBaseAimRotation (Ability|GetActivatingPawn))) (Math|Conversions|ToVector(Float) (Variables|Default|GetBuildingLineTraceOffset)))) (Variables|Default|GetObject%20Types) false (Utilities|Array|MakeArray (Ability|GetActivatingPawn))))
  (Variables|Default|SetBackBehindWall _returnvalue_2)
  (if (not (GameplayTags|HasAnyMatchingGameplayTags (Ability|GetActivatingPawn) (Variables|Default|GetPlay%20Upper%20Body%20Only%20Tags)))
    (bind _playercameramanager (Variables|/Script/Engine.PlayerController|GetPlayerCameraManager (Fortnite|GetFortPlayerControllerfromActor _returnvalue))
      (:then
        (Character|LaunchCharacter (Ability|GetActivatingPawn) (Utilities|Operators|Multiply (Math|Vector|SelectVector (Math|Vector|GetForwardVector (Transformation|GetWorldRotation (Variables|/Script/Engine.Actor|GetRootComponent _returnvalue))) (Math|Vector|GetForwardVector (Camera|GetCameraRotation _playercameramanager)) (> (.z (Math|Vector|GetForwardVector (Camera|GetCameraRotation _playercameramanager))) 0)) (Math|Float|SelectFloat (* (ScalableFloat|GetValueatLevel (Variables|Default|GetLaunchStrength)) (ScalableFloat|GetValueatLevel (Variables|Default|GetLaunchStrengthIndoorMultiplier))) (Math|Float|SelectFloat (* (ScalableFloat|GetValueatLevel (Variables|Default|GetLaunchStrength)) (ScalableFloat|GetValueatLevel (Variables|Default|GetLaunchStrengthSwingingMultiplier))) (ScalableFloat|GetValueatLevel (Variables|Default|GetLaunchStrength)) (GameplayTags|HasMatchingGameplayTag (Ability|GetActivatingPawn) (Variables|Default|GetSwingingTag))) (Variables|Default|GetBackBehindWall)) -1.0) true false)
        (GameplayCue|ExecuteGameplayCue%28Burst%29Local (Ability|GetActivatingPawn) (Variables|Default|GetOnFireInAirKnockbackCueTag) (Ability|GameplayCue|MakeGameplayCueParameters 0.0)))
      (:else))))
