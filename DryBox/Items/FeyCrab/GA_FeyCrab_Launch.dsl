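; Fey Crab item gameplay ability that, on event activation, launches the activating character or alters its movement.
; After the parent logic, it traces from the capsule along a height offset.
; Without "play upper body only" tags, it launches the pawn vertically, with zero strength if the trace hits.
; With those tags, a hit scales current velocity and a miss applies a low-gravity gameplay effect.
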(event Ability|EventOnEndAbility
  (params (bWasCancelled :type Boolean)))

(event Ability|EventActivateAbilityFromEvent
  (params (EventData :type Gameplay%20Event%20Data%20Structure%20%28by%20ref%29))
  (|CallParentFunction EventData)
  (bind (_outhit _returnvalue) (Collision|LineTraceForObjects (Transformation|GetWorldLocation (Variables|/Script/Engine.Character|GetCapsuleComponent (Ability|GetActivatingPawn))) (+ (Transformation|GetWorldLocation (Variables|/Script/Engine.Character|GetCapsuleComponent (Ability|GetActivatingPawn))) (Variables|Default|GetHeightOffset)) (Variables|Default|GetObject%20Types) false (Utilities|Array|MakeArray (Ability|GetActivatingPawn)) "ForDuration"))
  (if (not (GameplayTags|HasAnyMatchingGameplayTags (Ability|GetActivatingPawn) (Variables|Default|GetPlay%20Upper%20Body%20Only%20Tags)))
    (Character|LaunchCharacter (Ability|GetActivatingPawn) (* "0.000000,0.000000,1.000000" (Math|Float|SelectFloat 0.0 (ScalableFloat|GetValueatLevel (Variables|Default|GetKnockUp%20Strength)) _returnvalue)) false true)
    (if _returnvalue
      (Variables|/Script/Engine.MovementComponent|SetVelocity (* (Variables|/Script/Engine.MovementComponent|GetVelocity (Variables|/Script/Engine.Character|GetCharacterMovement (Ability|GetActivatingPawn))) (ScalableFloat|GetValueatLevel (Variables|Default|GetInAirVelocityMulitiplier))) (Variables|/Script/Engine.Character|GetCharacterMovement (Ability|GetActivatingPawn)))
      (GameplayEffects|ApplyGameplayEffectToSelf :self (Ability|GetAbilitySystemComponent (Ability|GetActivatingPawn)) :GameplayEffectClass (Variables|Default|GetGE_LowGravity) :Level 0.0))))
