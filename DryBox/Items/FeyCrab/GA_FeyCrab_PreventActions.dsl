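; GA_FeyCrab_PreventActions is a Gameplay Ability among the Fey Crab item assets that serves to prevent actions.
; Activation only calls the parent implementation, which a comment notes is needed for the prevention to work.
; The end event is empty and the graph holds no other logic, so any blocking comes from outside it.
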(event Ability|EventActivateAbility
  ;[1 1 1] Need this to prevent actions
  (|CallParentFunction))

(event Ability|EventOnEndAbility
  (params (bWasCancelled :type Boolean)))
