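; FeyCrab gameplay ability that, by its name and comment, prevents reloading while the player is swinging.
; It is activated from a gameplay event and calls the parent activation.
; A WaitGameplayTagRemove task watches the MovementMode.Swinging tag and ends the ability when that tag is removed.
; The reload block itself is not in this graph, which handles only the ability's lifetime.
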(event Ability|EventActivateAbilityFromEvent
  (params (EventData :type Gameplay%20Event%20Data%20Structure%20%28by%20ref%29))
  ;[1 1 1] When swinging tag is removed end ability that prevents reloading
  (|CallParentFunction EventData)
  (Ability|Tasks|WaitGameplayTagRemove "(TagName=\"MovementMode.Swinging\")"
    (:Removed
      (Ability|EndAbility))))

(event Ability|EventOnEndAbility
  (params (bWasCancelled :type Boolean)))
