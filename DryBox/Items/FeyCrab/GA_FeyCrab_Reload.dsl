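; Gameplay ability for the Fey Crab Launcher that defers reloading or removing the weapon until the activating pawn is free.
; On event activation it caches the current launcher and starts a looping TryReload timer behind a DoOnce.
; Each tick stops the timer if already reloading; otherwise it waits for no active montage, no projectile in the world and no Swinging tag.
; It then calls the launcher's ReloadAmmo and RemoveWeapon and clears the timer.
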(event Ability|EventActivateAbilityFromEvent
  (params (EventData :type Gameplay%20Event%20Data%20Structure%20%28by%20ref%29))
  ;[1 1 1] Reload/Remove wepaon if magazine and ammo is 0 // check against in projectiles in world and if a montage is currently playing/swinging
  (bind _asb_fey_crab_launcher (Utilities|Casting|CastToB_FeyCrab_Launcher (Weapon|GetCurrentWeapon (Ability|GetActivatingPawn))))
  (Variables|Default|SetFey%20Crab%20Launcher%20Ref _asb_fey_crab_launcher)
  (Utilities|FlowControl|DoOnce false
    (:Completed
      (bind _returnvalue (Utilities|Time|SetTimerbyEvent (delegate Custom|TryReload) (Variables|Default|GetReloadTimerCheck) true))
      (Variables|Default|SetTH_ReloadCheck _returnvalue))))

(event Custom|TryReload
  (bind _fey_crab_launcher_ref (Variables|Default|GetFey%20Crab%20Launcher%20Ref))
  ;[1 1 1] Reload/Remove wepaon if magazine and ammo is 0 // check against in projectiles in world and if a montage is currently playing/swinging
  (if (FortWeapon|IsReloading (Variables|Default|GetFey%20Crab%20Launcher%20Ref))
    (Utilities|Time|ClearandInvalidateTimerbyHandle (Variables|Default|GetTH_ReloadCheck))
    (elif (Math|Boolean|ANDBoolean (== (Animation|GetCurrentMontage (Ability|GetActivatingPawn) :!class /Script/Engine.Character) 0) (== (Variables|/DryBox/Items/FeyCrab/B_FeyCrab_Launcher.B_FeyCrab_Launcher_C|GetProjectileInWorld (Variables|Default|GetFey%20Crab%20Launcher%20Ref)) 0) (not (GameplayTags|HasMatchingGameplayTag (Ability|GetActivatingPawn) "(TagName=\"MovementMode.Swinging\")")))
      (Class|BFeyCrabLauncher|ReloadAmmo _fey_crab_launcher_ref)
      (Class|BFeyCrabLauncher|RemoveWeapon _fey_crab_launcher_ref)
      (Utilities|Time|ClearandInvalidateTimerbyHandle (Variables|Default|GetTH_ReloadCheck)))))
