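; Gameplay ability for the Fey Crab item, named for fall immunity until landing and weapon holstering.
; The activate event only holds a disabled parent call, so its logic comes from the parent class.
; On end, it calls the parent, casts the avatar to FortPawn and calls SetHolsterWeapon with false to undo the holster tagged "GrappleParent".
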(event Ability|EventActivateAbility
  (|CallParentFunction) ; This node is disabled and will not be called.\nDrag off pins to build functionality.
)

(event Ability|EventOnEndAbility
  (params (bWasCancelled :type Boolean))
  (|CallParentFunction bWasCancelled)
  (bind _asfort_pawn (Utilities|Casting|CastToFortPawn (Ability|GetAvatarActorfromActorInfo)))
  (Weapon|SetHolsterWeapon _asfort_pawn 0 false false "GrappleParent"))
