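# Physics asset for the FeyCrab Launcher mesh, an item in DryBox.
# It holds a single capsule body on the Melee_TwoHanded_Handle bone, offset and tilted from the bone origin.
# It has no constraints, so the launcher is represented as one rigid shape.
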PhysicsAsset /DryBox/Items/FeyCrab/Meshes/FeyCrab_Launcher_PhysicsAsset
bodies {Melee_TwoHanded_Handle [{shapeType=Capsule, length=75.855, radius=20.8547, center=<0.0443 -19.8194 -14.1041>, rotation=<rot 0 -179.9784 -83.2695>}]}
constraints []
